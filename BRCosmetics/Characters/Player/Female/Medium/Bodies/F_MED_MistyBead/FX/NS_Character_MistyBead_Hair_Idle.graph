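# Niagara idle effect for the MistyBead character's hair, emitting cyan water droplets and splashes from sampled sockets.
# System Timeline time drives spawn bursts in one droplet and two splash sprite emitters, over a constant base rate for droplets only.
# Hidden head suppresses regular droplet spawning; DebugReactivity adds spawning to all three.
# Front-end state switches colors, FrontendScale scales gravity, and splashes add time-of-day tint, platform-dependent front-end brightness and dynamic material parameters.
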NiagaraSystem /BRCosmetics/Characters/Player/Female/Medium/Bodies/F_MED_MistyBead/FX/NS_Character_MistyBead_Hair_Idle
dependencies {
 usedDataInterfaces [
  &NiagaraDataInterfaceCurve
  &NiagaraDataInterfaceSpriteRendererInfo
  &NiagaraDataInterfaceColorCurve
  &NiagaraDataInterfacePlatformSet
  &NiagaraDataInterfaceArrayInt32
  &NiagaraDataInterfaceSocketReader
 ]
 usedDynamicInputs [
  &FN_FrontEndFloatModifier
  &SetBoolByFloatComparison
  &Boolean_AndOperation
  &SetBoolByIntComparison
  &FNTimeOfDayColor
 ]
 usedModules [
  &self:SystemSpawnScript.SystemScriptSource.SystemScriptGraph.NiagaraNodeAssignment_0.SetVariables_#1
  &self:SystemSpawnScript.SystemScriptSource.SystemScriptGraph.NiagaraNodeAssignment_2.SetVariables_#2
  &self:SystemSpawnScript.SystemScriptSource.SystemScriptGraph.NiagaraNodeAssignment_1.SetVariables_#3
  &SystemState
  /Game/Effects/Niagara/Modules/Logic/UpdateValueStateChange
  &Timeline
  &self:SystemSpawnScript.SystemScriptSource.SystemScriptGraph.NiagaraNodeAssignment_3.SetVariables_#4
 ]
}
emitters {
 Droplets {
  emitterUpdateScript [
   &EmitterState()
   &SpawnRate(
    Spawn Probability=&MakeCustomFloatFromBool(Bool=$User.bIsHeadHidden, True Float=0, False Float=0.6)
    SpawnRate=&Add_Float(
     A=&FloatFromCurve(
      FloatCurve=DataInterface() {propertyValues {Curve {keys [
          {arriveTangent=6.637, interpMode=RCIM_Cubic, leaveTangent=6.637, tangentMode=RCTM_User, time=0.04, value=0}
          {arriveTangent=16.8545, interpMode=RCIM_Cubic, leaveTangent=-3.293, tangentMode=RCTM_Break, time=0.1, value=1}
          <at 0.9 0>
         ]}}}
      CurveIndex=$System.Timeline.Time
      Scale Curve=100
     )
     B=10
    )
    Use Spawn Probability=true
   )
   &SpawnRate(
    SpawnRate=&MakeCustomFloatFromBool(Bool=$User.DebugReactivity, True Float=100, False Float=0)
    Use Spawn Probability=true
   ) {module=SpawnRate001}
  ]
  particleSpawnScript [
   &FN_SetSpawnGroup(Number of Spawn Groups=2)
   &InitializeParticle(
    Color=&MakeLinearColorFromVectorAndFloat(
     "Vector (RGB)"=&FN_FrontEndVector3Modifier(
      Gameplay=&MakeVectorFromLinearColorRGB(Color=&RandomRangeLinearColor(Minimum=<rgba 0.07 0.845 1 1>, Maximum=<rgba 0.15 0.7167 1 1>))
      Front End=&MakeVectorFromLinearColorRGB(Color=&RandomRangeLinearColor(Minimum=<rgba 0.2696 0.9047 1 1>, Maximum=<rgba 0.3304 1.1725 2 1>))
     )
     "Float (Alpha)"=1
    )
    Lifetime Max=0.5
    Lifetime Min=0.3
    Lifetime Mode=::NewEnumerator1
    Sprite Rotation Mode=::NewEnumerator2
    Sprite Size Mode=::NewEnumerator3
    Uniform Sprite Size=&RandomRangeFloat(Minimum=1, Maximum=1.5)
    UsePositionOffset=false
    Write Material Random=true
   )
   &SocketLocation(
    "Inherit Velocity (Scale)"=0.1
    Inherited Velocity Speed Limit=300000
    Interpolate Socket Sampling=true
    SamplePosOffset=true
    Sampled Orientation Coordinate Space:ENiagaraCoordinateSpace=::Local
    Sampled Position Offset=&MaskVectorBySpawnGroup(Masked Vector=<-15 -10 0>, Passthrough Vector=<5 -5 0>)
    Sampled Velocity Coordinate Space:ENiagaraCoordinateSpace=::Simulation
    Socket Reader=$System.Socket Reader
    Socket Sampling Mode=::NewEnumerator3
    VelocitySpeedLimit=false
   )
   &SubUVAnimation(
    Evaluation Type=::NewEnumerator0
    "Sprite Renderer "=DataInterface() {propertyValues {SpriteRenderer=&self:Droplets_1.NiagaraSpriteRendererProperties_0}}
    SubUV Animation Mode:ENiagara_SubUVLookupModeV2=::NewEnumerator2
    UseEndFrame=false
    UseStartFrame=false
   )
   &ShapeLocation(
    Box Size=<15 20 20>
    MaskSpawnGroup=true
    Override Seed=false
    Rotation Axis=&RotateVectorByQuaternion(Vector To Rotate=<1 0.5 0>, Quaternion=$Output.SocketLocation.SampledRotation)
    Rotation Mode=::NewEnumerator1
    Shape Primitive=::NewEnumerator2
    Surface Only=false
    UseOwnerScale=true
   )
   &ShapeLocation(
    MaskSpawnGroup=true
    Non Uniform Scale=<1 1 1.5>
    Override Seed=false
    Rotation Axis=&RotateVectorByQuaternion(Vector To Rotate=<1 0.5 0>, Quaternion=$Output.SocketLocation.SampledRotation)
    Rotation Mode=::NewEnumerator1
    Spawn Group Mask=1
    Sphere Radius=6
    UseOwnerScale=true
   ) {module=ShapeLocation001}
   &AddVelocity(
    Module.Use Falloff Distance=false
    Origin Offset=<0 0 130>
    Override Seed=false
    Velocity Mode=::NewEnumerator1
    Velocity Speed=&RandomRangeFloat(Minimum=10, Maximum=15)
   )
   &CameraOffset(Camera Offset Amount=5)
  ]
  particleUpdateScript [
   &ParticleState()
   &SocketLocation(
    Inherited Velocity Speed Limit=300000
    Interpolate Socket Sampling=true
    Orientation Sampling=::NewEnumerator1
    Position Sampling=::NewEnumerator1
    SamplePosOffset=true
    Sampled Orientation Coordinate Space:ENiagaraCoordinateSpace=::Local
    Sampled Velocity Coordinate Space:ENiagaraCoordinateSpace=::Simulation
    Socket Reader=$System.Socket Reader
    Socket Sampling Mode=::NewEnumerator3
    VelocitySpeedLimit=false
   ) {module=SocketLocation001}
   &ScaleSpriteSize(Uniform Curve Sprite Scale=DataInterface() {propertyValues {Curve {keys [
        {interpMode=RCIM_Cubic, tangentMode=RCTM_User, time=0.33, value=1}
        {arriveTangent=-2.8079, interpMode=RCIM_Cubic, leaveTangent=-2.8079, tangentMode=RCTM_User, time=1, value=0}
       ]}}})
   &Drag(Drag=&RandomRangeFloat(Minimum=3, Maximum=4), Use Linear Drag=true, Use Rotational Drag=true)
   &GravityForce(Gravity=&Multiply_VectorByFloat(Vector=&RandomRangeVector(Minimum=<0 0 -200>, Maximum=<0 0 -300>), Float=$System.FrontendScale))
   &InheritSourceMovement(
    Applied Position Scale=&VectorFromFloat(Value=&FloatFromCurve(FloatCurve=DataInterface() {propertyValues {Curve {defaultValue=1, keys [{interpMode=RCIM_Cubic, tangentMode=RCTM_User, time=0.3, value=1}]}}}))
    ApplyForce=false
    ApplyPosition=true
    ApplyPositionFalloff=false
    Source Velocity:Vector3f=$Output.SocketLocation001.SampledVelocity
   )
   &SolveForcesAndVelocity(Clamp Velocity=false, Limit Acceleration=false)
   &ScaleSpriteSizeBySpeed(Max Scale Factor=<0.8 2.5>, Velocity Threshold=0)
   &ScaleColor(
    Scale RGB=&MakeVectorFromLinearColorRGB(Color=&ColorFromCurve(ColorCurve=DataInterface() {propertyValues {
        AlphaCurve {keys [<at 0 1 RCIM_Cubic>]}
        BlueCurve {keys [<at 0 1 RCIM_Cubic>,<at 0.9829 0.8>]}
        GreenCurve {keys [<at 0 1 RCIM_Cubic>,<at 0.9829 0.68>]}
        RedCurve {keys [<at 0 1 RCIM_Cubic>,<at 0.9829 0.56>]}
       }}))
    ScaleA=true
    ScaleRGB=true
   )
  ]
  rendererClasses [&NiagaraSpriteRendererProperties]
  renderers [{rendererClass=&NiagaraSpriteRendererProperties}]
 }
 SplashSprites {
  emitterUpdateScript [
   &EmitterState()
   &SpawnRate(
    Spawn Probability=0.6
    SpawnRate=&Add_Float(
     A=&FloatFromCurve(
      FloatCurve=DataInterface() {propertyValues {Curve {keys [
          {arriveTangent=6.637, interpMode=RCIM_Cubic, leaveTangent=6.637, tangentMode=RCTM_User, time=0.04, value=0}
          {arriveTangent=16.8545, interpMode=RCIM_Cubic, leaveTangent=-3.293, tangentMode=RCTM_Break, time=0.1, value=1}
          <at 0.9 0>
         ]}}}
      CurveIndex=$System.Timeline.Time
      Scale Curve=100
     )
     B=0
    )
    Use Spawn Probability=true
   )
   &SpawnRate(
    SpawnRate=&MakeCustomFloatFromBool(Bool=$User.DebugReactivity, True Float=50, False Float=0)
    Use Spawn Probability=true
   ) {module=SpawnRate001}
  ]
  particleSpawnScript [
   &InitializeParticle(
    Color=&MakeLinearColorFromVectorAndFloat(
     "Vector (RGB)"=&FN_FrontEndVector3Modifier(
      Gameplay=&MakeVectorFromLinearColorRGB(Color=&RandomRangeLinearColor(Minimum=<rgba 0.1439 0.8594 1 1>, Maximum=<rgba 0.1151 0.6288 0.8 1>))
      Front End=&MakeCustomVectorFromBool(
       Bool=&IsPlatformSetActive(New Platform Set=DataInterface() {propertyValues {Platforms {qualityLevelMask=-4}}})
       True Vector=&MakeVectorFromLinearColorRGB(Color=&RandomRangeLinearColor(Minimum=<rgba 1.6174 5.2803 6 1>, Maximum=<rgba 2.087 6.5217 8 1>))
       False Vector=&MakeVectorFromLinearColorRGB(Color=&RandomRangeLinearColor(Minimum=<rgba 0.1439 0.8594 1 1>, Maximum=<rgba 0.1151 0.6288 0.8 1>))
      )
     )
     "Float (Alpha)"=1
    )
    Lifetime Max=0.8
    Lifetime Min=0.5
    Lifetime Mode=::NewEnumerator1
    Position Mode=::NewEnumerator0
    Sprite Rotation Mode=::NewEnumerator2
    Sprite Size Mode=::NewEnumerator1
    Sprite UV Mode=::NewEnumerator3
    Uniform Sprite Size Max=50
    Uniform Sprite Size Min=40
    Write Material Random=true
   )
   /SetVariables(Particles.SubImageIndex=&RandomRangeFloat(Minimum=0, Maximum=3)) {isSetParameters=true, module=SetVariables_#5}
   &SocketLocation(
    Inherited Velocity Speed Limit=300000
    Interpolate Socket Sampling=true
    SamplePosOffset=true
    Sampled Position Offset=&Add_Vector(A=<-5 -10 0>, B=&RandomRangeVector(Minimum=<-20 -5 -10>, Maximum=<20 5 10>))
    Sampled Velocity Coordinate Space:ENiagaraCoordinateSpace=::Simulation
    Socket Reader=$System.Socket Reader
    Socket Sampling Mode=::NewEnumerator3
    VelocitySpeedLimit=false
   )
   &AddVelocity(
    Module.Use Falloff Distance=false
    Origin Offset=<0 0 130>
    Override Seed=false
    Velocity Mode=::NewEnumerator1
    Velocity Speed=&RandomRangeFloat(Minimum=10, Maximum=15)
   )
   &CameraOffset(Camera Offset Amount=10)
  ]
  particleUpdateScript [
   &ParticleState()
   &SpriteRotationRate(Rotation Rate=&FloatFromCurve(
     FloatCurve=DataInterface() {propertyValues {Curve {
        defaultValue=1
        keys [{interpMode=RCIM_Cubic, tangentMode=RCTM_User, time=0, value=1}, {interpMode=RCIM_Cubic, tangentMode=RCTM_User, time=1, value=0}]
       }}}
     Scale Curve=&Multiply_FloatByInt(
      Float=&RandomRangeFloat(Minimum=60, Maximum=100)
      Integer=&SelectIntFromArray(
       Array Sampling Mode=::NewEnumerator1
       Direct Array Index=?
       Int Selection Array=DataInterface() {propertyValues {IntData [-1,1]}}
       Random Seed=0
      )
     )
    ))
   &GravityForce(Gravity=&Multiply_VectorByFloat(
     Vector=&Multiply_VectorByFloat(Vector=<0 0 -300>, Float=&RandomRangeFloat(Minimum=0.7, Maximum=1))
     Float=$System.FrontendScale
    ))
   &Drag(Drag=&RandomRangeFloat(Minimum=4, Maximum=5), Use Linear Drag=true, Use Rotational Drag=true)
   &SolveForcesAndVelocity(Clamp Velocity=false, Limit Acceleration=false)
   &ScaleSpriteSize(
    Uniform Curve Scale=$System.FrontendScale
    Uniform Curve Sprite Scale=DataInterface() {propertyValues {Curve {keys [
        {arriveTangent=1.7973, interpMode=RCIM_Cubic, leaveTangent=1.7973, tangentMode=RCTM_User, time=0, value=0.1}
        {interpMode=RCIM_Cubic, tangentMode=RCTM_User, time=1, value=1}
       ]}}}
   )
   &ScaleColor(
    Scale Alpha=&FloatFromCurve(FloatCurve=DataInterface() {propertyValues {Curve {
        defaultValue=1
        keys [{interpMode=RCIM_Cubic, tangentMode=RCTM_User, time=0, value=1}, {interpMode=RCIM_Cubic, tangentMode=RCTM_User, time=1, value=0}]
       }}})
    Scale RGB=&MakeVectorFromLinearColorRGB(Color=&ColorFromCurve(ColorCurve=DataInterface() {propertyValues {
        AlphaCurve {keys [<at 0 1 RCIM_Cubic>]}
        BlueCurve {keys [<at 0 1>,<at 1 0.4261>]}
        GreenCurve {keys [<at 0 1>,<at 1 0.3199>]}
        RedCurve {keys [<at 0 1>,<at 1 0.1074>]}
       }}))
    ScaleA=true
    ScaleRGB=true
   )
   &ScaleColor(
    Color Value To Scale=$Particles.Color
    Scale RGB=&FN_FrontEndVector3Modifier(
     Gameplay=&MakeVectorFromLinearColorRGB(Color=$System.TODColor)
     Front End=&MakeVectorFromLinearColorRGB(Color=<rgba 1 1 1 1>)
    )
    ScaleA=true
    ScaleRGB=true
   ) {module=ScaleColor001}
   &DynamicMaterialParameters(
    Index 0 Param 1=&RandomRangeFloat(Minimum=0, Maximum=1)
    Index 0 Param 2=&MakeCustomFloatFromBool(
     Bool=&IsPlatformSetActive(New Platform Set=DataInterface() {propertyValues {Platforms {qualityLevelMask=-4}}})
     True Float=&FloatFromCurve(
      FloatCurve=DataInterface() {propertyValues {Curve {
         defaultValue=1
         keys [{interpMode=RCIM_Cubic, tangentMode=RCTM_User, time=0, value=0.2}, {interpMode=RCIM_Cubic, tangentMode=RCTM_User, time=0.7, value=0}]
        }}}
      Scale Curve=&RandomRangeFloat(Minimum=0.2, Maximum=0.6)
     )
     False Float=1
    )
    Index 0 Param 3=&Lerp_Float(
     A=0
     B=3
     Alpha=&FloatFromCurve(FloatCurve=DataInterface() {propertyValues {Curve {
         defaultValue=0
         keys [
          {interpMode=RCIM_Cubic, tangentMode=RCTM_User, time=0, value=0}
          {arriveTangent=3.1748, interpMode=RCIM_Cubic, leaveTangent=3.1748, tangentMode=RCTM_User, time=1, value=1}
         ]
        }}})
    )
    Index 0 Param 4=&FloatFromCurve(FloatCurve=DataInterface() {propertyValues {Curve {
        defaultValue=0
        keys [
         {arriveTangent=2.4825, interpMode=RCIM_Cubic, leaveTangent=2.4825, tangentMode=RCTM_User, time=0, value=0}
         {interpMode=RCIM_Cubic, tangentMode=RCTM_User, time=1, value=1}
        ]
       }}})
    Module.Index0Param0Write=true
    Module.Index0Param1Write=true
    Module.Index0Param2Write=true
    Module.Index0Param3Write=true
   )
  ]
  rendererClasses [&NiagaraSpriteRendererProperties]
  renderers [{rendererClass=&NiagaraSpriteRendererProperties}]
 }
 SplashSprites02 {
  emitterUpdateScript [
   &EmitterState()
   &SpawnRate(
    Spawn Probability=0.6
    SpawnRate=&Add_Float(
     A=&FloatFromCurve(
      FloatCurve=DataInterface() {propertyValues {Curve {keys [
          {arriveTangent=6.637, interpMode=RCIM_Cubic, leaveTangent=6.637, tangentMode=RCTM_User, time=0.04, value=0}
          {arriveTangent=16.8545, interpMode=RCIM_Cubic, leaveTangent=-3.293, tangentMode=RCTM_Break, time=0.1, value=1}
          <at 0.9 0>
         ]}}}
      CurveIndex=$System.Timeline.Time
      Scale Curve=60
     )
     B=0
    )
    Use Spawn Probability=true
   )
   &SpawnRate(
    SpawnRate=&MakeCustomFloatFromBool(Bool=$User.DebugReactivity, True Float=50, False Float=0)
    Use Spawn Probability=true
   ) {module=SpawnRate001}
  ]
  particleSpawnScript [
   &InitializeParticle(
    Color=&MakeLinearColorFromVectorAndFloat(
     "Vector (RGB)"=&FN_FrontEndVector3Modifier(
      Gameplay=&MakeVectorFromLinearColorRGB(Color=&RandomRangeLinearColor(Minimum=<rgba 0.1439 0.8594 1 1>, Maximum=<rgba 0.1151 0.6288 0.8 1>))
      Front End=&MakeCustomVectorFromBool(
       Bool=&IsPlatformSetActive(New Platform Set=DataInterface() {propertyValues {Platforms {qualityLevelMask=-4}}})
       True Vector=&MakeVectorFromLinearColorRGB(Color=&RandomRangeLinearColor(Minimum=<rgba 1.6174 5.2803 6 1>, Maximum=<rgba 2.087 6.5217 8 1>))
       False Vector=&MakeVectorFromLinearColorRGB(Color=&RandomRangeLinearColor(Minimum=<rgba 0.1439 0.8594 1 1>, Maximum=<rgba 0.1151 0.6288 0.8 1>))
      )
     )
     "Float (Alpha)"=1
    )
    Lifetime Max=0.6
    Lifetime Min=0.4
    Lifetime Mode=::NewEnumerator1
    Position Mode=::NewEnumerator0
    Sprite Rotation Mode=::NewEnumerator2
    Sprite Size Mode=::NewEnumerator1
    Sprite UV Mode=::NewEnumerator3
    Uniform Sprite Size Max=40
    Uniform Sprite Size Min=35
    Write Material Random=true
   )
   /SetVariables(Particles.SubImageIndex=&RandomRangeFloat(Minimum=0, Maximum=3)) {isSetParameters=true, module=SetVariables_#5}
   &SocketLocation(
    Inherited Velocity Speed Limit=300000
    Interpolate Socket Sampling=true
    SamplePosOffset=true
    Sampled Position Offset=&Add_Vector(A=<-5 -10 0>, B=&RandomRangeVector(Minimum=<-15 -5 -10>, Maximum=<12 5 10>))
    Sampled Velocity Coordinate Space:ENiagaraCoordinateSpace=::Simulation
    Socket Reader=$System.Socket Reader
    Socket Sampling Mode=::NewEnumerator3
    VelocitySpeedLimit=false
   )
   &AddVelocity(
    Module.Use Falloff Distance=false
    Origin Offset=<0 0 130>
    Override Seed=false
    Velocity Mode=::NewEnumerator1
    Velocity Speed=&RandomRangeFloat(Minimum=10, Maximum=15)
   )
   &CameraOffset(Camera Offset Amount=10)
  ]
  particleUpdateScript [
   &ParticleState()
   &SpriteRotationRate(Rotation Rate=&FloatFromCurve(
     FloatCurve=DataInterface() {propertyValues {Curve {
        defaultValue=1
        keys [{interpMode=RCIM_Cubic, tangentMode=RCTM_User, time=0, value=1}, {interpMode=RCIM_Cubic, tangentMode=RCTM_User, time=1, value=0}]
       }}}
     Scale Curve=&Multiply_FloatByInt(
      Float=&RandomRangeFloat(Minimum=60, Maximum=100)
      Integer=&SelectIntFromArray(
       Array Sampling Mode=::NewEnumerator1
       Direct Array Index=?
       Int Selection Array=DataInterface() {propertyValues {IntData [-1,1]}}
       Random Seed=0
      )
     )
    ))
   &GravityForce(Gravity=&Multiply_VectorByFloat(
     Vector=&Multiply_VectorByFloat(Vector=<0 0 -300>, Float=&RandomRangeFloat(Minimum=0.7, Maximum=1))
     Float=$System.FrontendScale
    ))
   &Drag(Drag=&RandomRangeFloat(Minimum=4, Maximum=5), Use Linear Drag=true, Use Rotational Drag=true)
   &SolveForcesAndVelocity(Clamp Velocity=false, Limit Acceleration=false)
   &ScaleSpriteSize(
    Uniform Curve Scale=$System.FrontendScale
    Uniform Curve Sprite Scale=DataInterface() {propertyValues {Curve {keys [
        {arriveTangent=1.7973, interpMode=RCIM_Cubic, leaveTangent=1.7973, tangentMode=RCTM_User, time=0, value=0.1}
        {interpMode=RCIM_Cubic, tangentMode=RCTM_User, time=1, value=1}
       ]}}}
   )
   &ScaleColor(
    Scale Alpha=&FloatFromCurve(FloatCurve=DataInterface() {propertyValues {Curve {
        defaultValue=1
        keys [{interpMode=RCIM_Cubic, tangentMode=RCTM_User, time=0, value=1}, {interpMode=RCIM_Cubic, tangentMode=RCTM_User, time=1, value=0}]
       }}})
    Scale RGB=&MakeVectorFromLinearColorRGB(Color=&ColorFromCurve(ColorCurve=DataInterface() {propertyValues {
        AlphaCurve {keys [<at 0 1 RCIM_Cubic>]}
        BlueCurve {keys [<at 0 1>,<at 1 0.4261>]}
        GreenCurve {keys [<at 0 1>,<at 1 0.3199>]}
        RedCurve {keys [<at 0 1>,<at 1 0.1074>]}
       }}))
    ScaleA=true
    ScaleRGB=true
   )
   &ScaleColor(
    Color Value To Scale=$Particles.Color
    Scale RGB=&FN_FrontEndVector3Modifier(
     Gameplay=&MakeVectorFromLinearColorRGB(Color=$System.TODColor)
     Front End=&MakeVectorFromLinearColorRGB(Color=<rgba 1 1 1 1>)
    )
    ScaleA=true
    ScaleRGB=true
   ) {module=ScaleColor001}
   &DynamicMaterialParameters(
    Index 0 Param 1=&RandomRangeFloat(Minimum=0, Maximum=1)
    Index 0 Param 2=&MakeCustomFloatFromBool(
     Bool=&IsPlatformSetActive(New Platform Set=DataInterface() {propertyValues {Platforms {qualityLevelMask=-4}}})
     True Float=&FloatFromCurve(
      FloatCurve=DataInterface() {propertyValues {Curve {
         defaultValue=1
         keys [{interpMode=RCIM_Cubic, tangentMode=RCTM_User, time=0, value=0.2}, {interpMode=RCIM_Cubic, tangentMode=RCTM_User, time=0.7, value=0}]
        }}}
      Scale Curve=&RandomRangeFloat(Minimum=0.2, Maximum=0.6)
     )
     False Float=1
    )
    Index 0 Param 3=&Lerp_Float(
     A=0
     B=3
     Alpha=&FloatFromCurve(FloatCurve=DataInterface() {propertyValues {Curve {
         defaultValue=0
         keys [
          {interpMode=RCIM_Cubic, tangentMode=RCTM_User, time=0, value=0}
          {arriveTangent=3.1748, interpMode=RCIM_Cubic, leaveTangent=3.1748, tangentMode=RCTM_User, time=1, value=1}
         ]
        }}})
    )
    Index 0 Param 4=&FloatFromCurve(FloatCurve=DataInterface() {propertyValues {Curve {
        defaultValue=0
        keys [
         {arriveTangent=2.4825, interpMode=RCIM_Cubic, leaveTangent=2.4825, tangentMode=RCTM_User, time=0, value=0}
         {interpMode=RCIM_Cubic, tangentMode=RCTM_User, time=1, value=1}
        ]
       }}})
    Module.Index0Param0Write=true
    Module.Index0Param1Write=true
    Module.Index0Param2Write=true
    Module.Index0Param3Write=true
   )
  ]
  rendererClasses [&NiagaraSpriteRendererProperties]
  renderers [{rendererClass=&NiagaraSpriteRendererProperties}]
 }
}
summary {
 systemName=NS_Character_MistyBead_Hair_Idle
 emitters [
  {emitterName=Droplets, rendererClasses [&NiagaraSpriteRendererProperties]}
  {emitterName=SplashSprites, rendererClasses [&NiagaraSpriteRendererProperties]}
  {emitterName=SplashSprites02, rendererClasses [&NiagaraSpriteRendererProperties]}
 ]
 userVariables [
  User.bIsFrontEnd {type=NiagaraBool}
  User.bIsFrontEndPreview {type=NiagaraBool}
  User.bIsHeadHidden {type=NiagaraBool}
  User.bIsSwinging {type=NiagaraBool}
  User.DebugReactivity {type=NiagaraBool}
  User.IsReactive {type=NiagaraFloat, defaultValue=1, description=""}
  User.KillCount {type=NiagaraInt32}
 ]
}
