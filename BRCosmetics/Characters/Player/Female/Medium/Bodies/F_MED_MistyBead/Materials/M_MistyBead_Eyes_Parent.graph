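# Parent material for MistyBead character eyes, adding a kill-reactive glow that replaces emissive and darkens base color.
# It overrides the shared MF_eyes_2023 output, then passes it through MF_Eyelash and MF_TechArtOverrides.
# The glow is a powered, scaled reactive head emissive texture.
# Timestamp parameters drive a timed fade-in, hold and fade-out intensity, with KillCount selecting the envelope and holding full intensity above three kills.
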Material /BRCosmetics/Characters/Player/Female/Medium/Bodies/F_MED_MistyBead/Materials/M_MistyBead_Eyes_Parent
graph {
 properties {
  bUseMaterialAttributes=true
  bUsedWithMeshDeformer=true
  bUsedWithSkeletalMesh=true
  pixelDepthOffsetMode=PDOM_Legacy
  shadingModel=MSM_FromMaterialExpression
  usedShadingModels=Default Lit | Subsurface
 }
 nodes [
  Constant()
  n1 = GetMaterialAttributes(None=MaterialFunctionCall() {materialFunction=/Game/Global/Materials/Character_Parent_2023/Eye_Parent_2023/MF_eyes_2023}.Result) {attributeGetTypes [#1, #2]}
  n2 = NamedRerouteUsage()
  n3 = NamedRerouteUsage()
  MaterialFunctionCall(In=MaterialFunctionCall!(Material=SetMaterialAttributes!(MaterialAttributes=@n1.MaterialAttributes, Emissive Color=@n2, Base Color=@n3) {attributeSetTypes [#1, #2]}) {materialFunction=/Game/Global/Materials/Character_Parent_2023/Eye_Parent_2023/MF_Eyelash}.Result) {materialFunction=&MF_TechArtOverrides}
  NamedRerouteDeclaration(None=@n1.Emissive Color) {name=InEmissive, nodeColor=<rgba 0 1 0.8125 1>}
  n4 = Power(
   Base=TextureSample() {texture=/BRCosmetics/Characters/Player/Female/Medium/Bodies/F_MED_MistyBead/Skins/Reactive/Textures/T_MistyBeadReactive_Head_E}.RGB
   Exp=ScalarParameter() {defaultValue=3.1215, group=MistyBead|Reactivity, parameterName=ReactiveEyesFresnelExp, sortPriority=41}
  )
  n5 = ScalarParameter() {defaultValue=4.2378, group=MistyBead|Reactivity, parameterName=ReactiveEyesIntensity}
  NamedRerouteDeclaration(None=LinearInterpolate(A=NamedRerouteUsage(), B=Multiply(A=@n4, B=@n5), Alpha=NamedRerouteUsage())) {name=OutEmissive, nodeColor=<rgba 1 0.2344 0 1>}
  n6 = ScalarParameter() {group=MistyBead|Reactivity, parameterName=KillCount}
  n7 = Time()
  n8 = ScalarParameter() {defaultValue=-999, group=MistyBead|Reactivity, parameterName=Timestamp}
  n9 = ScalarParameter() {defaultValue=1, group=MistyBead|Reactivity, parameterName=TimestampIn}
  n10 = Saturate(None=Divide(A=Subtract(A=@n7, B=@n8), B=@n9) {desc=InLength})
  n11 = Time()
  n12 = Saturate(None=Sine(None=Multiply(A=@n11) {constB=0.2}))
  n13 = Add(A=@n9, B=ScalarParameter() {defaultValue=0.75, group=MistyBead|Reactivity, parameterName=TimestampStay}) {constB=3, desc=eXTRAdURATION}
  n14 = Divide(A=Subtract(A=@n7, B=Add(A=@n8, B=@n13)), B=ScalarParameter() {defaultValue=1, group=MistyBead|Reactivity, parameterName=TimestampOut}) {constB=3, desc=OutLength}
  n15 = NamedRerouteDeclaration(None=OneMinus(None=Saturate(None=@n14))) {name=TimestampOut, nodeColor=<rgba 0 0.0859 1 1>}
  n16 = Multiply(A=@n10, B=@n15)
  n17 = If(
   A=@n6
   "A > B"=Constant() {r=1}
   "A == B"=@n10
   "A < B"=Multiply(
    A=StaticSwitchParameter(True=@n12, False=@n16) {group=ZenCrash_LandFoil, parameterName=DebugTimestamp}
    B=ScalarParameter() {defaultValue=1, group=MistyBead|Reactivity, parameterName=Reactivity - UseFadeInOut}
   )
  ) {constB=3}
  NamedRerouteDeclaration(None=If(A=@n6, "A > B"=@n17, "A == B"=@n15, "A < B"=@n17)) {name=FXIntensity, nodeColor=<rgba 0 0.0859 1 1>}
  NamedRerouteDeclaration(None=@n1.Base Color) {name=InBaseColor, nodeColor=<rgba 0.4062 1 0 1>}
  NamedRerouteDeclaration(None=LinearInterpolate(A=NamedRerouteUsage(), Alpha=NamedRerouteUsage()) {constB=0}) {name=OutBaseColor, nodeColor=<rgba 1 0 0.7266 1>}
  ScalarParameter() {defaultValue=1, group=MistyBead|Reactivity, parameterName=FXIntensity}
 ]
}
