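# Material function for the MistyBead character body that layers a masked, animated water and foam effect over an incoming material.
# Channel selections from the MaskMap and FXMask inputs separate water, facial, foam gradient and forehead symbol areas, feeding a final material attribute blend.
# Pre-skinned projected textures and normals pan over time, alongside refraction, Fresnel and GGX specular terms.
# A custom node evaluates two-band spherical harmonics for a directional term.
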MaterialFunction /BRCosmetics/Characters/Player/Female/Medium/Bodies/F_MED_MistyBead/Materials/MF_MistyBead
graph {expressions {
  MaterialExpressionAbs_6 {class=MaterialExpressionAbs, inputs [None {from=MaterialExpressionMultiply_141, output=""}], properties {}}
  MaterialExpressionAdd_0 {
   class=MaterialExpressionAdd
   inputs [
    A {from=MaterialExpressionAdd_1}
    B {from=MaterialExpressionMultiply_3}
   ]
  }
  MaterialExpressionAdd_1 {
   class=MaterialExpressionAdd
   inputs [
    A {from=MaterialExpressionAdd_3}
    B {from=MaterialExpressionLocalPosition_0, output=XYZ}
   ]
  }
  MaterialExpressionAdd_10 {
   class=MaterialExpressionAdd
   inputs [
    A {from=MaterialExpressionNamedRerouteUsage_8}
    B {from=MaterialExpressionMultiply_0}
   ]
  }
  MaterialExpressionAdd_11 {
   class=MaterialExpressionAdd
   inputs [
    A {from=MaterialExpressionMultiply_21}
    B {from=MaterialExpressionMultiply_82}
   ]
  }
  MaterialExpressionAdd_12 {
   class=MaterialExpressionAdd
   inputs [
    A {from=MaterialExpressionAdd_13}
    B {from=MaterialExpressionMultiply_22}
   ]
  }
  MaterialExpressionAdd_13 {
   class=MaterialExpressionAdd
   inputs [
    A {from=MaterialExpressionAdd_2}
    B {from=MaterialExpressionMultiply_42}
   ]
  }
  MaterialExpressionAdd_14 {
   class=MaterialExpressionAdd
   inputs [
    A {from=MaterialExpressionScalarParameter_77}
    B {from=MaterialExpressionScalarParameter_78}
   ]
   properties {constB=3, desc=eXTRAdURATION}
  }
  MaterialExpressionAdd_15 {
   class=MaterialExpressionAdd
   inputs [
    A {from=MaterialExpressionNamedRerouteUsage_98}
    B {from=MaterialExpressionMultiply_40}
   ]
  }
  MaterialExpressionAdd_16 {
   class=MaterialExpressionAdd
   inputs [
    A {from=MaterialExpressionSubtract_17}
    B {from=MaterialExpressionNamedRerouteUsage_97}
   ]
  }
  MaterialExpressionAdd_17 {
   class=MaterialExpressionAdd
   inputs [
    A {from=MaterialExpressionMultiply_52}
    B {from=MaterialExpressionScalarParameter_120}
   ]
   properties {constB=2}
  }
  MaterialExpressionAdd_18 {
   class=MaterialExpressionAdd
   inputs [
    A {from=MaterialExpressionGetMaterialAttributes_1, output=BaseColor}
    B {from=MaterialExpressionMultiply_69}
   ]
  }
  MaterialExpressionAdd_19 {
   class=MaterialExpressionAdd
   inputs [
    A {from=MaterialExpressionGetMaterialAttributes_1, output=EmissiveColor}
    B {from=MaterialExpressionMultiply_74}
   ]
  }
  MaterialExpressionAdd_2 {
   class=MaterialExpressionAdd
   inputs [
    A {from=MaterialExpressionMultiply_48}
    B {from=MaterialExpressionVectorParameter_2, output=RGB}
   ]
  }
  MaterialExpressionAdd_20 {
   class=MaterialExpressionAdd
   inputs [
    A {from=MaterialExpressionMultiply_62}
    B {from=MaterialExpressionMultiply_57}
   ]
  }
  MaterialExpressionAdd_21 {
   class=MaterialExpressionAdd
   inputs [
    A {from=MaterialExpressionMultiply_53}
    B {from=MaterialExpressionMultiply_55}
   ]
  }
  MaterialExpressionAdd_22 {
   class=MaterialExpressionAdd
   inputs [
    A {from=MaterialExpressionMultiply_62}
    B {from=MaterialExpressionMultiply_66}
   ]
  }
  MaterialExpressionAdd_23 {
   class=MaterialExpressionAdd
   inputs [
    A {from=MaterialExpressionMultiply_15}
    B {from=MaterialExpressionMultiply_16}
   ]
  }
  MaterialExpressionAdd_24 {
   class=MaterialExpressionAdd
   inputs [
    A {from=MaterialExpressionAdd_15}
    B {from=MaterialExpressionNamedRerouteUsage_100}
   ]
  }
  MaterialExpressionAdd_25 {
   class=MaterialExpressionAdd
   inputs [
    A {from=MaterialExpressionMultiply_79}
    B {from=""}
   ]
  }
  MaterialExpressionAdd_26 {
   class=MaterialExpressionAdd
   inputs [
    A {from=MaterialExpressionSkyAtmosphereLightDirection_2}
    B {from=MaterialExpressionCameraVectorWS_5}
   ]
  }
  MaterialExpressionAdd_27 {
   class=MaterialExpressionAdd
   inputs [
    A {from=MaterialExpressionAdd_4}
    B {from=MaterialExpressionNamedRerouteUsage_22}
   ]
  }
  MaterialExpressionAdd_28 {
   class=MaterialExpressionAdd
   inputs [
    A {from=MaterialExpressionPower_5}
    B {from=MaterialExpressionSaturate_0}
   ]
  }
  MaterialExpressionAdd_3 {
   class=MaterialExpressionAdd
   inputs [
    A {from=MaterialExpressionMultiply_6}
    B {from=MaterialExpressionSubtract_2}
   ]
  }
  MaterialExpressionAdd_4 {
   class=MaterialExpressionAdd
   inputs [
    A {from=MaterialExpressionNamedRerouteUsage_28}
    B {from=MaterialExpressionNamedRerouteUsage_83}
   ]
  }
  MaterialExpressionAdd_44 {
   class=MaterialExpressionAdd
   inputs [
    A {from=MaterialExpressionVertexColor_0, output=R}
    B {from=""}
   ]
  }
  MaterialExpressionAdd_45 {
   class=MaterialExpressionAdd
   inputs [
    A {from=MaterialExpressionMultiply_85}
    B {from=""}
   ]
   properties {constB=-0.5}
  }
  MaterialExpressionAdd_5 {
   class=MaterialExpressionAdd
   inputs [
    A {from=MaterialExpressionScalarParameter_76}
    B {from=MaterialExpressionAdd_14}
   ]
  }
  MaterialExpressionAdd_55 {
   class=MaterialExpressionAdd
   inputs [
    A {from=MaterialExpressionMultiply_108}
    B {from=MaterialExpressionAppendVector_22}
   ]
  }
  MaterialExpressionAdd_57 {
   class=MaterialExpressionAdd
   inputs [
    A {from=MaterialExpressionMaterialFunctionCall_34, output=R}
    B {from=MaterialExpressionMaterialFunctionCall_34, output=R}
   ]
  }
  MaterialExpressionAdd_58 {
   class=MaterialExpressionAdd
   inputs [
    A {from=MaterialExpressionAdd_57}
    B {from=MaterialExpressionMaterialFunctionCall_34, output=B}
   ]
  }
  MaterialExpressionAdd_59 {
   class=MaterialExpressionAdd
   inputs [
    A {from=MaterialExpressionSine_4}
    B {from=""}
   ]
  }
  MaterialExpressionAdd_6 {
   class=MaterialExpressionAdd
   inputs [
    A {from=MaterialExpressionMax_8}
    B {from=MaterialExpressionMultiply_29}
   ]
  }
  MaterialExpressionAdd_72 {
   class=MaterialExpressionAdd
   inputs [
    A {from=MaterialExpressionMaterialFunctionCall_43, output=R}
    B {from=MaterialExpressionMaterialFunctionCall_43, output=R}
   ]
  }
  MaterialExpressionAdd_73 {
   class=MaterialExpressionAdd
   inputs [
    A {from=MaterialExpressionAdd_72}
    B {from=MaterialExpressionMaterialFunctionCall_43, output=B}
   ]
  }
  MaterialExpressionAdd_74 {
   class=MaterialExpressionAdd
   inputs [
    A {from=MaterialExpressionSine_5}
    B {from=""}
   ]
  }
  MaterialExpressionAdd_75 {
   class=MaterialExpressionAdd
   inputs [
    A {from=MaterialExpressionMultiply_140}
    B {from=MaterialExpressionMultiply_123}
   ]
  }
  MaterialExpressionAdd_8 {
   class=MaterialExpressionAdd
   inputs [
    A {from=MaterialExpressionSubtract_7}
    B {from=MaterialExpressionMultiply_28}
   ]
  }
  MaterialExpressionAdd_9 {
   class=MaterialExpressionAdd
   inputs [
    A {from=MaterialExpressionMultiply_137}
    B {from=MaterialExpressionMultiply_39}
   ]
  }
  MaterialExpressionAppendVector_1 {
   class=MaterialExpressionAppendVector
   inputs [
    A {from=MaterialExpressionAdd_44, output=""}
    B {from=MaterialExpressionAdd_45, output=""}
   ]
  }
  MaterialExpressionAppendVector_21 {
   class=MaterialExpressionAppendVector
   inputs [
    A {from=MaterialExpressionVectorParameter_22, output=RGB}
    B {from=MaterialExpressionVectorParameter_22, output=RGB}
   ]
  }
  MaterialExpressionAppendVector_22 {
   class=MaterialExpressionAppendVector
   inputs [
    A {from=MaterialExpressionVectorParameter_23, output=RGB}
    B {from=MaterialExpressionVectorParameter_23, output=RGB}
   ]
  }
  MaterialExpressionAppendVector_23 {
   class=MaterialExpressionAppendVector
   inputs [
    A {from=MaterialExpressionVertexColor_0, output=""}
    B {from=MaterialExpressionVertexColor_0, output=""}
   ]
  }
  MaterialExpressionBlendMaterialAttributes_0 {
   class=MaterialExpressionBlendMaterialAttributes
   inputs [
    A {from=MaterialExpressionSetMaterialAttributes_2, output=""}
    B {from=MaterialExpressionNamedRerouteUsage_75, output=""}
    Alpha {from=MaterialExpressionLinearInterpolate_15, output=""}
   ]
   properties {}
  }
  MaterialExpressionCameraPositionWS_3 {class=MaterialExpressionCameraPositionWS, inputs [], properties {}}
  MaterialExpressionCameraVectorWS_4 {class=MaterialExpressionCameraVectorWS, inputs [], properties {}}
  MaterialExpressionCameraVectorWS_5 {class=MaterialExpressionCameraVectorWS, inputs [], properties {}}
  MaterialExpressionCeil_1 {class=MaterialExpressionCeil, inputs [None {from=MaterialExpressionNamedRerouteUsage_23, output=""}], properties {}}
  MaterialExpressionChannelMaskParameter_0 {
   class=MaterialExpressionChannelMaskParameter
   inputs [None {from=MaterialExpressionNamedRerouteUsage_2, output=""}]
   properties {group=MistyBead, parameterName=WaterFXMask}
  }
  MaterialExpressionChannelMaskParameter_1 {
   class=MaterialExpressionChannelMaskParameter
   inputs [None {from=MaterialExpressionNamedRerouteUsage_7, output=""}]
   properties {group=MistyBead, maskChannel=Green, parameterName=FacialMask, defaultValue=<rgba 0 1 0 0>}
  }
  MaterialExpressionChannelMaskParameter_2 {
   class=MaterialExpressionChannelMaskParameter
   inputs [None {from=MaterialExpressionNamedRerouteUsage_6, output=""}]
   properties {group=MistyBead, maskChannel=Blue, parameterName=FoamGradientMask, defaultValue=<rgba 0 0 1 0>}
  }
  MaterialExpressionChannelMaskParameter_3 {
   class=MaterialExpressionChannelMaskParameter
   inputs [None {from=MaterialExpressionNamedRerouteUsage_14, output=""}]
   properties {group=MistyBead, maskChannel=Green, parameterName=VerticalGradientMask, defaultValue=<rgba 0 1 0 0>}
  }
  MaterialExpressionChannelMaskParameter_4 {
   class=MaterialExpressionChannelMaskParameter
   inputs [None {from=MaterialExpressionNamedRerouteUsage_19, output=""}]
   properties {group=MistyBead, maskChannel=Blue, parameterName=ForeheadSymbolFXMask, defaultValue=<rgba 0 0 1 0>}
  }
  MaterialExpressionClamp_0 {
   class=MaterialExpressionClamp
   inputs [
    None {from=MaterialExpressionNamedRerouteDeclaration_37, output=""}
    Min {from="", output=""}
    Max {from="", output=""}
   ]
   properties {maxDefault=3}
  }
  MaterialExpressionClamp_2 {
   class=MaterialExpressionClamp
   inputs [
    None {from=MaterialExpressionNamedRerouteDeclaration_44, output=""}
    Min {from="", output=""}
    Max {from="", output=""}
   ]
   properties {maxDefault=3}
  }
  MaterialExpressionComponentMask_0 {class=MaterialExpressionComponentMask, inputs [None {from=MaterialExpressionNamedRerouteUsage_12}], properties {g=true}}
  MaterialExpressionComponentMask_1 {class=MaterialExpressionComponentMask, inputs [None {from=MaterialExpressionVertexNormalWS_0}], properties {b=true}}
  MaterialExpressionComponentMask_19 {class=MaterialExpressionComponentMask, inputs [None {from=MaterialExpressionMultiply_143}], properties {b=true, g=true, r=true}}
  MaterialExpressionComponentMask_3 {class=MaterialExpressionComponentMask, inputs [None {from=MaterialExpressionPixelNormalWS_0}], properties {b=true}}
  MaterialExpressionConstant3Vector_5 {class=MaterialExpressionConstant3Vector, properties {constant=<rgba 0 0 1 0>}}
  MaterialExpressionConstantBiasScale_0 {class=MaterialExpressionConstantBiasScale, inputs [None {from=MaterialExpressionComponentMask_3}]}
  MaterialExpressionConstantBiasScale_1 {class=MaterialExpressionConstantBiasScale, inputs [None {from=MaterialExpressionComponentMask_1}]}
  MaterialExpressionConstantBiasScale_2 {class=MaterialExpressionConstantBiasScale, inputs [None {from=MaterialExpressionVertexColor_1}], properties {bias=-0.5, scale=2}}
  MaterialExpressionConstant_0 {class=MaterialExpressionConstant}
  MaterialExpressionConstant_1 {class=MaterialExpressionConstant}
  MaterialExpressionConstant_10 {class=MaterialExpressionConstant, properties {r=1}}
  MaterialExpressionConstant_11 {class=MaterialExpressionConstant, properties {r=1}}
  MaterialExpressionConstant_14 {class=MaterialExpressionConstant}
  MaterialExpressionConstant_20 {class=MaterialExpressionConstant}
  MaterialExpressionConstant_3 {class=MaterialExpressionConstant}
  MaterialExpressionConstant_5 {class=MaterialExpressionConstant}
  MaterialExpressionConstant_6 {class=MaterialExpressionConstant}
  MaterialExpressionConstant_7 {class=MaterialExpressionConstant}
  MaterialExpressionConstant_8 {class=MaterialExpressionConstant}
  MaterialExpressionConstant_9 {class=MaterialExpressionConstant}
  MaterialExpressionConvert_0 {
   class=MaterialExpressionConvert
   inputs [None {from=MaterialExpressionScalarParameter_94}]
   properties {
    convertInputs [{
      type=Scalar
      expressionInput {
       expression=&self:MaterialExpressionScalarParameter_94
       inputName=None
       mask=0
       maskA=0
       maskB=0
       maskG=0
       maskR=0
       outputIndex=0
      }
     }]
    convertOutputs [{type=Vector2}]
   }
  }
  MaterialExpressionCustom_1 {
   class=MaterialExpressionCustom
   inputs [
    SH {from=MaterialExpressionStaticSwitchParameter_8, output=""}
    Direction {from=MaterialExpressionNamedRerouteUsage_21, output=""}
   ]
   properties {
    code="//SH decode,  Direction = LightDir, SH = 2band coefficients. (encoded in Y-up)\r\nDirection = normalize(Direction);\r\nfloat4 shEvalDirection;\r\nshEvalDirection.x = 0.282095;\r\nshEvalDirection.y =0.488603 * Direction.z;\r\nshEvalDirection.z = 0.488603 * Direction.y;\r\nshEvalDirection.w =-0.488603 * Direction.x;\r\nreturn saturate(dot(SH, shEvalDirection));"
    description=SphericalHarmonicsEval
    outputType=CMOT_Float1
    inputs [
     {
      inputName=SH
      input {
       expression=&self:MaterialExpressionStaticSwitchParameter_8
       inputName=None
       mask=0
       maskA=0
       maskB=0
       maskG=0
       maskR=0
       outputIndex=0
      }
     }
     {
      inputName=Direction
      input {
       expression=&self:MaterialExpressionNamedRerouteUsage_21
       inputName=None
       mask=0
       maskA=0
       maskB=0
       maskG=0
       maskR=0
       outputIndex=0
      }
     }
    ]
   }
  }
  MaterialExpressionDesaturation_0 {
   class=MaterialExpressionDesaturation
   inputs [
    None {from=MaterialExpressionMultiply_72, output=""}
    Fraction {from=MaterialExpressionScalarParameter_103, output=""}
   ]
   properties {}
  }
  MaterialExpressionDivide_0 {
   class=MaterialExpressionDivide
   inputs [
    A {from=MaterialExpressionSubtract_1}
    B {from=MaterialExpressionScalarParameter_24}
   ]
  }
  MaterialExpressionDivide_5 {
   class=MaterialExpressionDivide
   inputs [
    A {from=MaterialExpressionSubtract_12}
    B {from=MaterialExpressionScalarParameter_77}
   ]
   properties {desc=InLength}
  }
  MaterialExpressionDivide_6 {
   class=MaterialExpressionDivide
   inputs [
    A {from=MaterialExpressionSubtract_8}
    B {from=MaterialExpressionScalarParameter_81}
   ]
   properties {constB=3, desc=OutLength}
  }
  MaterialExpressionFrac_3 {class=MaterialExpressionFrac, inputs [None {from=MaterialExpressionMultiply_122, output=""}], properties {}}
  MaterialExpressionFrac_4 {class=MaterialExpressionFrac, inputs [None {from=MaterialExpressionMultiply_139, output=""}], properties {}}
  MaterialExpressionFrac_5 {class=MaterialExpressionFrac, inputs [None {from=MaterialExpressionMultiply_138, output=""}], properties {}}
  MaterialExpressionFresnel_3 {
   class=MaterialExpressionFresnel
   inputs [
    ExponentIn {from=MaterialExpressionScalarParameter_30, output=""}
    BaseReflectFractionIn {from="", output=""}
    Normal {from="", output=""}
   ]
   properties {baseReflectFraction=0, exponent=3}
  }
  MaterialExpressionFunctionInput_0 {
   class=MaterialExpressionFunctionInput
   inputs [Preview {from="", output=""}]
   properties {inputName=InMaterial, inputType=FunctionInput_MaterialAttributes}
  }
  MaterialExpressionFunctionInput_1 {class=MaterialExpressionFunctionInput, inputs [Preview {from="", output=""}], properties {inputName=MaskMap, sortPriority=1}}
  MaterialExpressionFunctionInput_2 {class=MaterialExpressionFunctionInput, inputs [Preview {from="", output=""}], properties {inputName=FXMask, sortPriority=2}}
  MaterialExpressionFunctionOutput_0 {class=MaterialExpressionFunctionOutput, inputs [None {from=MaterialExpressionBlendMaterialAttributes_0, output=""}]}
  MaterialExpressionGetMaterialAttributes_0 {
   class=MaterialExpressionGetMaterialAttributes
   inputs [None {from=MaterialExpressionNamedRerouteUsage_1, output=""}]
   properties {attributeGetTypes [
     #1
     #2
     #3
     #4
     #5
    ]}
  }
  MaterialExpressionGetMaterialAttributes_1 {
   class=MaterialExpressionGetMaterialAttributes
   inputs [None {from=MaterialExpressionNamedRerouteUsage_40, output=""}]
   properties {attributeGetTypes [#1, #2]}
  }
  MaterialExpressionGetMaterialAttributes_2 {
   class=MaterialExpressionGetMaterialAttributes
   inputs [None {from=MaterialExpressionNamedRerouteUsage_77, output=""}]
   properties {attributeGetTypes [#2]}
  }
  MaterialExpressionIf_0 {
   class=MaterialExpressionIf
   inputs [
    A {from=MaterialExpressionNamedRerouteUsage_64, output=""}
    B {from=MaterialExpressionScalarParameter_118, output=""}
    "A > B" {from=MaterialExpressionConstant_11, output=""}
    "A == B" {from=MaterialExpressionSaturate_9, output=""}
    "A < B" {from=MaterialExpressionNamedRerouteDeclaration_38, output=""}
   ]
   properties {constB=3}
  }
  MaterialExpressionIf_1 {
   class=MaterialExpressionIf
   inputs [
    A {from=MaterialExpressionNamedRerouteUsage_64, output=""}
    B {from="", output=""}
    "A > B" {from=MaterialExpressionIf_0, output=""}
    "A == B" {from=MaterialExpressionNamedRerouteDeclaration_33, output=""}
    "A < B" {from=MaterialExpressionIf_0, output=""}
   ]
   properties {}
  }
  MaterialExpressionLinearInterpolate_0 {
   class=MaterialExpressionLinearInterpolate
   inputs [
    A {from=MaterialExpressionAdd_6}
    B {from=MaterialExpressionMultiply_60}
    Alpha {from=MaterialExpressionNamedRerouteUsage_90}
   ]
   properties {constB=0}
  }
  MaterialExpressionLinearInterpolate_1 {
   class=MaterialExpressionLinearInterpolate
   inputs [
    A {from=MaterialExpressionRayTracingQualitySwitch_1}
    B {from=MaterialExpressionMultiply_38}
    Alpha {from=MaterialExpressionNamedRerouteUsage_56}
   ]
  }
  MaterialExpressionLinearInterpolate_10 {
   class=MaterialExpressionLinearInterpolate
   inputs [
    A {from=MaterialExpressionDesaturation_0}
    B {from=MaterialExpressionMultiply_51}
    Alpha {from=MaterialExpressionNamedRerouteUsage_30}
   ]
  }
  MaterialExpressionLinearInterpolate_11 {
   class=MaterialExpressionLinearInterpolate
   inputs [
    A {from=MaterialExpressionSwitch_0}
    B {from=MaterialExpressionSwitch_2}
    Alpha {from=MaterialExpressionNamedRerouteDeclaration_22}
   ]
   properties {desc="This was set to transition between levels, but ended up not bei ng needed"}
  }
  MaterialExpressionLinearInterpolate_12 {
   class=MaterialExpressionLinearInterpolate
   inputs [
    A {from=""}
    B {from=""}
    Alpha {from=MaterialExpressionMultiply_20}
   ]
   properties {constB=1.65}
  }
  MaterialExpressionLinearInterpolate_13 {
   class=MaterialExpressionLinearInterpolate
   inputs [
    A {from=""}
    B {from=MaterialExpressionSmoothStep_1}
    Alpha {from=MaterialExpressionScalarParameter_95}
   ]
   properties {constA=1}
  }
  MaterialExpressionLinearInterpolate_14 {
   class=MaterialExpressionLinearInterpolate
   inputs [
    A {from=MaterialExpressionLinearInterpolate_5}
    B {from=MaterialExpressionMultiply_92}
    Alpha {from=MaterialExpressionScalarParameter_90}
   ]
   properties {constAlpha=0.4}
  }
  MaterialExpressionLinearInterpolate_15 {
   class=MaterialExpressionLinearInterpolate
   inputs [
    A {from=MaterialExpressionNamedRerouteUsage_66}
    B {from=MaterialExpressionMax_6}
    Alpha {from=MaterialExpressionNamedRerouteUsage_52}
   ]
   properties {constA=1}
  }
  MaterialExpressionLinearInterpolate_16 {
   class=MaterialExpressionLinearInterpolate
   inputs [
    A {from=""}
    B {from=MaterialExpressionPower_3}
    Alpha {from=MaterialExpressionScalarParameter_99}
   ]
   properties {constA=1}
  }
  MaterialExpressionLinearInterpolate_17 {
   class=MaterialExpressionLinearInterpolate
   inputs [
    A {from=MaterialExpressionNamedRerouteUsage_43}
    B {from=MaterialExpressionOneMinus_15}
    Alpha {from=MaterialExpressionScalarParameter_100}
   ]
  }
  MaterialExpressionLinearInterpolate_18 {
   class=MaterialExpressionLinearInterpolate
   inputs [
    A {from=MaterialExpressionSaturate_1}
    B {from=MaterialExpressionNamedRerouteUsage_44}
    Alpha {from=MaterialExpressionScalarParameter_41}
   ]
  }
  MaterialExpressionLinearInterpolate_19 {
   class=MaterialExpressionLinearInterpolate
   inputs [
    A {from=MaterialExpressionNormalize_0}
    B {from=MaterialExpressionConstantBiasScale_2}
    Alpha {from=MaterialExpressionScalarParameter_109}
   ]
  }
  MaterialExpressionLinearInterpolate_2 {
   class=MaterialExpressionLinearInterpolate
   inputs [
    A {from=MaterialExpressionLinearInterpolate_0}
    B {from=MaterialExpressionNamedRerouteUsage_20}
    Alpha {from=MaterialExpressionNamedRerouteUsage_24}
   ]
  }
  MaterialExpressionLinearInterpolate_20 {
   class=MaterialExpressionLinearInterpolate
   inputs [
    A {from=""}
    B {from=MaterialExpressionNamedRerouteUsage_81}
    Alpha {from=MaterialExpressionScalarParameter_110}
   ]
   properties {constA=1}
  }
  MaterialExpressionLinearInterpolate_21 {
   class=MaterialExpressionLinearInterpolate
   inputs [
    A {from=""}
    B {from=MaterialExpressionSaturate_19}
    Alpha {from=MaterialExpressionScalarParameter_113}
   ]
   properties {constA=1}
  }
  MaterialExpressionLinearInterpolate_22 {
   class=MaterialExpressionLinearInterpolate
   inputs [
    A {from=""}
    B {from=MaterialExpressionScalarParameter_117}
    Alpha {from=MaterialExpressionNamedRerouteUsage_84}
   ]
   properties {constA=1}
  }
  MaterialExpressionLinearInterpolate_23 {
   class=MaterialExpressionLinearInterpolate
   inputs [
    A {from=MaterialExpressionNamedRerouteUsage_57}
    B {from=MaterialExpressionNamedRerouteUsage_85}
    Alpha {from=MaterialExpressionScalarParameter_119}
   ]
  }
  MaterialExpressionLinearInterpolate_25 {
   class=MaterialExpressionLinearInterpolate
   inputs [
    A {from=MaterialExpressionNamedRerouteUsage_86}
    B {from=MaterialExpressionNamedRerouteUsage_87}
    Alpha {from=MaterialExpressionNamedRerouteUsage_88}
   ]
  }
  MaterialExpressionLinearInterpolate_3 {
   class=MaterialExpressionLinearInterpolate
   inputs [
    A {from=""}
    B {from=MaterialExpressionNamedRerouteUsage_31}
    Alpha {from=MaterialExpressionScalarParameter_33}
   ]
   properties {constA=1}
  }
  MaterialExpressionLinearInterpolate_4 {
   class=MaterialExpressionLinearInterpolate
   inputs [
    A {from=MaterialExpressionLinearInterpolate_14}
    B {from=MaterialExpressionVectorParameter_8, output=RGB}
    Alpha {from=MaterialExpressionMax_3}
   ]
  }
  MaterialExpressionLinearInterpolate_5 {
   class=MaterialExpressionLinearInterpolate
   inputs [
    A {from=MaterialExpressionVectorParameter_4, output=RGB}
    B {from=MaterialExpressionVectorParameter_5, output=RGB}
    Alpha {from=MaterialExpressionMultiply_44}
   ]
  }
  MaterialExpressionLinearInterpolate_6 {
   class=MaterialExpressionLinearInterpolate
   inputs [
    A {from=""}
    B {from=MaterialExpressionNamedRerouteUsage_27}
    Alpha {from=MaterialExpressionScalarParameter_22}
   ]
   properties {constA=1}
  }
  MaterialExpressionLinearInterpolate_7 {
   class=MaterialExpressionLinearInterpolate
   inputs [
    A {from=""}
    B {from=MaterialExpressionPower_2}
    Alpha {from=MaterialExpressionScalarParameter_57}
   ]
   properties {constA=1}
  }
  MaterialExpressionLocalPosition_0 {class=MaterialExpressionLocalPosition, inputs [], properties {localOrigin=InstancePreSkinning}}
  MaterialExpressionLocalPosition_1 {class=MaterialExpressionLocalPosition, inputs [], properties {localOrigin=InstancePreSkinning}}
  MaterialExpressionMaterialFunctionCall_11 {
   class=MaterialExpressionMaterialFunctionCall
   inputs [
    VectorToRotate {from=MaterialExpressionStaticSwitchParameter_13}
    PreSkinnedLocalNormal {from=MaterialExpressionNamedRerouteUsage_16}
    VertexlNormalWS {from=MaterialExpressionTransform_1}
   ]
   properties {materialFunction=/Game/Packages/Fortress_SharedMaterials/Base_Material_Functions/RotateVector_FromSkinning}
  }
  MaterialExpressionMaterialFunctionCall_14 {
   class=MaterialExpressionMaterialFunctionCall
   inputs [
    Normal {from=MaterialExpressionPixelNormalWS_4}
    Light Vector {from=MaterialExpressionNormalize_1}
    Roughness {from=MaterialExpressionScalarParameter_115}
   ]
   properties {materialFunction=/Engine/Functions/Engine_MaterialFunctions01/Lighting/GGXSpecular}
  }
  MaterialExpressionMaterialFunctionCall_15 {
   class=MaterialExpressionMaterialFunctionCall
   inputs [
    Circumference Tiling {from=MaterialExpressionScalarParameter_42}
    Position {from=MaterialExpressionNamedRerouteUsage_79}
    Z Height {from=MaterialExpressionScalarParameter_44}
    HQ Angle Calculation? {from=""}
   ]
   properties {materialFunction=/Game/Characters/Player/Common/MaterialFunctions/MF_CylindricalUVs_New}
  }
  MaterialExpressionMaterialFunctionCall_17 {
   class=MaterialExpressionMaterialFunctionCall
   inputs [Color {from=MaterialExpressionMaterialFunctionCall_18, output=XYZ Texture}]
   properties {materialFunction=&SplitComponents}
  }
  MaterialExpressionMaterialFunctionCall_18 {
   class=MaterialExpressionMaterialFunctionCall
   inputs [
    TextureObject {from=MaterialExpressionTextureObject_1}
    TextureSize {from=MaterialExpressionScalarParameter_53}
    WorldPosition {from=MaterialExpressionAdd_12}
    Export Float 4 {from=""}
    In {from=""}
    World Space Normal {from=MaterialExpressionNamedRerouteUsage_25}
    ProjectionTransitionContrast {from=MaterialExpressionNamedRerouteUsage_94}
   ]
   properties {materialFunction=/Game/Athena/SafeZone/Preskinned_WP_Texture_ScalePan}
  }
  MaterialExpressionMaterialFunctionCall_19 {
   class=MaterialExpressionMaterialFunctionCall
   inputs [Color {from=MaterialExpressionMaterialFunctionCall_20, output=XYZ Texture}]
   properties {materialFunction=&SplitComponents}
  }
  MaterialExpressionMaterialFunctionCall_2 {
   class=MaterialExpressionMaterialFunctionCall
   inputs [
    TextureObject {from=MaterialExpressionTextureObjectParameter_0}
    TextureSize {from=MaterialExpressionMultiply_5}
    RotYAxis {from=""}
    RotZAxis {from=""}
    RotXAxis {from=""}
    PreSkinNormal {from=""}
    ProjectionTransitionContrast {from=""}
    WorldPosition {from=MaterialExpressionVertexInterpolator_0, output=PS}
   ]
   properties {materialFunction=/Game/Packages/Fortress_SharedMaterials/Base_Material_Functions/BaseCustomizationFunctions/PreSkinnedAlignedNormal}
  }
  MaterialExpressionMaterialFunctionCall_20 {
   class=MaterialExpressionMaterialFunctionCall
   inputs [
    TextureObject {from=MaterialExpressionTextureObject_2}
    TextureSize {from=MaterialExpressionScalarParameter_52}
    WorldPosition {from=MaterialExpressionMultiply_93}
    Export Float 4 {from=""}
    In {from=""}
    World Space Normal {from=MaterialExpressionNamedRerouteUsage_32}
    ProjectionTransitionContrast {from=MaterialExpressionNamedRerouteUsage_95}
   ]
   properties {materialFunction=&Preskinned_WP_Texture_ScalePan}
  }
  MaterialExpressionMaterialFunctionCall_21 {
   class=MaterialExpressionMaterialFunctionCall
   inputs [
    Ray Direction {from=MaterialExpressionMultiply_89}
    Surface Normal {from=MaterialExpressionPixelNormalWS_6}
    Refractive Index Origin {from=""}
    Refractive Index Target {from=MaterialExpressionScalarParameter_21}
   ]
   properties {materialFunction=/Engine/Functions/Engine_MaterialFunctions01/Vectors/Refract}
  }
  MaterialExpressionMaterialFunctionCall_22 {
   class=MaterialExpressionMaterialFunctionCall
   inputs [
    TextureObject {from=MaterialExpressionTextureObjectParameter_1}
    TextureSize {from=MaterialExpressionScalarParameter_59}
    WorldPosition {from=MaterialExpressionAdd_20}
    Export Float 4 {from=""}
    World Space Normal {from=MaterialExpressionNamedRerouteUsage_4}
    ProjectionTransitionContrast {from=""}
   ]
   properties {materialFunction=&WorldAlignedTexture}
  }
  MaterialExpressionMaterialFunctionCall_23 {
   class=MaterialExpressionMaterialFunctionCall
   inputs [
    TextureObject {from=MaterialExpressionTextureObjectParameter_1}
    TextureSize {from=MaterialExpressionScalarParameter_60}
    WorldPosition {from=MaterialExpressionAdd_22}
    Export Float 4 {from=""}
    World Space Normal {from=MaterialExpressionNamedRerouteUsage_4}
    ProjectionTransitionContrast {from=""}
   ]
   properties {materialFunction=&WorldAlignedTexture}
  }
  MaterialExpressionMaterialFunctionCall_28 {
   class=MaterialExpressionMaterialFunctionCall
   inputs [
    Tile {from=MaterialExpressionScalarParameter_9}
    DistortionAdjustment {from=MaterialExpressionConstant_8}
   ]
   properties {materialFunction=/Game/Global/Materials/MaterialGym/Materials/MaterialElements/UVs/CylindricalProjection/MF_CylindricalProjUV}
  }
  MaterialExpressionMaterialFunctionCall_29 {
   class=MaterialExpressionMaterialFunctionCall
   inputs [
    Axis {from=MaterialExpressionMaterialFunctionCall_30, output=Binormal}
    Exponent {from=""}
    Normal {from=""}
   ]
   properties {materialFunction=&AxisAlignedFresnel}
  }
  MaterialExpressionMaterialFunctionCall_3 {
   class=MaterialExpressionMaterialFunctionCall
   inputs [Color {from=MaterialExpressionNamedRerouteUsage_5}]
   properties {materialFunction=&SplitComponents}
  }
  MaterialExpressionMaterialFunctionCall_30 {
   class=MaterialExpressionMaterialFunctionCall
   inputs [
    UVs {from=MaterialExpressionTextureCoordinate_5}
    Normal {from=MaterialExpressionPixelNormalWS_2}
   ]
   properties {materialFunction=/Engine/Functions/Engine_MaterialFunctions01/Texturing/DeriveTangentBasis}
  }
  MaterialExpressionMaterialFunctionCall_34 {
   class=MaterialExpressionMaterialFunctionCall
   inputs [Float3 {from=MaterialExpressionMultiply_120}]
   properties {materialFunction=&BreakOutFloat3Components}
  }
  MaterialExpressionMaterialFunctionCall_35 {
   class=MaterialExpressionMaterialFunctionCall
   inputs [
    Input {from=MaterialExpressionCustom_1}
    Input Low {from=MaterialExpressionScalarParameter_64}
    Input High {from=MaterialExpressionScalarParameter_65}
    Target Low {from=MaterialExpressionConstant_9}
    Target High {from=MaterialExpressionConstant_10}
   ]
   properties {materialFunction=&RemapValueRange}
  }
  MaterialExpressionMaterialFunctionCall_4 {
   class=MaterialExpressionMaterialFunctionCall
   inputs [
    BaseNormal {from=MaterialExpressionNamedRerouteUsage_10}
    AdditionalNormal {from=MaterialExpressionMaterialFunctionCall_5, output=Result}
   ]
   properties {materialFunction=&BlendAngleCorrectedNormals}
  }
  MaterialExpressionMaterialFunctionCall_43 {
   class=MaterialExpressionMaterialFunctionCall
   inputs [Float3 {from=MaterialExpressionAbs_6}]
   properties {materialFunction=&BreakOutFloat3Components}
  }
  MaterialExpressionMaterialFunctionCall_48 {
   class=MaterialExpressionMaterialFunctionCall
   inputs []
   properties {materialFunction=&CameraDirectionVector}
  }
  MaterialExpressionMaterialFunctionCall_5 {
   class=MaterialExpressionMaterialFunctionCall
   inputs [
    Normal {from=MaterialExpressionMaterialFunctionCall_2, output=XYZ Texture}
    Flatness {from=MaterialExpressionScalarParameter_3}
   ]
   properties {materialFunction=&FlattenNormal}
  }
  MaterialExpressionMaterialFunctionCall_9 {
   class=MaterialExpressionMaterialFunctionCall
   inputs [
    X {from=MaterialExpressionVertexColor_0, output=G}
    Y {from=MaterialExpressionVertexColor_0, output=G}
    Z {from=MaterialExpressionVertexColor_0, output=G}
   ]
   properties {materialFunction=&MakeFloat3}
  }
  MaterialExpressionMax_0 {
   class=MaterialExpressionMax
   inputs [
    A {from=MaterialExpressionNamedRerouteDeclaration_32}
    B {from=MaterialExpressionCeil_1}
   ]
  }
  MaterialExpressionMax_2 {
   class=MaterialExpressionMax
   inputs [
    A {from=MaterialExpressionAdd_28}
    B {from=MaterialExpressionMultiply_19}
   ]
  }
  MaterialExpressionMax_3 {
   class=MaterialExpressionMax
   inputs [
    A {from=MaterialExpressionMultiply_45}
    B {from=MaterialExpressionSaturate_16}
   ]
  }
  MaterialExpressionMax_4 {
   class=MaterialExpressionMax
   inputs [
    A {from=MaterialExpressionMultiply_17}
    B {from=MaterialExpressionMultiply_84}
   ]
  }
  MaterialExpressionMax_6 {
   class=MaterialExpressionMax
   inputs [
    A {from=MaterialExpressionNamedRerouteUsage_49}
    B {from=MaterialExpressionMultiply_96}
   ]
  }
  MaterialExpressionMax_7 {
   class=MaterialExpressionMax
   inputs [
    A {from=MaterialExpressionSmoothStep_7}
    B {from=MaterialExpressionSaturate_8}
   ]
  }
  MaterialExpressionMax_8 {
   class=MaterialExpressionMax
   inputs [
    A {from=MaterialExpressionNamedRerouteUsage_15}
    B {from=MaterialExpressionNamedRerouteUsage_51}
   ]
  }
  MaterialExpressionMultiply_0 {
   class=MaterialExpressionMultiply
   inputs [
    A {from=MaterialExpressionNormalize_3}
    B {from=MaterialExpressionScalarParameter_50}
   ]
  }
  MaterialExpressionMultiply_1 {
   class=MaterialExpressionMultiply
   inputs [
    A {from=MaterialExpressionTime_7}
    B {from=""}
   ]
   properties {constB=0.2}
  }
  MaterialExpressionMultiply_10 {
   class=MaterialExpressionMultiply
   inputs [
    A {from=MaterialExpressionVectorParameter_6, output=RGB}
    B {from=MaterialExpressionMultiply_49}
   ]
  }
  MaterialExpressionMultiply_108 {
   class=MaterialExpressionMultiply
   inputs [
    A {from=MaterialExpressionSubtract_13}
    B {from=MaterialExpressionAppendVector_23}
   ]
  }
  MaterialExpressionMultiply_11 {
   class=MaterialExpressionMultiply
   inputs [
    A {from=MaterialExpressionStaticSwitchParameter_2}
    B {from=MaterialExpressionNamedRerouteUsage_65}
   ]
  }
  MaterialExpressionMultiply_12 {
   class=MaterialExpressionMultiply
   inputs [
    A {from=MaterialExpressionChannelMaskParameter_1, output=R}
    B {from=MaterialExpressionNamedRerouteUsage_50}
   ]
  }
  MaterialExpressionMultiply_120 {
   class=MaterialExpressionMultiply
   inputs [
    A {from=MaterialExpressionNamedRerouteUsage_33}
    B {from=MaterialExpressionScalarParameter_92}
   ]
   properties {constB=0.02}
  }
  MaterialExpressionMultiply_121 {
   class=MaterialExpressionMultiply
   inputs [
    A {from=MaterialExpressionLinearInterpolate_19}
    B {from=MaterialExpressionAdd_9}
   ]
  }
  MaterialExpressionMultiply_122 {
   class=MaterialExpressionMultiply
   inputs [
    A {from=MaterialExpressionSubtract_24}
    B {from=MaterialExpressionScalarParameter_91}
   ]
   properties {constB=0.6}
  }
  MaterialExpressionMultiply_123 {
   class=MaterialExpressionMultiply
   inputs [
    A {from=MaterialExpressionAdd_59}
    B {from=""}
   ]
   properties {constB=0.5}
  }
  MaterialExpressionMultiply_13 {
   class=MaterialExpressionMultiply
   inputs [
    A {from=MaterialExpressionVectorParameter_13, output=RGB}
    B {from=MaterialExpressionLinearInterpolate_6}
   ]
  }
  MaterialExpressionMultiply_137 {
   class=MaterialExpressionMultiply
   inputs [
    A {from=MaterialExpressionAdd_75}
    B {from=MaterialExpressionScalarParameter_72}
   ]
  }
  MaterialExpressionMultiply_138 {
   class=MaterialExpressionMultiply
   inputs [
    A {from=MaterialExpressionWorldPosition_0, output=XYZ}
    B {from=""}
   ]
   properties {constB=0.0025}
  }
  MaterialExpressionMultiply_139 {
   class=MaterialExpressionMultiply
   inputs [
    A {from=MaterialExpressionTime_11}
    B {from=""}
   ]
   properties {constB=0.25}
  }
  MaterialExpressionMultiply_14 {
   class=MaterialExpressionMultiply
   inputs [
    A {from=MaterialExpressionSaturate_2}
    B {from=MaterialExpressionSaturate_6}
   ]
  }
  MaterialExpressionMultiply_140 {
   class=MaterialExpressionMultiply
   inputs [
    A {from=MaterialExpressionAdd_74}
    B {from=""}
   ]
   properties {constB=0.5}
  }
  MaterialExpressionMultiply_141 {
   class=MaterialExpressionMultiply
   inputs [
    A {from=MaterialExpressionSubtract_27}
    B {from=""}
   ]
   properties {constB=2}
  }
  MaterialExpressionMultiply_143 {
   class=MaterialExpressionMultiply
   inputs [
    A {from=MaterialExpressionMaterialFunctionCall_48, output=Result}
    B {from=MaterialExpressionScalarParameter_80}
   ]
  }
  MaterialExpressionMultiply_15 {
   class=MaterialExpressionMultiply
   inputs [
    A {from=MaterialExpressionMaterialFunctionCall_22, output=XY Texture}
    B {from=""}
   ]
   properties {constB=0.5}
  }
  MaterialExpressionMultiply_16 {
   class=MaterialExpressionMultiply
   inputs [
    A {from=MaterialExpressionMaterialFunctionCall_23, output=XY Texture}
    B {from=""}
   ]
   properties {constB=0.5}
  }
  MaterialExpressionMultiply_169 {
   class=MaterialExpressionMultiply
   inputs [
    A {from=MaterialExpressionTextureSampleParameter2D_4, output=R}
    B {from=MaterialExpressionMultiply_121}
   ]
  }
  MaterialExpressionMultiply_17 {
   class=MaterialExpressionMultiply
   inputs [
    A {from=MaterialExpressionNamedRerouteUsage_13}
    B {from=MaterialExpressionLinearInterpolate_16}
   ]
  }
  MaterialExpressionMultiply_18 {
   class=MaterialExpressionMultiply
   inputs [
    A {from=MaterialExpressionSmoothStep_0}
    B {from=MaterialExpressionScalarParameter_16}
   ]
  }
  MaterialExpressionMultiply_19 {
   class=MaterialExpressionMultiply
   inputs [
    A {from=MaterialExpressionPower_0}
    B {from=MaterialExpressionScalarParameter_17}
   ]
  }
  MaterialExpressionMultiply_20 {
   class=MaterialExpressionMultiply
   inputs [
    A {from=MaterialExpressionLinearInterpolate_11}
    B {from=MaterialExpressionScalarParameter_37}
   ]
  }
  MaterialExpressionMultiply_21 {
   class=MaterialExpressionMultiply
   inputs [
    A {from=MaterialExpressionNamedRerouteUsage_35}
    B {from=MaterialExpressionScalarParameter_34}
   ]
  }
  MaterialExpressionMultiply_22 {
   class=MaterialExpressionMultiply
   inputs [
    A {from=MaterialExpressionTime_1}
    B {from=MaterialExpressionMultiply_25}
   ]
  }
  MaterialExpressionMultiply_23 {
   class=MaterialExpressionMultiply
   inputs [
    A {from=MaterialExpressionMax_7}
    B {from=MaterialExpressionLinearInterpolate_13}
   ]
  }
  MaterialExpressionMultiply_24 {
   class=MaterialExpressionMultiply
   inputs [
    A {from=MaterialExpressionMultiply_53}
    B {from=MaterialExpressionSubtract_18}
   ]
  }
  MaterialExpressionMultiply_25 {
   class=MaterialExpressionMultiply
   inputs [
    A {from=MaterialExpressionMultiply_88}
    B {from=MaterialExpressionVectorParameter_10, output=RGB}
   ]
  }
  MaterialExpressionMultiply_26 {
   class=MaterialExpressionMultiply
   inputs [
    A {from=MaterialExpressionTextureCoordinate_1}
    B {from=MaterialExpressionScalarParameter_48}
   ]
   properties {constB=6}
  }
  MaterialExpressionMultiply_27 {
   class=MaterialExpressionMultiply
   inputs [
    A {from=MaterialExpressionAdd_8}
    B {from=MaterialExpressionSubtract_7}
   ]
  }
  MaterialExpressionMultiply_28 {
   class=MaterialExpressionMultiply
   inputs [
    A {from=MaterialExpressionSubtract_6}
    B {from=MaterialExpressionScalarParameter_93}
   ]
   properties {constB=0.1}
  }
  MaterialExpressionMultiply_29 {
   class=MaterialExpressionMultiply
   inputs [
    A {from=MaterialExpressionAdd_27}
    B {from=MaterialExpressionLinearInterpolate_3}
   ]
  }
  MaterialExpressionMultiply_3 {
   class=MaterialExpressionMultiply
   inputs [
    A {from=MaterialExpressionMultiply_4}
    B {from=MaterialExpressionVectorParameter_1, output=RGB}
   ]
  }
  MaterialExpressionMultiply_30 {
   class=MaterialExpressionMultiply
   inputs [
    A {from=MaterialExpressionMultiply_50}
    B {from=MaterialExpressionScalarParameter_49}
   ]
  }
  MaterialExpressionMultiply_31 {
   class=MaterialExpressionMultiply
   inputs [
    A {from=MaterialExpressionSubtract_5}
    B {from=MaterialExpressionScalarParameter_39}
   ]
  }
  MaterialExpressionMultiply_32 {
   class=MaterialExpressionMultiply
   inputs [
    A {from=MaterialExpressionVectorParameter_3, output=A}
    B {from=MaterialExpressionSaturate_10}
   ]
  }
  MaterialExpressionMultiply_33 {
   class=MaterialExpressionMultiply
   inputs [
    A {from=MaterialExpressionSmoothStep_2}
    B {from=MaterialExpressionNamedRerouteUsage_46}
   ]
  }
  MaterialExpressionMultiply_34 {
   class=MaterialExpressionMultiply
   inputs [
    A {from=MaterialExpressionChannelMaskParameter_4, output=R}
    B {from=MaterialExpressionNamedRerouteUsage_55}
   ]
  }
  MaterialExpressionMultiply_35 {
   class=MaterialExpressionMultiply
   inputs [
    A {from=MaterialExpressionMax_2}
    B {from=MaterialExpressionLinearInterpolate_20}
   ]
  }
  MaterialExpressionMultiply_36 {
   class=MaterialExpressionMultiply
   inputs [
    A {from=MaterialExpressionSphereMask_0}
    B {from=MaterialExpressionScalarParameter_96}
   ]
  }
  MaterialExpressionMultiply_37 {
   class=MaterialExpressionMultiply
   inputs [
    A {from=MaterialExpressionSubtract_11}
    B {from=MaterialExpressionScalarParameter_43}
   ]
  }
  MaterialExpressionMultiply_38 {
   class=MaterialExpressionMultiply
   inputs [
    A {from=MaterialExpressionLinearInterpolate_23}
    B {from=MaterialExpressionAdd_17}
   ]
  }
  MaterialExpressionMultiply_39 {
   class=MaterialExpressionMultiply
   inputs [
    A {from=MaterialExpressionMultiply_37}
    B {from=MaterialExpressionNamedRerouteUsage_39}
   ]
  }
  MaterialExpressionMultiply_4 {
   class=MaterialExpressionMultiply
   inputs [
    A {from=MaterialExpressionTime_0}
    B {from=MaterialExpressionScalarParameter_2}
   ]
  }
  MaterialExpressionMultiply_40 {
   class=MaterialExpressionMultiply
   inputs [
    A {from=MaterialExpressionTime_2}
    B {from=MaterialExpressionMultiply_41}
   ]
  }
  MaterialExpressionMultiply_41 {
   class=MaterialExpressionMultiply
   inputs [
    A {from=MaterialExpressionMultiply_90}
    B {from=MaterialExpressionVectorParameter_11, output=RGB}
   ]
  }
  MaterialExpressionMultiply_42 {
   class=MaterialExpressionMultiply
   inputs [
    A {from=MaterialExpressionAdd_16}
    B {from=MaterialExpressionScalarParameter_54}
   ]
   properties {constB=0.25}
  }
  MaterialExpressionMultiply_43 {
   class=MaterialExpressionMultiply
   inputs [
    A {from=MaterialExpressionSmoothStep_2}
    B {from=MaterialExpressionOneMinus_10}
   ]
  }
  MaterialExpressionMultiply_44 {
   class=MaterialExpressionMultiply
   inputs [
    A {from=MaterialExpressionSaturate_5}
    B {from=MaterialExpressionLinearInterpolate_17}
   ]
  }
  MaterialExpressionMultiply_45 {
   class=MaterialExpressionMultiply
   inputs [
    A {from=MaterialExpressionFresnel_3}
    B {from=MaterialExpressionLinearInterpolate_7}
   ]
  }
  MaterialExpressionMultiply_46 {
   class=MaterialExpressionMultiply
   inputs [
    A {from=MaterialExpressionPower_13}
    B {from=MaterialExpressionScalarParameter_40}
   ]
  }
  MaterialExpressionMultiply_47 {
   class=MaterialExpressionMultiply
   inputs [
    A {from=MaterialExpressionSmoothStep_3}
    B {from=MaterialExpressionNamedRerouteUsage_53}
   ]
  }
  MaterialExpressionMultiply_48 {
   class=MaterialExpressionMultiply
   inputs [
    A {from=MaterialExpressionNamedRerouteUsage_99}
    B {from=MaterialExpressionVectorParameter_7, output=RGB}
   ]
  }
  MaterialExpressionMultiply_49 {
   class=MaterialExpressionMultiply
   inputs [
    A {from=MaterialExpressionVectorParameter_6, output=A}
    B {from=MaterialExpressionMax_4}
   ]
  }
  MaterialExpressionMultiply_5 {
   class=MaterialExpressionMultiply
   inputs [
    A {from=MaterialExpressionScalarParameter_4}
    B {from=MaterialExpressionVectorParameter_0, output=RGB}
   ]
  }
  MaterialExpressionMultiply_50 {
   class=MaterialExpressionMultiply
   inputs [
    A {from=MaterialExpressionPower_6}
    B {from=MaterialExpressionMultiply_23}
   ]
  }
  MaterialExpressionMultiply_51 {
   class=MaterialExpressionMultiply
   inputs [
    A {from=MaterialExpressionDesaturation_0}
    B {from=MaterialExpressionVectorParameter_14, output=RGB}
   ]
  }
  MaterialExpressionMultiply_52 {
   class=MaterialExpressionMultiply
   inputs [
    A {from=MaterialExpressionNamedRerouteUsage_58}
    B {from=MaterialExpressionScalarParameter_84}
   ]
  }
  MaterialExpressionMultiply_53 {
   class=MaterialExpressionMultiply
   inputs [
    A {from=MaterialExpressionNamedRerouteUsage_103}
    B {from=""}
   ]
  }
  MaterialExpressionMultiply_54 {
   class=MaterialExpressionMultiply
   inputs [
    A {from=MaterialExpressionLinearInterpolate_2}
    B {from=""}
   ]
   properties {constB=0.1}
  }
  MaterialExpressionMultiply_55 {
   class=MaterialExpressionMultiply
   inputs [
    A {from=MaterialExpressionNamedRerouteUsage_101}
    B {from=MaterialExpressionScalarParameter_88}
   ]
   properties {constB=0.5}
  }
  MaterialExpressionMultiply_56 {
   class=MaterialExpressionMultiply
   inputs [
    A {from=MaterialExpressionTextureCoordinate_2}
    B {from=""}
   ]
   properties {constB=25}
  }
  MaterialExpressionMultiply_57 {
   class=MaterialExpressionMultiply
   inputs [
    A {from=MaterialExpressionTime_3}
    B {from=MaterialExpressionMultiply_58}
   ]
  }
  MaterialExpressionMultiply_58 {
   class=MaterialExpressionMultiply
   inputs [
    A {from=MaterialExpressionScalarParameter_58}
    B {from=MaterialExpressionConstant3Vector_5}
   ]
  }
  MaterialExpressionMultiply_6 {
   class=MaterialExpressionMultiply
   inputs [
    A {from=MaterialExpressionSubtract_0}
    B {from=MaterialExpressionScalarParameter_98}
   ]
  }
  MaterialExpressionMultiply_60 {
   class=MaterialExpressionMultiply
   inputs [
    A {from=MaterialExpressionNamedRerouteUsage_91}
    B {from=MaterialExpressionScalarParameter_18}
   ]
   properties {constB=0.2}
  }
  MaterialExpressionMultiply_61 {
   class=MaterialExpressionMultiply
   inputs [
    A {from=MaterialExpressionMultiply_70}
    B {from=MaterialExpressionScalarParameter_79}
   ]
  }
  MaterialExpressionMultiply_62 {
   class=MaterialExpressionMultiply
   inputs [
    A {from=MaterialExpressionNamedRerouteUsage_47}
    B {from=MaterialExpressionVectorParameter_9, output=RGB}
   ]
  }
  MaterialExpressionMultiply_63 {
   class=MaterialExpressionMultiply
   inputs [
    A {from=MaterialExpressionTextureSample_0, output=R}
    B {from=""}
   ]
   properties {constB=0.25}
  }
  MaterialExpressionMultiply_64 {
   class=MaterialExpressionMultiply
   inputs [
    A {from=MaterialExpressionNamedRerouteUsage_29}
    B {from=""}
   ]
   properties {constB=0.5}
  }
  MaterialExpressionMultiply_65 {
   class=MaterialExpressionMultiply
   inputs [
    A {from=MaterialExpressionNamedRerouteUsage_9}
    B {from=MaterialExpressionScalarParameter_63}
   ]
   properties {constB=0.5}
  }
  MaterialExpressionMultiply_66 {
   class=MaterialExpressionMultiply
   inputs [
    A {from=MaterialExpressionMultiply_57}
    B {from=""}
   ]
   properties {constB=2}
  }
  MaterialExpressionMultiply_67 {
   class=MaterialExpressionMultiply
   inputs [
    A {from=MaterialExpressionNamedRerouteUsage_48}
    B {from=MaterialExpressionNamedRerouteUsage_59}
   ]
  }
  MaterialExpressionMultiply_68 {
   class=MaterialExpressionMultiply
   inputs [
    A {from=MaterialExpressionSubtract_28}
    B {from=MaterialExpressionScalarParameter_8}
   ]
   properties {constB=2}
  }
  MaterialExpressionMultiply_69 {
   class=MaterialExpressionMultiply
   inputs [
    A {from=MaterialExpressionMultiply_76}
    B {from=MaterialExpressionScalarParameter_85}
   ]
   properties {constB=0.5}
  }
  MaterialExpressionMultiply_7 {
   class=MaterialExpressionMultiply
   inputs [
    A {from=MaterialExpressionTextureCoordinate_0}
    B {from=MaterialExpressionScalarParameter_6}
   ]
   properties {constB=5}
  }
  MaterialExpressionMultiply_70 {
   class=MaterialExpressionMultiply
   inputs [
    A {from=MaterialExpressionPower_10}
    B {from=MaterialExpressionNamedRerouteUsage_63}
   ]
  }
  MaterialExpressionMultiply_71 {
   class=MaterialExpressionMultiply
   inputs [
    A {from=MaterialExpressionPower_9}
    B {from=MaterialExpressionScalarParameter_70}
   ]
  }
  MaterialExpressionMultiply_72 {
   class=MaterialExpressionMultiply
   inputs [
    A {from=MaterialExpressionLinearInterpolate_4}
    B {from=MaterialExpressionMultiply_80}
   ]
  }
  MaterialExpressionMultiply_74 {
   class=MaterialExpressionMultiply
   inputs [
    A {from=MaterialExpressionMultiply_76}
    B {from=MaterialExpressionScalarParameter_86}
   ]
  }
  MaterialExpressionMultiply_75 {
   class=MaterialExpressionMultiply
   inputs [
    A {from=MaterialExpressionMultiply_67}
    B {from=MaterialExpressionNamedRerouteUsage_74}
   ]
  }
  MaterialExpressionMultiply_76 {
   class=MaterialExpressionMultiply
   inputs [
    A {from=MaterialExpressionMultiply_75}
    B {from=MaterialExpressionVectorParameter_15, output=RGB}
   ]
  }
  MaterialExpressionMultiply_79 {
   class=MaterialExpressionMultiply
   inputs [
    A {from=MaterialExpressionNamedRerouteUsage_60}
    B {from=MaterialExpressionScalarParameter_87}
   ]
  }
  MaterialExpressionMultiply_80 {
   class=MaterialExpressionMultiply
   inputs [
    A {from=MaterialExpressionScalarParameter_73}
    B {from=MaterialExpressionAdd_25}
   ]
  }
  MaterialExpressionMultiply_81 {
   class=MaterialExpressionMultiply
   inputs [
    A {from=MaterialExpressionSmoothStep_6}
    B {from=MaterialExpressionScalarParameter_114}
   ]
  }
  MaterialExpressionMultiply_82 {
   class=MaterialExpressionMultiply
   inputs [
    A {from=MaterialExpressionMultiply_169}
    B {from=MaterialExpressionScalarParameter_68}
   ]
  }
  MaterialExpressionMultiply_83 {
   class=MaterialExpressionMultiply
   inputs [
    A {from=MaterialExpressionPower_7}
    B {from=MaterialExpressionScalarParameter_112}
   ]
  }
  MaterialExpressionMultiply_84 {
   class=MaterialExpressionMultiply
   inputs [
    A {from=MaterialExpressionSaturate_15}
    B {from=MaterialExpressionNamedRerouteUsage_73}
   ]
  }
  MaterialExpressionMultiply_85 {
   class=MaterialExpressionMultiply
   inputs [
    A {from=MaterialExpressionMaterialFunctionCall_9, output=Result}
    B {from=""}
   ]
  }
  MaterialExpressionMultiply_86 {
   class=MaterialExpressionMultiply
   inputs [
    A {from=MaterialExpressionPower_14}
    B {from=MaterialExpressionScalarParameter_102}
   ]
  }
  MaterialExpressionMultiply_87 {
   class=MaterialExpressionMultiply
   inputs [
    A {from=MaterialExpressionMultiply_35}
    B {from=MaterialExpressionLinearInterpolate_21}
   ]
  }
  MaterialExpressionMultiply_88 {
   class=MaterialExpressionMultiply
   inputs [
    A {from=MaterialExpressionScalarParameter_47}
    B {from=MaterialExpressionNamedRerouteUsage_92}
   ]
   properties {constB=2}
  }
  MaterialExpressionMultiply_89 {
   class=MaterialExpressionMultiply
   inputs [
    A {from=MaterialExpressionCameraVectorWS_4}
    B {from=""}
   ]
   properties {constB=-1}
  }
  MaterialExpressionMultiply_9 {
   class=MaterialExpressionMultiply
   inputs [
    A {from=MaterialExpressionSaturate_9}
    B {from=MaterialExpressionNamedRerouteDeclaration_33}
   ]
  }
  MaterialExpressionMultiply_90 {
   class=MaterialExpressionMultiply
   inputs [
    A {from=MaterialExpressionScalarParameter_51}
    B {from=MaterialExpressionNamedRerouteUsage_93}
   ]
  }
  MaterialExpressionMultiply_91 {
   class=MaterialExpressionMultiply
   inputs [
    A {from=MaterialExpressionGetMaterialAttributes_2, output=EmissiveColor}
    B {from=MaterialExpressionScalarParameter_127}
   ]
  }
  MaterialExpressionMultiply_92 {
   class=MaterialExpressionMultiply
   inputs [
    A {from=MaterialExpressionNamedRerouteUsage_69}
    B {from=MaterialExpressionLinearInterpolate_5}
   ]
  }
  MaterialExpressionMultiply_93 {
   class=MaterialExpressionMultiply
   inputs [
    A {from=MaterialExpressionAdd_24}
    B {from=""}
   ]
   properties {constB=-1}
  }
  MaterialExpressionMultiply_95 {
   class=MaterialExpressionMultiply
   inputs [
    A {from=MaterialExpressionNamedRerouteDeclaration_50}
    B {from=MaterialExpressionNamedRerouteDeclaration_49}
   ]
  }
  MaterialExpressionMultiply_96 {
   class=MaterialExpressionMultiply
   inputs [
    A {from=MaterialExpressionNamedRerouteUsage_61}
    B {from=MaterialExpressionNamedRerouteUsage_89}
   ]
  }
  MaterialExpressionNamedRerouteDeclaration_0 {
   class=MaterialExpressionNamedRerouteDeclaration
   inputs [None {from=MaterialExpressionGetMaterialAttributes_0, output=EmissiveColor}]
   properties {name=InEmissive, nodeColor=<rgba 1 0.0469 0 1>}
  }
  MaterialExpressionNamedRerouteDeclaration_10 {
   class=MaterialExpressionNamedRerouteDeclaration
   inputs [None {from=MaterialExpressionLinearInterpolate_1}]
   properties {name=OutEmissive, nodeColor=<rgba 0 0.5312 1 1>}
  }
  MaterialExpressionNamedRerouteDeclaration_11 {
   class=MaterialExpressionNamedRerouteDeclaration
   inputs [None {from=MaterialExpressionMultiply_87}]
   properties {name=Caustics, nodeColor=<rgba 0.9219 1 0 1>}
  }
  MaterialExpressionNamedRerouteDeclaration_12 {
   class=MaterialExpressionNamedRerouteDeclaration
   inputs [None {from=MaterialExpressionPower_4}]
   properties {name=AOMap, nodeColor=<rgba 0 0.2969 1 1>}
  }
  MaterialExpressionNamedRerouteDeclaration_13 {
   class=MaterialExpressionNamedRerouteDeclaration
   inputs [None {from=MaterialExpressionChannelMaskParameter_3, output=R}]
   properties {name=VerticalGradient, nodeColor=<rgba 0 0.2266 1 1>}
  }
  MaterialExpressionNamedRerouteDeclaration_14 {
   class=MaterialExpressionNamedRerouteDeclaration
   inputs [None {from=MaterialExpressionStaticSwitchParameter_6}]
   properties {name=Foam, nodeColor=<rgba 0 0.8125 1 1>}
  }
  MaterialExpressionNamedRerouteDeclaration_15 {
   class=MaterialExpressionNamedRerouteDeclaration
   inputs [None {from=MaterialExpressionMultiply_36}]
   properties {name=InnerRim, nodeColor=<rgba 0 0.25 1 1>}
  }
  MaterialExpressionNamedRerouteDeclaration_16 {
   class=MaterialExpressionNamedRerouteDeclaration
   inputs [None {from=MaterialExpressionMultiply_12}]
   properties {name=FacialMask, nodeColor=<rgba 0.7109 0 1 1>}
  }
  MaterialExpressionNamedRerouteDeclaration_17 {
   class=MaterialExpressionNamedRerouteDeclaration
   inputs [None {from=MaterialExpressionShaderStageSwitch_2}]
   properties {name=PreSkinnedNormal, nodeColor=<rgba 0 0.1328 1 1>}
  }
  MaterialExpressionNamedRerouteDeclaration_18 {
   class=MaterialExpressionNamedRerouteDeclaration
   inputs [None {from=MaterialExpressionStaticSwitchParameter_1}]
   properties {name=SH_Light_Direction, nodeColor=<rgba 0.4766 0 1 1>}
  }
  MaterialExpressionNamedRerouteDeclaration_19 {
   class=MaterialExpressionNamedRerouteDeclaration
   inputs [None {from=MaterialExpressionLinearInterpolate_18}]
   properties {name=SH_Shading, nodeColor=<rgba 0 1 0.1562 1>}
  }
  MaterialExpressionNamedRerouteDeclaration_2 {
   class=MaterialExpressionNamedRerouteDeclaration
   inputs [None {from=MaterialExpressionFunctionInput_2}]
   properties {name=InFXMask, nodeColor=<rgba 0.0312 0 1 1>}
  }
  MaterialExpressionNamedRerouteDeclaration_20 {
   class=MaterialExpressionNamedRerouteDeclaration
   inputs [None {from=MaterialExpressionShaderStageSwitch_0}]
   properties {name=PreSkinnedPosition, nodeColor=<rgba 0 0.1328 1 1>}
  }
  MaterialExpressionNamedRerouteDeclaration_21 {
   class=MaterialExpressionNamedRerouteDeclaration
   inputs [None {from=MaterialExpressionGetMaterialAttributes_0, output=WorldPositionOffset}]
   properties {name=InWPO, nodeColor=<rgba 0.4531 1 0 1>}
  }
  MaterialExpressionNamedRerouteDeclaration_22 {
   class=MaterialExpressionNamedRerouteDeclaration
   inputs [None {from=MaterialExpressionMultiply_14}]
   properties {name=Timestamp, nodeColor=<rgba 0.2656 1 0 1>}
  }
  MaterialExpressionNamedRerouteDeclaration_24 {
   class=MaterialExpressionNamedRerouteDeclaration
   inputs [None {from=MaterialExpressionStaticSwitchParameter_9}]
   properties {name=EffectMask, nodeColor=<rgba 0 1 0.625 1>}
  }
  MaterialExpressionNamedRerouteDeclaration_25 {
   class=MaterialExpressionNamedRerouteDeclaration
   inputs [None {from=MaterialExpressionMaterialFunctionCall_17, output=G}]
   properties {name=Triplannar_1_CausticsTexture, nodeColor=<rgba 0 0.0391 1 1>}
  }
  MaterialExpressionNamedRerouteDeclaration_27 {
   class=MaterialExpressionNamedRerouteDeclaration
   inputs [None {from=MaterialExpressionStaticSwitchParameter_0}]
   properties {name=ReactivityEdgeFoam, nodeColor=<rgba 0 0.7188 1 1>}
  }
  MaterialExpressionNamedRerouteDeclaration_28 {
   class=MaterialExpressionNamedRerouteDeclaration
   inputs [None {from=MaterialExpressionScalarParameter_56}]
   properties {name=IsReactive, nodeColor=<rgba 1 0 0.9141 1>}
  }
  MaterialExpressionNamedRerouteDeclaration_29 {
   class=MaterialExpressionNamedRerouteDeclaration
   inputs [None {from=MaterialExpressionLinearInterpolate_25}]
   properties {name=OutBaseColor, nodeColor=<rgba 0.8984 0 1 1>}
  }
  MaterialExpressionNamedRerouteDeclaration_3 {
   class=MaterialExpressionNamedRerouteDeclaration
   inputs [None {from=MaterialExpressionFunctionInput_0}]
   properties {name=InMaterial, nodeColor=<rgba 0.5234 1 0 1>}
  }
  MaterialExpressionNamedRerouteDeclaration_30 {
   class=MaterialExpressionNamedRerouteDeclaration
   inputs [None {from=MaterialExpressionStaticSwitchParameter_3}]
   properties {name=TimestampInOut - Clamped, nodeColor=<rgba 1 0 0.9141 1>}
  }
  MaterialExpressionNamedRerouteDeclaration_31 {
   class=MaterialExpressionNamedRerouteDeclaration
   inputs [None {from=MaterialExpressionStaticSwitchParameter_10}]
   properties {name=OutWPO, nodeColor=<rgba 0 1 0.9062 1>}
  }
  MaterialExpressionNamedRerouteDeclaration_32 {
   class=MaterialExpressionNamedRerouteDeclaration
   inputs [None {from=MaterialExpressionChannelMaskParameter_0, output=R}]
   properties {name=SkinMask, nodeColor=<rgba 0 0.625 1 1>}
  }
  MaterialExpressionNamedRerouteDeclaration_33 {
   class=MaterialExpressionNamedRerouteDeclaration
   inputs [None {from=MaterialExpressionOneMinus_12}]
   properties {name=TimestampOut, nodeColor=<rgba 0 0.0859 1 1>}
  }
  MaterialExpressionNamedRerouteDeclaration_34 {
   class=MaterialExpressionNamedRerouteDeclaration
   inputs [None {from=MaterialExpressionScalarParameter_83}]
   properties {name=IsHead, nodeColor=<rgba 1 0 0.0938 1>}
  }
  MaterialExpressionNamedRerouteDeclaration_35 {
   class=MaterialExpressionNamedRerouteDeclaration
   inputs [None {from=MaterialExpressionSaturate_11}]
   properties {name=ForeheadSymbolMask, nodeColor=<rgba 0 0.4141 1 1>}
  }
  MaterialExpressionNamedRerouteDeclaration_36 {
   class=MaterialExpressionNamedRerouteDeclaration
   inputs [None {from=MaterialExpressionGetMaterialAttributes_0, output=BaseColor}]
   properties {name=InBaseColor, nodeColor=<rgba 0.7344 1 0 1>}
  }
  MaterialExpressionNamedRerouteDeclaration_37 {
   class=MaterialExpressionNamedRerouteDeclaration
   inputs [None {from=MaterialExpressionScalarParameter_25}]
   properties {name=KillCount, nodeColor=<rgba 0.5234 0 1 1>}
  }
  MaterialExpressionNamedRerouteDeclaration_38 {
   class=MaterialExpressionNamedRerouteDeclaration
   inputs [None {from=MaterialExpressionMultiply_11}]
   properties {name=TimestampInOut, nodeColor=<rgba 1 0 0.9141 1>}
  }
  MaterialExpressionNamedRerouteDeclaration_39 {
   class=MaterialExpressionNamedRerouteDeclaration
   inputs [None {from=MaterialExpressionSaturate_4}]
   properties {name=RawFoamMask, nodeColor=<rgba 1 0.1172 0 1>}
  }
  MaterialExpressionNamedRerouteDeclaration_4 {
   class=MaterialExpressionNamedRerouteDeclaration
   inputs [None {from=MaterialExpressionMultiply_13}]
   properties {name=FacialExmissive, nodeColor=<rgba 0 0.6484 1 1>}
  }
  MaterialExpressionNamedRerouteDeclaration_40 {
   class=MaterialExpressionNamedRerouteDeclaration
   inputs [None {from=MaterialExpressionSkyLightEnvMapSample_1}]
   properties {name=EnvironmentMap, nodeColor=<rgba 0 1 0.0625 1>}
  }
  MaterialExpressionNamedRerouteDeclaration_41 {
   class=MaterialExpressionNamedRerouteDeclaration
   inputs [None {from=MaterialExpressionSetMaterialAttributes_0}]
   properties {name=ReactivtyMaterial, nodeColor=<rgba 0.8984 1 0 1>}
  }
  MaterialExpressionNamedRerouteDeclaration_42 {
   class=MaterialExpressionNamedRerouteDeclaration
   inputs [None {from=MaterialExpressionMultiply_81}]
   properties {name=FakeSpecular, nodeColor=<rgba 0.6172 1 0 1>}
  }
  MaterialExpressionNamedRerouteDeclaration_43 {
   class=MaterialExpressionNamedRerouteDeclaration
   inputs [None {from=MaterialExpressionStaticSwitchParameter_4}]
   properties {name=BaseSkinCaustics, nodeColor=<rgba 1 0 0.3516 1>}
  }
  MaterialExpressionNamedRerouteDeclaration_44 {
   class=MaterialExpressionNamedRerouteDeclaration
   inputs [None {from=MaterialExpressionScalarParameter_125}]
   properties {name=PreviousKillCount, nodeColor=<rgba 0.7109 1 0 1>}
  }
  MaterialExpressionNamedRerouteDeclaration_45 {
   class=MaterialExpressionNamedRerouteDeclaration
   inputs [None {from=MaterialExpressionLinearInterpolate_10}]
   properties {name=WaterColor, nodeColor=<rgba 1 0 0.9375 1>}
  }
  MaterialExpressionNamedRerouteDeclaration_46 {
   class=MaterialExpressionNamedRerouteDeclaration
   inputs [None {from=MaterialExpressionMultiply_10}]
   properties {name=CausticsColor, nodeColor=<rgba 0 1 0.7422 1>}
  }
  MaterialExpressionNamedRerouteDeclaration_48 {
   class=MaterialExpressionNamedRerouteDeclaration
   inputs [None {from=MaterialExpressionMultiply_95}]
   properties {name=FoamBaseColor, nodeColor=<rgba 0.1016 0 1 1>}
  }
  MaterialExpressionNamedRerouteDeclaration_49 {
   class=MaterialExpressionNamedRerouteDeclaration
   inputs [None {from=MaterialExpressionMultiply_32}]
   properties {name=FoamMask, nodeColor=<rgba 0.2187 1 0 1>}
  }
  MaterialExpressionNamedRerouteDeclaration_5 {
   class=MaterialExpressionNamedRerouteDeclaration
   inputs [None {from=MaterialExpressionSaturate_20}]
   properties {name=FXMask, nodeColor=<rgba 0 0.9297 1 1>}
  }
  MaterialExpressionNamedRerouteDeclaration_50 {
   class=MaterialExpressionNamedRerouteDeclaration
   inputs [None {from=MaterialExpressionVectorParameter_3, output=RGB}]
   properties {name=FoamColor, nodeColor=<rgba 1 0 0.5391 1>}
  }
  MaterialExpressionNamedRerouteDeclaration_51 {
   class=MaterialExpressionNamedRerouteDeclaration
   inputs [None {from=MaterialExpressionLinearInterpolate_22}]
   properties {name=TriplannarSpeed, nodeColor=<rgba 0 0.8594 1 1>}
  }
  MaterialExpressionNamedRerouteDeclaration_52 {
   class=MaterialExpressionNamedRerouteDeclaration
   inputs [None {from=MaterialExpressionScalarParameter_0}]
   properties {name=TriplannarContrast, nodeColor=<rgba 1 0.8203 0 1>}
  }
  MaterialExpressionNamedRerouteDeclaration_53 {
   class=MaterialExpressionNamedRerouteDeclaration
   inputs [None {from=MaterialExpressionMaterialFunctionCall_19, output=R}]
   properties {name=Triplannar_2_R, nodeColor=<rgba 0.5 0 1 1>}
  }
  MaterialExpressionNamedRerouteDeclaration_54 {
   class=MaterialExpressionNamedRerouteDeclaration
   inputs [None {from=MaterialExpressionStaticSwitchParameter_5}]
   properties {name=Triplannar Deformer, nodeColor=<rgba 0 1 0.1797 1>}
  }
  MaterialExpressionNamedRerouteDeclaration_55 {
   class=MaterialExpressionNamedRerouteDeclaration
   inputs [None {from=MaterialExpressionAdd_10}]
   properties {name=TriplannarPosition, nodeColor=<rgba 1 0 0.1406 1>}
  }
  MaterialExpressionNamedRerouteDeclaration_57 {
   class=MaterialExpressionNamedRerouteDeclaration
   inputs [None {from=MaterialExpressionMaterialFunctionCall_19, output=G}]
   properties {name=Triplannar_2_G, nodeColor=<rgba 0.7812 1 0 1>}
  }
  MaterialExpressionNamedRerouteDeclaration_58 {
   class=MaterialExpressionNamedRerouteDeclaration
   inputs [None {from=MaterialExpressionMaterialFunctionCall_17, output=B}]
   properties {name=Triplannar_1_Caustics SDF Texture, nodeColor=<rgba 0.0781 1 0 1>}
  }
  MaterialExpressionNamedRerouteDeclaration_6 {
   class=MaterialExpressionNamedRerouteDeclaration
   inputs [None {from=MaterialExpressionGetMaterialAttributes_0, output=Normal}]
   properties {name=InNormal, nodeColor=<rgba 0 0.1797 1 1>}
  }
  MaterialExpressionNamedRerouteDeclaration_7 {
   class=MaterialExpressionNamedRerouteDeclaration
   inputs [None {from=MaterialExpressionStaticSwitchParameter_7}]
   properties {name=OutNormal, nodeColor=<rgba 1 0.1406 0 1>}
  }
  MaterialExpressionNamedRerouteDeclaration_8 {
   class=MaterialExpressionNamedRerouteDeclaration
   inputs [None {from=MaterialExpressionFunctionInput_1}]
   properties {name=InMask, nodeColor=<rgba 0 1 0.6719 1>}
  }
  MaterialExpressionNamedRerouteDeclaration_9 {
   class=MaterialExpressionNamedRerouteDeclaration
   inputs [None {from=MaterialExpressionChannelMaskParameter_2, output=R}]
   properties {name=FoamGradient, nodeColor=<rgba 1 0 0.2109 1>}
  }
  MaterialExpressionNamedRerouteUsage_0 {class=MaterialExpressionNamedRerouteUsage}
  MaterialExpressionNamedRerouteUsage_1 {class=MaterialExpressionNamedRerouteUsage}
  MaterialExpressionNamedRerouteUsage_10 {class=MaterialExpressionNamedRerouteUsage}
  MaterialExpressionNamedRerouteUsage_100 {class=MaterialExpressionNamedRerouteUsage}
  MaterialExpressionNamedRerouteUsage_101 {class=MaterialExpressionNamedRerouteUsage}
  MaterialExpressionNamedRerouteUsage_102 {class=MaterialExpressionNamedRerouteUsage}
  MaterialExpressionNamedRerouteUsage_103 {class=MaterialExpressionNamedRerouteUsage}
  MaterialExpressionNamedRerouteUsage_11 {class=MaterialExpressionNamedRerouteUsage}
  MaterialExpressionNamedRerouteUsage_12 {class=MaterialExpressionNamedRerouteUsage}
  MaterialExpressionNamedRerouteUsage_13 {class=MaterialExpressionNamedRerouteUsage}
  MaterialExpressionNamedRerouteUsage_14 {class=MaterialExpressionNamedRerouteUsage}
  MaterialExpressionNamedRerouteUsage_15 {class=MaterialExpressionNamedRerouteUsage}
  MaterialExpressionNamedRerouteUsage_16 {class=MaterialExpressionNamedRerouteUsage}
  MaterialExpressionNamedRerouteUsage_17 {class=MaterialExpressionNamedRerouteUsage}
  MaterialExpressionNamedRerouteUsage_18 {class=MaterialExpressionNamedRerouteUsage}
  MaterialExpressionNamedRerouteUsage_19 {class=MaterialExpressionNamedRerouteUsage}
  MaterialExpressionNamedRerouteUsage_2 {class=MaterialExpressionNamedRerouteUsage}
  MaterialExpressionNamedRerouteUsage_20 {class=MaterialExpressionNamedRerouteUsage}
  MaterialExpressionNamedRerouteUsage_21 {class=MaterialExpressionNamedRerouteUsage}
  MaterialExpressionNamedRerouteUsage_22 {class=MaterialExpressionNamedRerouteUsage}
  MaterialExpressionNamedRerouteUsage_23 {class=MaterialExpressionNamedRerouteUsage}
  MaterialExpressionNamedRerouteUsage_24 {class=MaterialExpressionNamedRerouteUsage}
  MaterialExpressionNamedRerouteUsage_25 {class=MaterialExpressionNamedRerouteUsage}
  MaterialExpressionNamedRerouteUsage_26 {class=MaterialExpressionNamedRerouteUsage}
  MaterialExpressionNamedRerouteUsage_27 {class=MaterialExpressionNamedRerouteUsage}
  MaterialExpressionNamedRerouteUsage_28 {class=MaterialExpressionNamedRerouteUsage}
  MaterialExpressionNamedRerouteUsage_29 {class=MaterialExpressionNamedRerouteUsage}
  MaterialExpressionNamedRerouteUsage_30 {class=MaterialExpressionNamedRerouteUsage}
  MaterialExpressionNamedRerouteUsage_31 {class=MaterialExpressionNamedRerouteUsage}
  MaterialExpressionNamedRerouteUsage_32 {class=MaterialExpressionNamedRerouteUsage}
  MaterialExpressionNamedRerouteUsage_33 {class=MaterialExpressionNamedRerouteUsage}
  MaterialExpressionNamedRerouteUsage_34 {class=MaterialExpressionNamedRerouteUsage}
  MaterialExpressionNamedRerouteUsage_35 {class=MaterialExpressionNamedRerouteUsage}
  MaterialExpressionNamedRerouteUsage_36 {class=MaterialExpressionNamedRerouteUsage}
  MaterialExpressionNamedRerouteUsage_37 {class=MaterialExpressionNamedRerouteUsage}
  MaterialExpressionNamedRerouteUsage_38 {class=MaterialExpressionNamedRerouteUsage}
  MaterialExpressionNamedRerouteUsage_39 {class=MaterialExpressionNamedRerouteUsage}
  MaterialExpressionNamedRerouteUsage_4 {class=MaterialExpressionNamedRerouteUsage}
  MaterialExpressionNamedRerouteUsage_40 {class=MaterialExpressionNamedRerouteUsage}
  MaterialExpressionNamedRerouteUsage_41 {class=MaterialExpressionNamedRerouteUsage}
  MaterialExpressionNamedRerouteUsage_42 {class=MaterialExpressionNamedRerouteUsage}
  MaterialExpressionNamedRerouteUsage_43 {class=MaterialExpressionNamedRerouteUsage}
  MaterialExpressionNamedRerouteUsage_44 {class=MaterialExpressionNamedRerouteUsage}
  MaterialExpressionNamedRerouteUsage_45 {class=MaterialExpressionNamedRerouteUsage}
  MaterialExpressionNamedRerouteUsage_46 {class=MaterialExpressionNamedRerouteUsage}
  MaterialExpressionNamedRerouteUsage_47 {class=MaterialExpressionNamedRerouteUsage}
  MaterialExpressionNamedRerouteUsage_48 {class=MaterialExpressionNamedRerouteUsage}
  MaterialExpressionNamedRerouteUsage_49 {class=MaterialExpressionNamedRerouteUsage}
  MaterialExpressionNamedRerouteUsage_5 {class=MaterialExpressionNamedRerouteUsage}
  MaterialExpressionNamedRerouteUsage_50 {class=MaterialExpressionNamedRerouteUsage}
  MaterialExpressionNamedRerouteUsage_51 {class=MaterialExpressionNamedRerouteUsage}
  MaterialExpressionNamedRerouteUsage_52 {class=MaterialExpressionNamedRerouteUsage}
  MaterialExpressionNamedRerouteUsage_53 {class=MaterialExpressionNamedRerouteUsage}
  MaterialExpressionNamedRerouteUsage_54 {class=MaterialExpressionNamedRerouteUsage}
  MaterialExpressionNamedRerouteUsage_55 {class=MaterialExpressionNamedRerouteUsage}
  MaterialExpressionNamedRerouteUsage_56 {class=MaterialExpressionNamedRerouteUsage}
  MaterialExpressionNamedRerouteUsage_57 {class=MaterialExpressionNamedRerouteUsage}
  MaterialExpressionNamedRerouteUsage_58 {class=MaterialExpressionNamedRerouteUsage}
  MaterialExpressionNamedRerouteUsage_59 {class=MaterialExpressionNamedRerouteUsage}
  MaterialExpressionNamedRerouteUsage_6 {class=MaterialExpressionNamedRerouteUsage}
  MaterialExpressionNamedRerouteUsage_60 {class=MaterialExpressionNamedRerouteUsage}
  MaterialExpressionNamedRerouteUsage_61 {class=MaterialExpressionNamedRerouteUsage}
  MaterialExpressionNamedRerouteUsage_63 {class=MaterialExpressionNamedRerouteUsage}
  MaterialExpressionNamedRerouteUsage_64 {class=MaterialExpressionNamedRerouteUsage}
  MaterialExpressionNamedRerouteUsage_65 {class=MaterialExpressionNamedRerouteUsage}
  MaterialExpressionNamedRerouteUsage_66 {class=MaterialExpressionNamedRerouteUsage}
  MaterialExpressionNamedRerouteUsage_67 {class=MaterialExpressionNamedRerouteUsage}
  MaterialExpressionNamedRerouteUsage_69 {class=MaterialExpressionNamedRerouteUsage}
  MaterialExpressionNamedRerouteUsage_7 {class=MaterialExpressionNamedRerouteUsage}
  MaterialExpressionNamedRerouteUsage_70 {class=MaterialExpressionNamedRerouteUsage}
  MaterialExpressionNamedRerouteUsage_72 {class=MaterialExpressionNamedRerouteUsage}
  MaterialExpressionNamedRerouteUsage_73 {class=MaterialExpressionNamedRerouteUsage}
  MaterialExpressionNamedRerouteUsage_74 {class=MaterialExpressionNamedRerouteUsage}
  MaterialExpressionNamedRerouteUsage_75 {class=MaterialExpressionNamedRerouteUsage}
  MaterialExpressionNamedRerouteUsage_76 {class=MaterialExpressionNamedRerouteUsage}
  MaterialExpressionNamedRerouteUsage_77 {class=MaterialExpressionNamedRerouteUsage}
  MaterialExpressionNamedRerouteUsage_78 {class=MaterialExpressionNamedRerouteUsage}
  MaterialExpressionNamedRerouteUsage_79 {class=MaterialExpressionNamedRerouteUsage}
  MaterialExpressionNamedRerouteUsage_8 {class=MaterialExpressionNamedRerouteUsage}
  MaterialExpressionNamedRerouteUsage_81 {class=MaterialExpressionNamedRerouteUsage}
  MaterialExpressionNamedRerouteUsage_82 {class=MaterialExpressionNamedRerouteUsage}
  MaterialExpressionNamedRerouteUsage_83 {class=MaterialExpressionNamedRerouteUsage}
  MaterialExpressionNamedRerouteUsage_84 {class=MaterialExpressionNamedRerouteUsage}
  MaterialExpressionNamedRerouteUsage_85 {class=MaterialExpressionNamedRerouteUsage}
  MaterialExpressionNamedRerouteUsage_86 {class=MaterialExpressionNamedRerouteUsage}
  MaterialExpressionNamedRerouteUsage_87 {class=MaterialExpressionNamedRerouteUsage}
  MaterialExpressionNamedRerouteUsage_88 {class=MaterialExpressionNamedRerouteUsage}
  MaterialExpressionNamedRerouteUsage_89 {class=MaterialExpressionNamedRerouteUsage}
  MaterialExpressionNamedRerouteUsage_9 {class=MaterialExpressionNamedRerouteUsage}
  MaterialExpressionNamedRerouteUsage_90 {class=MaterialExpressionNamedRerouteUsage}
  MaterialExpressionNamedRerouteUsage_91 {class=MaterialExpressionNamedRerouteUsage}
  MaterialExpressionNamedRerouteUsage_92 {class=MaterialExpressionNamedRerouteUsage}
  MaterialExpressionNamedRerouteUsage_93 {class=MaterialExpressionNamedRerouteUsage}
  MaterialExpressionNamedRerouteUsage_94 {class=MaterialExpressionNamedRerouteUsage}
  MaterialExpressionNamedRerouteUsage_95 {class=MaterialExpressionNamedRerouteUsage}
  MaterialExpressionNamedRerouteUsage_96 {class=MaterialExpressionNamedRerouteUsage}
  MaterialExpressionNamedRerouteUsage_97 {class=MaterialExpressionNamedRerouteUsage}
  MaterialExpressionNamedRerouteUsage_98 {class=MaterialExpressionNamedRerouteUsage}
  MaterialExpressionNamedRerouteUsage_99 {class=MaterialExpressionNamedRerouteUsage}
  MaterialExpressionNormalize_0 {class=MaterialExpressionNormalize, inputs [VectorInput {from=MaterialExpressionTransform_0, output=""}]}
  MaterialExpressionNormalize_1 {class=MaterialExpressionNormalize, inputs [VectorInput {from=MaterialExpressionAdd_26, output=""}]}
  MaterialExpressionNormalize_14 {class=MaterialExpressionNormalize, inputs [VectorInput {from=MaterialExpressionVectorParameter_25, output=RGB}]}
  MaterialExpressionNormalize_3 {class=MaterialExpressionNormalize, inputs [VectorInput {from=MaterialExpressionSubtract_15, output=""}]}
  MaterialExpressionOneMinus_0 {class=MaterialExpressionOneMinus, inputs [None {from=MaterialExpressionComponentMask_0}]}
  MaterialExpressionOneMinus_1 {class=MaterialExpressionOneMinus, inputs [None {from=MaterialExpressionTextureSample_2, output=R}]}
  MaterialExpressionOneMinus_10 {class=MaterialExpressionOneMinus, inputs [None {from=MaterialExpressionSphereMask_2}]}
  MaterialExpressionOneMinus_12 {class=MaterialExpressionOneMinus, inputs [None {from=MaterialExpressionSaturate_12}]}
  MaterialExpressionOneMinus_13 {class=MaterialExpressionOneMinus, inputs [None {from=MaterialExpressionConstantBiasScale_1}]}
  MaterialExpressionOneMinus_15 {class=MaterialExpressionOneMinus, inputs [None {from=MaterialExpressionNamedRerouteUsage_43}]}
  MaterialExpressionOneMinus_17 {class=MaterialExpressionOneMinus, inputs [None {from=MaterialExpressionMaterialFunctionCall_29, output=Fresnel}]}
  MaterialExpressionOneMinus_2 {class=MaterialExpressionOneMinus, inputs [None {from=MaterialExpressionTextureSampleParameter2D_1, output=R}]}
  MaterialExpressionOneMinus_3 {class=MaterialExpressionOneMinus, inputs [None {from=MaterialExpressionScalarParameter_106}]}
  MaterialExpressionOneMinus_6 {class=MaterialExpressionOneMinus, inputs [None {from=MaterialExpressionNamedRerouteUsage_82}]}
  MaterialExpressionOneMinus_7 {class=MaterialExpressionOneMinus, inputs [None {from=MaterialExpressionMaterialFunctionCall_35, output=Result}]}
  MaterialExpressionOneMinus_9 {class=MaterialExpressionOneMinus, inputs [None {from=MaterialExpressionConstantBiasScale_0}]}
  MaterialExpressionPanner_0 {
   class=MaterialExpressionPanner
   inputs [
    Coordinate {from=MaterialExpressionMultiply_7}
    Time {from=""}
    Speed {from=""}
   ]
   properties {speedX=0.1, speedY=0.25}
  }
  MaterialExpressionPanner_1 {
   class=MaterialExpressionPanner
   inputs [
    Coordinate {from=MaterialExpressionMultiply_26}
    Time {from=""}
    Speed {from=""}
   ]
   properties {speedY=0.2}
  }
  MaterialExpressionPanner_2 {
   class=MaterialExpressionPanner
   inputs [
    Coordinate {from=MaterialExpressionMultiply_56}
    Time {from=""}
    Speed {from=""}
   ]
   properties {speedY=-0.2}
  }
  MaterialExpressionPanner_6 {
   class=MaterialExpressionPanner
   inputs [
    Coordinate {from=MaterialExpressionMaterialFunctionCall_15, output=UVs}
    Time {from=""}
    Speed {from=""}
   ]
   properties {speedY=-0.25}
  }
  MaterialExpressionPanner_8 {
   class=MaterialExpressionPanner
   inputs [
    Coordinate {from=MaterialExpressionMaterialFunctionCall_28, output=UVs}
    Time {from=""}
    Speed {from=MaterialExpressionConvert_0}
   ]
   properties {speedY=0.2}
  }
  MaterialExpressionPixelNormalWS_0 {class=MaterialExpressionPixelNormalWS, inputs [], properties {}}
  MaterialExpressionPixelNormalWS_2 {class=MaterialExpressionPixelNormalWS, inputs [], properties {}}
  MaterialExpressionPixelNormalWS_4 {class=MaterialExpressionPixelNormalWS, inputs [], properties {}}
  MaterialExpressionPixelNormalWS_6 {class=MaterialExpressionPixelNormalWS, inputs [], properties {}}
  MaterialExpressionPower_0 {
   class=MaterialExpressionPower
   inputs [
    Base {from=MaterialExpressionNamedRerouteUsage_102}
    Exp {from=MaterialExpressionScalarParameter_7}
   ]
  }
  MaterialExpressionPower_1 {
   class=MaterialExpressionPower
   inputs [
    Base {from=MaterialExpressionSaturate_14}
    Exp {from=MaterialExpressionScalarParameter_10}
   ]
  }
  MaterialExpressionPower_10 {
   class=MaterialExpressionPower
   inputs [
    Base {from=MaterialExpressionNamedRerouteUsage_67}
    Exp {from=""}
   ]
   properties {constExponent=0.25}
  }
  MaterialExpressionPower_12 {
   class=MaterialExpressionPower
   inputs [
    Base {from=MaterialExpressionSaturate_17}
    Exp {from=MaterialExpressionScalarParameter_66}
   ]
  }
  MaterialExpressionPower_13 {
   class=MaterialExpressionPower
   inputs [
    Base {from=MaterialExpressionNamedRerouteUsage_41}
    Exp {from=MaterialExpressionScalarParameter_89}
   ]
  }
  MaterialExpressionPower_14 {
   class=MaterialExpressionPower
   inputs [
    Base {from=MaterialExpressionNamedRerouteUsage_72}
    Exp {from=MaterialExpressionScalarParameter_101}
   ]
   properties {constExponent=6}
  }
  MaterialExpressionPower_2 {
   class=MaterialExpressionPower
   inputs [
    Base {from=MaterialExpressionOneMinus_9}
    Exp {from=MaterialExpressionScalarParameter_31}
   ]
  }
  MaterialExpressionPower_3 {
   class=MaterialExpressionPower
   inputs [
    Base {from=MaterialExpressionNamedRerouteUsage_45}
    Exp {from=MaterialExpressionScalarParameter_5}
   ]
  }
  MaterialExpressionPower_4 {
   class=MaterialExpressionPower
   inputs [
    Base {from=MaterialExpressionMaterialFunctionCall_3, output=R}
    Exp {from=""}
   ]
   properties {constExponent=8}
  }
  MaterialExpressionPower_5 {
   class=MaterialExpressionPower
   inputs [
    Base {from=MaterialExpressionSaturate_14}
    Exp {from=MaterialExpressionScalarParameter_13}
   ]
  }
  MaterialExpressionPower_6 {
   class=MaterialExpressionPower
   inputs [
    Base {from=MaterialExpressionOneMinus_13}
    Exp {from=MaterialExpressionScalarParameter_15}
   ]
  }
  MaterialExpressionPower_7 {
   class=MaterialExpressionPower
   inputs [
    Base {from=MaterialExpressionOneMinus_6}
    Exp {from=MaterialExpressionScalarParameter_111}
   ]
   properties {constExponent=6}
  }
  MaterialExpressionPower_8 {
   class=MaterialExpressionPower
   inputs [
    Base {from=MaterialExpressionNamedRerouteUsage_26}
    Exp {from=MaterialExpressionScalarParameter_28}
   ]
  }
  MaterialExpressionPower_9 {
   class=MaterialExpressionPower
   inputs [
    Base {from=MaterialExpressionNamedRerouteUsage_42}
    Exp {from=MaterialExpressionScalarParameter_69}
   ]
  }
  MaterialExpressionPreSkinnedNormal_1 {class=MaterialExpressionPreSkinnedNormal, inputs [], properties {}}
  MaterialExpressionRayTracingQualitySwitch_1 {
   class=MaterialExpressionRayTracingQualitySwitch
   inputs [
    Normal {from=MaterialExpressionLinearInterpolate_2, output=""}
    RayTraced {from=MaterialExpressionMultiply_54, output=""}
   ]
   properties {}
  }
  MaterialExpressionReroute_1 {class=MaterialExpressionReroute, inputs [None {from=MaterialExpressionConstant_0}]}
  MaterialExpressionReroute_2 {class=MaterialExpressionReroute, inputs [None {from=MaterialExpressionScalarParameter_14}]}
  MaterialExpressionReroute_3 {class=MaterialExpressionReroute, inputs [None {from=MaterialExpressionScalarParameter_46}]}
  MaterialExpressionReroute_5 {class=MaterialExpressionReroute, inputs [None {from=MaterialExpressionScalarParameter_67}]}
  MaterialExpressionSaturate_0 {class=MaterialExpressionSaturate, inputs [None {from=MaterialExpressionMultiply_18}]}
  MaterialExpressionSaturate_1 {class=MaterialExpressionSaturate, inputs [None {from=MaterialExpressionPower_12}]}
  MaterialExpressionSaturate_10 {class=MaterialExpressionSaturate, inputs [None {from=MaterialExpressionSubtract_14}]}
  MaterialExpressionSaturate_11 {class=MaterialExpressionSaturate, inputs [None {from=MaterialExpressionMultiply_34}]}
  MaterialExpressionSaturate_12 {class=MaterialExpressionSaturate, inputs [None {from=MaterialExpressionDivide_6}]}
  MaterialExpressionSaturate_13 {class=MaterialExpressionSaturate, inputs [None {from=MaterialExpressionSine_1}]}
  MaterialExpressionSaturate_14 {class=MaterialExpressionSaturate, inputs [None {from=MaterialExpressionMultiply_24}]}
  MaterialExpressionSaturate_15 {class=MaterialExpressionSaturate, inputs [None {from=MaterialExpressionMultiply_61}]}
  MaterialExpressionSaturate_16 {class=MaterialExpressionSaturate, inputs [None {from=MaterialExpressionMultiply_46}]}
  MaterialExpressionSaturate_17 {class=MaterialExpressionSaturate, inputs [None {from=MaterialExpressionOneMinus_7}]}
  MaterialExpressionSaturate_18 {class=MaterialExpressionSaturate, inputs [None {from=MaterialExpressionSubtract_20}]}
  MaterialExpressionSaturate_19 {class=MaterialExpressionSaturate, inputs [None {from=MaterialExpressionMultiply_83}]}
  MaterialExpressionSaturate_2 {class=MaterialExpressionSaturate, inputs [None {from=MaterialExpressionDivide_0}]}
  MaterialExpressionSaturate_20 {class=MaterialExpressionSaturate, inputs [None {from=MaterialExpressionMax_0}]}
  MaterialExpressionSaturate_3 {class=MaterialExpressionSaturate, inputs [None {from=MaterialExpressionSubtract_4}]}
  MaterialExpressionSaturate_4 {class=MaterialExpressionSaturate, inputs [None {from=MaterialExpressionMultiply_30}]}
  MaterialExpressionSaturate_5 {class=MaterialExpressionSaturate, inputs [None {from=MaterialExpressionMultiply_71}]}
  MaterialExpressionSaturate_6 {class=MaterialExpressionSaturate, inputs [None {from=MaterialExpressionScalarParameter_23}]}
  MaterialExpressionSaturate_7 {class=MaterialExpressionSaturate, inputs [None {from=MaterialExpressionSubtract_9}]}
  MaterialExpressionSaturate_8 {class=MaterialExpressionSaturate, inputs [None {from=MaterialExpressionSubtract_10}]}
  MaterialExpressionSaturate_9 {class=MaterialExpressionSaturate, inputs [None {from=MaterialExpressionDivide_5}]}
  MaterialExpressionScalarParameter_0 {class=MaterialExpressionScalarParameter, properties {defaultValue=1, group=MistyBead|Triplannar, parameterName=Triplanar_1_Contrast}}
  MaterialExpressionScalarParameter_10 {class=MaterialExpressionScalarParameter, properties {defaultValue=2, group=MistyBead|Caustics, parameterName=CausticsPower, sortPriority=34}}
  MaterialExpressionScalarParameter_100 {class=MaterialExpressionScalarParameter, properties {group=MistyBead|BaseWater, parameterName=BaseWaterVerticalGradientFlip}}
  MaterialExpressionScalarParameter_101 {class=MaterialExpressionScalarParameter, properties {defaultValue=6, group=MistyBead|Caustics, parameterName=CausticsVerticalGradientExp}}
  MaterialExpressionScalarParameter_102 {class=MaterialExpressionScalarParameter, properties {defaultValue=1, group=MistyBead|Caustics, parameterName=CausticsVerticalGradientInt}}
  MaterialExpressionScalarParameter_103 {class=MaterialExpressionScalarParameter, properties {group=MistyBead|BaseWater, parameterName=BaseWaterSaturation}}
  MaterialExpressionScalarParameter_106 {class=MaterialExpressionScalarParameter, properties {defaultValue=0.25, group=MistyBead|Foam, parameterName=MinimumFoam}}
  MaterialExpressionScalarParameter_107 {class=MaterialExpressionScalarParameter, properties {defaultValue=1, group=MistyBead|Foam, parameterName=FoamVerticalGradientBias}}
  MaterialExpressionScalarParameter_109 {class=MaterialExpressionScalarParameter, properties {group=MistyBead|WPO, parameterName=WPO_UseVertColors}}
  MaterialExpressionScalarParameter_11 {class=MaterialExpressionScalarParameter, properties {defaultValue=0.8, group=MistyBead|Caustics, parameterName=CausticsMin}}
  MaterialExpressionScalarParameter_110 {class=MaterialExpressionScalarParameter, properties {defaultValue=0.5, group=MistyBead|Caustics, parameterName=CausticsAOMapContribution}}
  MaterialExpressionScalarParameter_111 {
   class=MaterialExpressionScalarParameter
   properties {defaultValue=1.5, group=MistyBead|Caustics, parameterName=CausticsVerticalGradientFadeExp}
  }
  MaterialExpressionScalarParameter_112 {
   class=MaterialExpressionScalarParameter
   properties {defaultValue=5, group=MistyBead|Caustics, parameterName=CausticsVerticalGradientFadeInt}
  }
  MaterialExpressionScalarParameter_113 {class=MaterialExpressionScalarParameter, properties {group=MistyBead|Caustics, parameterName=UseCausticsVerticalGradientFade}}
  MaterialExpressionScalarParameter_114 {class=MaterialExpressionScalarParameter, properties {defaultValue=1, group=MistyBead|BaseWater, parameterName=FakeSpecularIntensity}}
  MaterialExpressionScalarParameter_115 {class=MaterialExpressionScalarParameter, properties {defaultValue=0.05, group=MistyBead|BaseWater, parameterName=FakeSpecularRoughness}}
  MaterialExpressionScalarParameter_117 {
   class=MaterialExpressionScalarParameter
   properties {defaultValue=2, group=MistyBead|Triplannar, parameterName=Triplannar_Speed_Reactivity_Boost}
  }
  MaterialExpressionScalarParameter_118 {
   class=MaterialExpressionScalarParameter
   properties {defaultValue=3, group=MistyBead|Reactivity, parameterName=ReactivityAccentsPermanentStage}
  }
  MaterialExpressionScalarParameter_119 {class=MaterialExpressionScalarParameter, properties {group=MistyBead|Reactivity, parameterName=ReactivityAccentUseEmissive}}
  MaterialExpressionScalarParameter_12 {class=MaterialExpressionScalarParameter, properties {defaultValue=1, group=MistyBead|Caustics, parameterName=CausticsMax, sortPriority=33}}
  MaterialExpressionScalarParameter_120 {class=MaterialExpressionScalarParameter, properties {group=MistyBead|Reactivity, parameterName=ReactivityAccentIntensityAdd}}
  MaterialExpressionScalarParameter_122 {class=MaterialExpressionScalarParameter, properties {defaultValue=0.5, group=MistyBead|Reactivity, parameterName=ReactivityFoamMin}}
  MaterialExpressionScalarParameter_123 {class=MaterialExpressionScalarParameter, properties {defaultValue=0.6, group=MistyBead|Reactivity, parameterName=ReactivityFoamMax}}
  MaterialExpressionScalarParameter_124 {class=MaterialExpressionScalarParameter, properties {defaultValue=-0.2, group=MistyBead|Reactivity, parameterName=ReactibityMaskBias}}
  MaterialExpressionScalarParameter_125 {class=MaterialExpressionScalarParameter, properties {group=MistyBead|Reactivity, parameterName=PreviousKillCount}}
  MaterialExpressionScalarParameter_127 {class=MaterialExpressionScalarParameter, properties {defaultValue=1, group=MistyBead, parameterName=BaseEmissiveOverride}}
  MaterialExpressionScalarParameter_13 {class=MaterialExpressionScalarParameter, properties {defaultValue=1.5, group=MistyBead|Caustics, parameterName=CausticsEdgeExp}}
  MaterialExpressionScalarParameter_14 {class=MaterialExpressionScalarParameter, properties {defaultValue=0.2, group=MistyBead|Reactivity, parameterName=Reactivity Stage1}}
  MaterialExpressionScalarParameter_15 {class=MaterialExpressionScalarParameter, properties {defaultValue=2, group=MistyBead|Foam, parameterName=FoamNormalMaskIntensity}}
  MaterialExpressionScalarParameter_16 {class=MaterialExpressionScalarParameter, properties {defaultValue=1, group=MistyBead|Caustics, parameterName=CausticsIntensity}}
  MaterialExpressionScalarParameter_17 {class=MaterialExpressionScalarParameter, properties {group=MistyBead|Caustics, parameterName=SecondaryCausticsIntensity}}
  MaterialExpressionScalarParameter_18 {class=MaterialExpressionScalarParameter, properties {defaultValue=0.2, group=MistyBead|Foam, parameterName=FoamEmissiveContribution}}
  MaterialExpressionScalarParameter_19 {class=MaterialExpressionScalarParameter, properties {defaultValue=0.75, group=MistyBead, parameterName=Specular}}
  MaterialExpressionScalarParameter_2 {class=MaterialExpressionScalarParameter, properties {defaultValue=13, group=MistyBead|PanningNormal, parameterName=PanningNormalSpeed}}
  MaterialExpressionScalarParameter_20 {class=MaterialExpressionScalarParameter, properties {group=MistyBead|Foam, parameterName=Min}}
  MaterialExpressionScalarParameter_21 {class=MaterialExpressionScalarParameter, properties {defaultValue=1.33, group=MistyBead|BaseWater, parameterName=RefractionAmount}}
  MaterialExpressionScalarParameter_22 {class=MaterialExpressionScalarParameter, properties {defaultValue=0.75, group=MistyBead|BaseWater, parameterName=FacialAOMapContribution}}
  MaterialExpressionScalarParameter_23 {class=MaterialExpressionScalarParameter, properties {defaultValue=-999, group=MistyBead|Reactivity, parameterName=Timestamp}}
  MaterialExpressionScalarParameter_24 {class=MaterialExpressionScalarParameter, properties {defaultValue=2, group=MistyBead|Reactivity, parameterName=TimestampLength}}
  MaterialExpressionScalarParameter_25 {class=MaterialExpressionScalarParameter, properties {group=MistyBead|Reactivity, parameterName=KillCount}}
  MaterialExpressionScalarParameter_26 {class=MaterialExpressionScalarParameter, properties {defaultValue=0.5, group=MistyBead|Foam, parameterName=Max}}
  MaterialExpressionScalarParameter_27 {class=MaterialExpressionScalarParameter, properties {defaultValue=0.3, group=MistyBead|BaseWater, parameterName=RimWidth}}
  MaterialExpressionScalarParameter_28 {class=MaterialExpressionScalarParameter, properties {defaultValue=2.9756, group=MistyBead|Foam, parameterName=FoamGradientExp}}
  MaterialExpressionScalarParameter_29 {class=MaterialExpressionScalarParameter, properties {group=MistyBead, parameterName=Metallic}}
  MaterialExpressionScalarParameter_3 {class=MaterialExpressionScalarParameter, properties {defaultValue=0.85, group=MistyBead|PanningNormal, parameterName=PanningNormalFlatness}}
  MaterialExpressionScalarParameter_30 {class=MaterialExpressionScalarParameter, properties {defaultValue=0.5, group=MistyBead|BaseWater, parameterName=BaseWaterFresnelExp}}
  MaterialExpressionScalarParameter_31 {class=MaterialExpressionScalarParameter, properties {defaultValue=0.25, group=MistyBead|BaseWater, parameterName=BaseWaterTopDownMaskExp}}
  MaterialExpressionScalarParameter_32 {class=MaterialExpressionScalarParameter, properties {defaultValue=1, group=MistyBead|BaseWater, parameterName=RimOffset}}
  MaterialExpressionScalarParameter_33 {class=MaterialExpressionScalarParameter, properties {defaultValue=0.5, group=MistyBead|BaseWater, parameterName=BaseWater_AOMapContribution}}
  MaterialExpressionScalarParameter_34 {class=MaterialExpressionScalarParameter, properties {defaultValue=1, group=MistyBead|WPO, parameterName=Pre_Slurp_WPO_Scale}}
  MaterialExpressionScalarParameter_35 {class=MaterialExpressionScalarParameter, properties {group=MistyBead|Foam, parameterName=FoamGradientBias}}
  MaterialExpressionScalarParameter_36 {class=MaterialExpressionScalarParameter, properties {defaultValue=0.15, group=MistyBead|Foam, parameterName=FoamTextureBias}}
  MaterialExpressionScalarParameter_37 {class=MaterialExpressionScalarParameter, properties {defaultValue=1, group=MistyBead|Reactivity, parameterName=ReactivityOn}}
  MaterialExpressionScalarParameter_39 {class=MaterialExpressionScalarParameter, properties {defaultValue=1, group=MistyBead|Foam, parameterName=FoamTextureIntensity}}
  MaterialExpressionScalarParameter_4 {class=MaterialExpressionScalarParameter, properties {defaultValue=35, group=MistyBead|PanningNormal, parameterName=PanningNormalScale}}
  MaterialExpressionScalarParameter_40 {class=MaterialExpressionScalarParameter, properties {defaultValue=1, group=MistyBead|BaseWater, parameterName=BaseWaterVerticalGradientInt}}
  MaterialExpressionScalarParameter_41 {class=MaterialExpressionScalarParameter, properties {group=MistyBead|Spherical Harmonics, parameterName=SH_UseVerticalGradient}}
  MaterialExpressionScalarParameter_42 {class=MaterialExpressionScalarParameter, properties {defaultValue=5, group=MistyBead|WPO, parameterName=WPO_TextureNoise_Scale}}
  MaterialExpressionScalarParameter_43 {class=MaterialExpressionScalarParameter, properties {defaultValue=15, group=MistyBead|WPO, parameterName=WPO_TextureNoise_Intensity}}
  MaterialExpressionScalarParameter_44 {class=MaterialExpressionScalarParameter, properties {defaultValue=30, group=MistyBead|WPO, parameterName=WPO_TextureNoise_Scale_Z}}
  MaterialExpressionScalarParameter_46 {class=MaterialExpressionScalarParameter, properties {defaultValue=0.45, group=MistyBead|Reactivity, parameterName=Reactivity Stage2}}
  MaterialExpressionScalarParameter_47 {class=MaterialExpressionScalarParameter, properties {defaultValue=1, group=MistyBead|Triplannar, parameterName=Triplanar_1_Speed}}
  MaterialExpressionScalarParameter_48 {class=MaterialExpressionScalarParameter, properties {defaultValue=6, group=MistyBead|Reactivity, parameterName=ReactivityMaskNoiseScale}}
  MaterialExpressionScalarParameter_49 {class=MaterialExpressionScalarParameter, properties {defaultValue=7, group=MistyBead|Foam, parameterName=Foam Intensity}}
  MaterialExpressionScalarParameter_5 {
   class=MaterialExpressionScalarParameter
   properties {defaultValue=2.2273, group=MistyBead|Caustics, parameterName=CausticsVerticalGradientMaskExp}
  }
  MaterialExpressionScalarParameter_50 {
   class=MaterialExpressionScalarParameter
   properties {defaultValue=5, group=MistyBead|Triplannar, parameterName=Triplanar_ParralaxDepth, sortPriority=3}
  }
  MaterialExpressionScalarParameter_51 {class=MaterialExpressionScalarParameter, properties {defaultValue=1.5, group=MistyBead|Triplannar, parameterName=Triplanar_2_Speed}}
  MaterialExpressionScalarParameter_52 {class=MaterialExpressionScalarParameter, properties {defaultValue=55, group=MistyBead|Triplannar, parameterName=Triplanar_2_Scale}}
  MaterialExpressionScalarParameter_53 {class=MaterialExpressionScalarParameter, properties {defaultValue=40, group=MistyBead|Triplannar, parameterName=Triplanar_1_Scale}}
  MaterialExpressionScalarParameter_54 {class=MaterialExpressionScalarParameter, properties {defaultValue=5, group=MistyBead|Triplannar, parameterName=Triplanar_1_Warp_Int}}
  MaterialExpressionScalarParameter_55 {class=MaterialExpressionScalarParameter, properties {defaultValue=0.15, group=MistyBead, parameterName=RoughnessBias}}
  MaterialExpressionScalarParameter_56 {class=MaterialExpressionScalarParameter, properties {group=MistyBead, parameterName=IsReactive}}
  MaterialExpressionScalarParameter_57 {class=MaterialExpressionScalarParameter, properties {defaultValue=1, group=MistyBead|BaseWater, parameterName=BaseWaterTopDownMaskIntensity}}
  MaterialExpressionScalarParameter_58 {class=MaterialExpressionScalarParameter, properties {defaultValue=2.5, group=MistyBead|Foam, parameterName=FoamSpeed}}
  MaterialExpressionScalarParameter_59 {class=MaterialExpressionScalarParameter, properties {defaultValue=10, group=MistyBead|Foam, parameterName=FoamScale_1}}
  MaterialExpressionScalarParameter_6 {class=MaterialExpressionScalarParameter, properties {defaultValue=4, group=MistyBead|PanningNormal, parameterName=PanningNormalWarpTiling}}
  MaterialExpressionScalarParameter_60 {class=MaterialExpressionScalarParameter, properties {defaultValue=25, group=MistyBead|Foam, parameterName=FoamScale_2}}
  MaterialExpressionScalarParameter_63 {class=MaterialExpressionScalarParameter, properties {defaultValue=1, group=MistyBead, parameterName=EmissiveColorContribution}}
  MaterialExpressionScalarParameter_64 {
   class=MaterialExpressionScalarParameter
   properties {defaultValue=0.3, group=MistyBead|Spherical Harmonics, parameterName=SHContrast_LowValue}
  }
  MaterialExpressionScalarParameter_65 {
   class=MaterialExpressionScalarParameter
   properties {defaultValue=0.5, group=MistyBead|Spherical Harmonics, parameterName=SHContrast_HighValue}
  }
  MaterialExpressionScalarParameter_66 {class=MaterialExpressionScalarParameter, properties {defaultValue=8, group=MistyBead|Spherical Harmonics, parameterName=SHPow}}
  MaterialExpressionScalarParameter_67 {class=MaterialExpressionScalarParameter, properties {defaultValue=1, group=MistyBead|Reactivity, parameterName=Reactivity Stage3}}
  MaterialExpressionScalarParameter_68 {class=MaterialExpressionScalarParameter, properties {defaultValue=0.05, group=MistyBead|WPO, parameterName=Slurp_WPO_FinalScale}}
  MaterialExpressionScalarParameter_69 {class=MaterialExpressionScalarParameter, properties {defaultValue=1, group=MistyBead|BaseWater, parameterName=BaseWaterCausticsExp}}
  MaterialExpressionScalarParameter_7 {class=MaterialExpressionScalarParameter, properties {defaultValue=3, group=MistyBead|Caustics, parameterName=SecondaryCausticsExp}}
  MaterialExpressionScalarParameter_70 {class=MaterialExpressionScalarParameter, properties {defaultValue=1.5, group=MistyBead|BaseWater, parameterName=BaseWaterCausticsInt}}
  MaterialExpressionScalarParameter_71 {class=MaterialExpressionScalarParameter, properties {group=MistyBead|WPO, parameterName=Slurp_WPO_random_seed}}
  MaterialExpressionScalarParameter_72 {class=MaterialExpressionScalarParameter, properties {defaultValue=1, group=MistyBead|WPO, parameterName=Slurp_WPO_WorldNoiseIntensity}}
  MaterialExpressionScalarParameter_73 {class=MaterialExpressionScalarParameter, properties {defaultValue=1, group=MistyBead|BaseWater, parameterName=BaseWaterIntensity}}
  MaterialExpressionScalarParameter_76 {class=MaterialExpressionScalarParameter, properties {defaultValue=-999, group=MistyBead|Reactivity, parameterName=Timestamp}}
  MaterialExpressionScalarParameter_77 {class=MaterialExpressionScalarParameter, properties {defaultValue=1, group=MistyBead|Reactivity, parameterName=TimestampIn}}
  MaterialExpressionScalarParameter_78 {class=MaterialExpressionScalarParameter, properties {defaultValue=0.75, group=MistyBead|Reactivity, parameterName=TimestampStay}}
  MaterialExpressionScalarParameter_79 {class=MaterialExpressionScalarParameter, properties {defaultValue=2, group=MistyBead|Caustics, parameterName=CausticsFoamEdgeIntensity}}
  MaterialExpressionScalarParameter_8 {class=MaterialExpressionScalarParameter, properties {defaultValue=5, group=MistyBead|Triplannar, parameterName=Triplanar_2_Warp_Int}}
  MaterialExpressionScalarParameter_80 {
   class=MaterialExpressionScalarParameter
   properties {defaultValue=1, group=MistyBead|Spherical Harmonics, parameterName=Slurp_DirectionVectorFlip}
  }
  MaterialExpressionScalarParameter_81 {class=MaterialExpressionScalarParameter, properties {defaultValue=1, group=MistyBead|Reactivity, parameterName=TimestampOut}}
  MaterialExpressionScalarParameter_83 {class=MaterialExpressionScalarParameter, properties {group=MistyBead, parameterName=IsHead}}
  MaterialExpressionScalarParameter_84 {class=MaterialExpressionScalarParameter, properties {defaultValue=2, group=MistyBead|Reactivity, parameterName=ReactivityAccentIntensity}}
  MaterialExpressionScalarParameter_85 {class=MaterialExpressionScalarParameter, properties {defaultValue=0.4, group=MistyBead, parameterName=SkinCausticsBaseColorInt}}
  MaterialExpressionScalarParameter_86 {class=MaterialExpressionScalarParameter, properties {defaultValue=0.2, group=MistyBead, parameterName=SkinCausticsEmissiveInt}}
  MaterialExpressionScalarParameter_87 {class=MaterialExpressionScalarParameter, properties {defaultValue=0.5, group=MistyBead|BaseWater, parameterName=ReactivityGlowIntensity}}
  MaterialExpressionScalarParameter_88 {class=MaterialExpressionScalarParameter, properties {defaultValue=1, group=MistyBead|Caustics, parameterName=Triplanar_2_CausticInt}}
  MaterialExpressionScalarParameter_89 {
   class=MaterialExpressionScalarParameter
   properties {defaultValue=0.75, group=MistyBead|BaseWater, parameterName=BaseWaterVerticalGradientPow}
  }
  MaterialExpressionScalarParameter_9 {class=MaterialExpressionScalarParameter, properties {defaultValue=3, group=MistyBead|Triplannar, parameterName=Triplanar_2_Warp_Scale}}
  MaterialExpressionScalarParameter_90 {class=MaterialExpressionScalarParameter, properties {defaultValue=0.1, group=MistyBead|BaseWater, parameterName=EnvironmentMapBlend}}
  MaterialExpressionScalarParameter_91 {class=MaterialExpressionScalarParameter, properties {defaultValue=2, group=MistyBead|WPO, parameterName=Slurp_WPO_LocalNoise_Speed}}
  MaterialExpressionScalarParameter_92 {class=MaterialExpressionScalarParameter, properties {defaultValue=-0.03, group=MistyBead|WPO, parameterName=Slurp_WPO_LocalNoise_Size}}
  MaterialExpressionScalarParameter_93 {
   class=MaterialExpressionScalarParameter
   properties {defaultValue=0.1, group=MistyBead|Reactivity, parameterName=ReactivityMaskNoiseIntensity}
  }
  MaterialExpressionScalarParameter_94 {class=MaterialExpressionScalarParameter, properties {defaultValue=0.2, group=MistyBead|Triplannar, parameterName=Triplanar_2_Warp_Speed}}
  MaterialExpressionScalarParameter_95 {class=MaterialExpressionScalarParameter, properties {defaultValue=1, group=MistyBead|Foam, parameterName=FoamVerticalGradientMaskInt}}
  MaterialExpressionScalarParameter_96 {class=MaterialExpressionScalarParameter, properties {defaultValue=1, group=MistyBead|BaseWater, parameterName=InnerRimIntensity}}
  MaterialExpressionScalarParameter_97 {class=MaterialExpressionScalarParameter, properties {defaultValue=1, group=MistyBead|BaseWater, parameterName=RimHardness}}
  MaterialExpressionScalarParameter_98 {
   class=MaterialExpressionScalarParameter
   properties {defaultValue=3, group=MistyBead|PanningNormal, parameterName=PanningNormalWarpIntensity}
  }
  MaterialExpressionScalarParameter_99 {
   class=MaterialExpressionScalarParameter
   properties {defaultValue=1, group=MistyBead|Caustics, parameterName=CausticsVerticalGradientMaskInt}
  }
  MaterialExpressionSetMaterialAttributes_0 {
   class=MaterialExpressionSetMaterialAttributes
   inputs [
    MaterialAttributes {from=MaterialExpressionNamedRerouteUsage_78, output=""}
    BaseColor {from=MaterialExpressionNamedRerouteUsage_54, output=""}
    EmissiveColor {from=MaterialExpressionMultiply_65, output=""}
    Normal {from=MaterialExpressionNamedRerouteUsage_11, output=""}
    WorldPositionOffset {from=MaterialExpressionNamedRerouteUsage_34, output=""}
    Roughness {from=MaterialExpressionScalarParameter_55, output=""}
    Specular {from=MaterialExpressionScalarParameter_19, output=""}
    Opacity {from=MaterialExpressionConstant_5, output=""}
    Metallic {from=MaterialExpressionScalarParameter_29, output=""}
    Anisotropy {from="", output=""}
    Tangent {from="", output=""}
   ]
   properties {attributeSetTypes [
     #1
     #2
     #3
     #4
     #5
     #6
     #7
     #8
     #9
     #10
    ]}
  }
  MaterialExpressionSetMaterialAttributes_1 {
   class=MaterialExpressionSetMaterialAttributes
   inputs [
    MaterialAttributes {from=MaterialExpressionGetMaterialAttributes_1, output=MaterialAttributes}
    BaseColor {from=MaterialExpressionAdd_18, output=""}
    EmissiveColor {from=MaterialExpressionAdd_19, output=""}
   ]
   properties {attributeSetTypes [#1, #2]}
  }
  MaterialExpressionSetMaterialAttributes_2 {
   class=MaterialExpressionSetMaterialAttributes
   inputs [
    MaterialAttributes {from=MaterialExpressionGetMaterialAttributes_2, output=MaterialAttributes}
    EmissiveColor {from=MaterialExpressionMultiply_91, output=""}
   ]
   properties {attributeSetTypes [#2]}
  }
  MaterialExpressionShaderStageSwitch_0 {
   class=MaterialExpressionShaderStageSwitch
   inputs [
    PixelShader {from=MaterialExpressionVertexInterpolator_2, output=PS}
    VertexShader {from=MaterialExpressionLocalPosition_1, output=XYZ}
   ]
   properties {}
  }
  MaterialExpressionShaderStageSwitch_2 {
   class=MaterialExpressionShaderStageSwitch
   inputs [
    PixelShader {from=MaterialExpressionVertexInterpolator_1, output=PS}
    VertexShader {from=MaterialExpressionPreSkinnedNormal_1, output=""}
   ]
   properties {}
  }
  MaterialExpressionSine_1 {class=MaterialExpressionSine, inputs [None {from=MaterialExpressionMultiply_1}]}
  MaterialExpressionSine_4 {class=MaterialExpressionSine, inputs [None {from=MaterialExpressionSubtract_22}]}
  MaterialExpressionSine_5 {class=MaterialExpressionSine, inputs [None {from=MaterialExpressionSubtract_26}]}
  MaterialExpressionSkyAtmosphereLightDirection_2 {class=MaterialExpressionSkyAtmosphereLightDirection, inputs [], properties {}}
  MaterialExpressionSkyLightEnvMapSample_1 {
   class=MaterialExpressionSkyLightEnvMapSample
   inputs [
    Direction {from=MaterialExpressionMaterialFunctionCall_21, output=Vector}
    Roughness {from="", output=""}
   ]
   properties {}
  }
  MaterialExpressionSmoothStep_0 {
   class=MaterialExpressionSmoothStep
   inputs [
    Min {from=MaterialExpressionScalarParameter_11, output=""}
    Max {from=MaterialExpressionScalarParameter_12, output=""}
    Value {from=MaterialExpressionPower_1, output=""}
   ]
   properties {}
  }
  MaterialExpressionSmoothStep_1 {
   class=MaterialExpressionSmoothStep
   inputs [
    Min {from="", output=""}
    Max {from="", output=""}
    Value {from=MaterialExpressionNamedRerouteUsage_18, output=""}
   ]
   properties {constMin=0.2}
  }
  MaterialExpressionSmoothStep_2 {
   class=MaterialExpressionSmoothStep
   inputs [
    Min {from="", output=""}
    Max {from="", output=""}
    Value {from=MaterialExpressionSphereMask_2, output=""}
   ]
   properties {constMax=0.2}
  }
  MaterialExpressionSmoothStep_3 {
   class=MaterialExpressionSmoothStep
   inputs [
    Min {from=MaterialExpressionScalarParameter_122, output=""}
    Max {from=MaterialExpressionScalarParameter_123, output=""}
    Value {from=MaterialExpressionMultiply_43, output=""}
   ]
   properties {constMax=0.6, constMin=0.5}
  }
  MaterialExpressionSmoothStep_6 {
   class=MaterialExpressionSmoothStep
   inputs [
    Min {from="", output=""}
    Max {from="", output=""}
    Value {from=MaterialExpressionMaterialFunctionCall_14, output=Result}
   ]
   properties {constMin=0.1}
  }
  MaterialExpressionSmoothStep_7 {
   class=MaterialExpressionSmoothStep
   inputs [
    Min {from=MaterialExpressionScalarParameter_20, output=""}
    Max {from=MaterialExpressionScalarParameter_26, output=""}
    Value {from=MaterialExpressionSaturate_7, output=""}
   ]
   properties {}
  }
  MaterialExpressionSphereMask_0 {
   class=MaterialExpressionSphereMask
   inputs [
    A {from=MaterialExpressionOneMinus_17, output=""}
    B {from=MaterialExpressionScalarParameter_32, output=""}
    Radius {from=MaterialExpressionScalarParameter_27, output=""}
    Hardness {from=MaterialExpressionScalarParameter_97, output=""}
   ]
   properties {attenuationRadius=0.1}
  }
  MaterialExpressionSphereMask_2 {
   class=MaterialExpressionSphereMask
   inputs [
    A {from=MaterialExpressionMultiply_27, output=""}
    B {from=MaterialExpressionConstant_7, output=""}
    Radius {from=MaterialExpressionLinearInterpolate_12, output=""}
    Hardness {from="", output=""}
   ]
   properties {attenuationRadius=1, hardnessPercent=70}
  }
  MaterialExpressionStaticSwitchParameter_0 {
   class=MaterialExpressionStaticSwitchParameter
   inputs [
    True {from=MaterialExpressionMultiply_47, output=""}
    False {from=MaterialExpressionConstant_1, output=""}
   ]
   properties {group=MistyBead|Reactivity, parameterName=UseTextureReactivity}
  }
  MaterialExpressionStaticSwitchParameter_1 {
   class=MaterialExpressionStaticSwitchParameter
   inputs [
    True {from=MaterialExpressionMaterialFunctionCall_11, output=RotatedVectorFromSkinning}
    False {from=MaterialExpressionStaticSwitchParameter_13, output=""}
   ]
   properties {defaultValue=true, group=MistyBead|Spherical Harmonics, parameterName=UseRotateVector_FromSkinning_SH_Dir}
  }
  MaterialExpressionStaticSwitchParameter_10 {
   class=MaterialExpressionStaticSwitchParameter
   inputs [
    True {from=MaterialExpressionAdd_11, output=""}
    False {from=MaterialExpressionNamedRerouteUsage_36, output=""}
   ]
   properties {group=MistyBead|WPO, parameterName=UseSlurpWPO}
  }
  MaterialExpressionStaticSwitchParameter_13 {
   class=MaterialExpressionStaticSwitchParameter
   inputs [
    True {from=MaterialExpressionNormalize_14, output=""}
    False {from=MaterialExpressionTransform_2, output=""}
   ]
   properties {group=MistyBead|Spherical Harmonics, parameterName=UseCustomDirection}
  }
  MaterialExpressionStaticSwitchParameter_2 {
   class=MaterialExpressionStaticSwitchParameter
   inputs [
    True {from=MaterialExpressionSaturate_13, output=""}
    False {from=MaterialExpressionMultiply_9, output=""}
   ]
   properties {group=MistyBead|Reactivity, parameterName=DebugTimestamp}
  }
  MaterialExpressionStaticSwitchParameter_3 {
   class=MaterialExpressionStaticSwitchParameter
   inputs [
    True {from=MaterialExpressionIf_1, output=""}
    False {from=MaterialExpressionConstant_3, output=""}
   ]
   properties {defaultValue=true, group=MistyBead|Reactivity, parameterName=UseReactivity}
  }
  MaterialExpressionStaticSwitchParameter_4 {
   class=MaterialExpressionStaticSwitchParameter
   inputs [
    True {from=MaterialExpressionSetMaterialAttributes_1, output=""}
    False {from=MaterialExpressionNamedRerouteUsage_76, output=""}
   ]
   properties {group=MistyBead, parameterName=UseBaseSkinCaustics}
  }
  MaterialExpressionStaticSwitchParameter_5 {
   class=MaterialExpressionStaticSwitchParameter
   inputs [
    True {from=MaterialExpressionMultiply_68, output=""}
    False {from=MaterialExpressionConstant_6, output=""}
   ]
   properties {group=MistyBead|Triplannar, parameterName=UseSecondTriplannarDeformer}
  }
  MaterialExpressionStaticSwitchParameter_6 {
   class=MaterialExpressionStaticSwitchParameter
   inputs [
    True {from=MaterialExpressionSaturate_3, output=""}
    False {from=MaterialExpressionConstant_14, output=""}
   ]
   properties {group=MistyBead|Foam, parameterName=UseFoam}
  }
  MaterialExpressionStaticSwitchParameter_7 {
   class=MaterialExpressionStaticSwitchParameter
   inputs [
    True {from=MaterialExpressionMaterialFunctionCall_4, output=Result}
    False {from=MaterialExpressionNamedRerouteUsage_0, output=""}
   ]
   properties {group=MistyBead|PanningNormal, parameterName=UsePanningNormals}
  }
  MaterialExpressionStaticSwitchParameter_8 {
   class=MaterialExpressionStaticSwitchParameter
   inputs [
    True {from=MaterialExpressionAdd_55, output=""}
    False {from=MaterialExpressionAppendVector_1, output=""}
   ]
   properties {defaultValue=true, group=MistyBead|Spherical Harmonics, parameterName=UseCustomSH Fit}
  }
  MaterialExpressionStaticSwitchParameter_9 {
   class=MaterialExpressionStaticSwitchParameter
   inputs [
    True {from=MaterialExpressionMultiply_33, output=""}
    False {from=MaterialExpressionConstant_20, output=""}
   ]
   properties {group=MistyBead|Reactivity, parameterName=UseTextureReactivity}
  }
  MaterialExpressionSubtract_0 {
   class=MaterialExpressionSubtract
   inputs [
    A {from=MaterialExpressionOneMinus_1}
    B {from=""}
   ]
   properties {constB=0.5}
  }
  MaterialExpressionSubtract_1 {
   class=MaterialExpressionSubtract
   inputs [
    A {from=MaterialExpressionTime_4}
    B {from=MaterialExpressionScalarParameter_23}
   ]
  }
  MaterialExpressionSubtract_10 {
   class=MaterialExpressionSubtract
   inputs [
    A {from=MaterialExpressionNamedRerouteUsage_37}
    B {from=MaterialExpressionOneMinus_3}
   ]
   properties {constB=0.75}
  }
  MaterialExpressionSubtract_11 {
   class=MaterialExpressionSubtract
   inputs [
    A {from=MaterialExpressionTextureSample_8, output=G}
    B {from=""}
   ]
   properties {constB=0}
  }
  MaterialExpressionSubtract_12 {
   class=MaterialExpressionSubtract
   inputs [
    A {from=MaterialExpressionTime_8}
    B {from=MaterialExpressionScalarParameter_76}
   ]
  }
  MaterialExpressionSubtract_13 {
   class=MaterialExpressionSubtract
   inputs [
    A {from=MaterialExpressionAppendVector_21}
    B {from=MaterialExpressionAppendVector_22}
   ]
  }
  MaterialExpressionSubtract_14 {
   class=MaterialExpressionSubtract
   inputs [
    A {from=MaterialExpressionNamedRerouteUsage_17}
    B {from=MaterialExpressionMultiply_64}
   ]
  }
  MaterialExpressionSubtract_15 {
   class=MaterialExpressionSubtract
   inputs [
    A {from=MaterialExpressionNamedRerouteUsage_8}
    B {from=MaterialExpressionTransformPosition_3}
   ]
  }
  MaterialExpressionSubtract_16 {
   class=MaterialExpressionSubtract
   inputs [
    A {from=MaterialExpressionSubtract_3}
    B {from=MaterialExpressionSaturate_18}
   ]
  }
  MaterialExpressionSubtract_17 {
   class=MaterialExpressionSubtract
   inputs [
    A {from=MaterialExpressionNamedRerouteUsage_96}
    B {from=""}
   ]
   properties {constB=-0.5}
  }
  MaterialExpressionSubtract_18 {
   class=MaterialExpressionSubtract
   inputs [
    A {from=MaterialExpressionAdd_21}
    B {from=MaterialExpressionMultiply_86}
   ]
  }
  MaterialExpressionSubtract_2 {
   class=MaterialExpressionSubtract
   inputs [
    A {from=MaterialExpressionOneMinus_0}
    B {from=""}
   ]
   properties {constB=0.5}
  }
  MaterialExpressionSubtract_20 {
   class=MaterialExpressionSubtract
   inputs [
    A {from=MaterialExpressionNamedRerouteUsage_70}
    B {from=MaterialExpressionScalarParameter_107}
   ]
  }
  MaterialExpressionSubtract_21 {
   class=MaterialExpressionSubtract
   inputs [
    A {from=MaterialExpressionAdd_58}
    B {from=""}
   ]
   properties {constB=2}
  }
  MaterialExpressionSubtract_22 {
   class=MaterialExpressionSubtract
   inputs [
    A {from=MaterialExpressionSubtract_21}
    B {from=MaterialExpressionFrac_3}
   ]
  }
  MaterialExpressionSubtract_24 {
   class=MaterialExpressionSubtract
   inputs [
    A {from=MaterialExpressionTime_9}
    B {from=MaterialExpressionScalarParameter_71}
   ]
  }
  MaterialExpressionSubtract_25 {
   class=MaterialExpressionSubtract
   inputs [
    A {from=MaterialExpressionAdd_73}
    B {from=""}
   ]
   properties {constB=2}
  }
  MaterialExpressionSubtract_26 {
   class=MaterialExpressionSubtract
   inputs [
    A {from=MaterialExpressionSubtract_25}
    B {from=MaterialExpressionFrac_4}
   ]
  }
  MaterialExpressionSubtract_27 {
   class=MaterialExpressionSubtract
   inputs [
    A {from=MaterialExpressionFrac_5}
    B {from=""}
   ]
   properties {constB=0.5}
  }
  MaterialExpressionSubtract_28 {
   class=MaterialExpressionSubtract
   inputs [
    A {from=MaterialExpressionTextureSample_11, output=R}
    B {from=""}
   ]
   properties {constB=0.5}
  }
  MaterialExpressionSubtract_3 {
   class=MaterialExpressionSubtract
   inputs [
    A {from=MaterialExpressionPower_8}
    B {from=MaterialExpressionScalarParameter_35}
   ]
  }
  MaterialExpressionSubtract_4 {
   class=MaterialExpressionSubtract
   inputs [
    A {from=MaterialExpressionSaturate_4}
    B {from=MaterialExpressionMultiply_63}
   ]
  }
  MaterialExpressionSubtract_5 {
   class=MaterialExpressionSubtract
   inputs [
    A {from=MaterialExpressionAdd_23}
    B {from=MaterialExpressionScalarParameter_36}
   ]
   properties {constB=0.2}
  }
  MaterialExpressionSubtract_6 {
   class=MaterialExpressionSubtract
   inputs [
    A {from=MaterialExpressionTextureSample_1, output=R}
    B {from=""}
   ]
   properties {constB=0.5}
  }
  MaterialExpressionSubtract_7 {
   class=MaterialExpressionSubtract
   inputs [
    A {from=MaterialExpressionOneMinus_2}
    B {from=MaterialExpressionScalarParameter_124}
   ]
   properties {constB=-0.2}
  }
  MaterialExpressionSubtract_8 {
   class=MaterialExpressionSubtract
   inputs [
    A {from=MaterialExpressionTime_8}
    B {from=MaterialExpressionAdd_5}
   ]
  }
  MaterialExpressionSubtract_9 {
   class=MaterialExpressionSubtract
   inputs [
    A {from=MaterialExpressionSubtract_16}
    B {from=MaterialExpressionMultiply_31}
   ]
  }
  MaterialExpressionSwitch_0 {
   class=MaterialExpressionSwitch
   inputs [
    SwitchValue {from=MaterialExpressionClamp_2, output=""}
    Default {from="", output=""}
    None {from=MaterialExpressionConstant_0, output=""}
    None {from=MaterialExpressionScalarParameter_14, output=""}
    None {from=MaterialExpressionScalarParameter_46, output=""}
    None {from=MaterialExpressionScalarParameter_67, output=""}
   ]
   properties {inputs [
     {
      inputName=None
      input {
       expression=&self:MaterialExpressionConstant_0
       inputName=None
       mask=0
       maskA=0
       maskB=0
       maskG=0
       maskR=0
       outputIndex=0
      }
     }
     {
      inputName=None
      input {
       expression=&self:MaterialExpressionScalarParameter_14
       inputName=None
       mask=0
       maskA=0
       maskB=0
       maskG=0
       maskR=0
       outputIndex=0
      }
     }
     {
      inputName=None
      input {
       expression=&self:MaterialExpressionScalarParameter_46
       inputName=None
       mask=0
       maskA=0
       maskB=0
       maskG=0
       maskR=0
       outputIndex=0
      }
     }
     {
      inputName=None
      input {
       expression=&self:MaterialExpressionScalarParameter_67
       inputName=None
       mask=0
       maskA=0
       maskB=0
       maskG=0
       maskR=0
       outputIndex=0
      }
     }
    ]}
  }
  MaterialExpressionSwitch_2 {
   class=MaterialExpressionSwitch
   inputs [
    SwitchValue {from=MaterialExpressionClamp_0, output=""}
    Default {from="", output=""}
    None {from=MaterialExpressionReroute_1, output=""}
    None {from=MaterialExpressionReroute_2, output=""}
    None {from=MaterialExpressionReroute_3, output=""}
    None {from=MaterialExpressionReroute_5, output=""}
   ]
   properties {inputs [
     {
      inputName=None
      input {
       expression=&self:MaterialExpressionReroute_1
       inputName=None
       mask=0
       maskA=0
       maskB=0
       maskG=0
       maskR=0
       outputIndex=0
      }
     }
     {
      inputName=None
      input {
       expression=&self:MaterialExpressionReroute_2
       inputName=None
       mask=0
       maskA=0
       maskB=0
       maskG=0
       maskR=0
       outputIndex=0
      }
     }
     {
      inputName=None
      input {
       expression=&self:MaterialExpressionReroute_3
       inputName=None
       mask=0
       maskA=0
       maskB=0
       maskG=0
       maskR=0
       outputIndex=0
      }
     }
     {
      inputName=None
      input {
       expression=&self:MaterialExpressionReroute_5
       inputName=None
       mask=0
       maskA=0
       maskB=0
       maskG=0
       maskR=0
       outputIndex=0
      }
     }
    ]}
  }
  MaterialExpressionTextureCoordinate_0 {class=MaterialExpressionTextureCoordinate}
  MaterialExpressionTextureCoordinate_1 {class=MaterialExpressionTextureCoordinate}
  MaterialExpressionTextureCoordinate_2 {class=MaterialExpressionTextureCoordinate}
  MaterialExpressionTextureCoordinate_5 {class=MaterialExpressionTextureCoordinate}
  MaterialExpressionTextureObjectParameter_0 {
   class=MaterialExpressionTextureObjectParameter
   inputs [
    UVs {from="", output=""}
    Apply View MipBias {from="", output=""}
   ]
   properties {
    group=MistyBead|PanningNormal
    parameterName=PanningNormalMap
    samplerType=SAMPLERTYPE_Normal
    texture=&/T_MistyBead_Normal
   }
  }
  MaterialExpressionTextureObjectParameter_1 {
   class=MaterialExpressionTextureObjectParameter
   inputs [
    UVs {from="", output=""}
    Apply View MipBias {from="", output=""}
   ]
   properties {
    group=MistyBead|Foam
    parameterName=FoamTexture
    texture=/BRCosmetics/Athena/Effects/Particles/Skydiving/NS_Contrail_PinkSpike/Textures/T_PinkSpike_Voronoi2D_Height
   }
  }
  MaterialExpressionTextureObject_1 {
   class=MaterialExpressionTextureObject
   inputs []
   properties {samplerType=SAMPLERTYPE_LinearColor, texture=&/T_MistyBead_Mask}
  }
  MaterialExpressionTextureObject_2 {
   class=MaterialExpressionTextureObject
   inputs []
   properties {samplerType=SAMPLERTYPE_LinearColor, texture=&T_MistyBead_Mask}
  }
  MaterialExpressionTextureSampleParameter2D_1 {
   class=MaterialExpressionTextureSampleParameter2D
   inputs [
    UVs {from="", output=""}
    Apply View MipBias {from="", output=""}
   ]
   properties {
    group=MistyBead|Reactivity
    parameterName=EffectMask
    samplerType=SAMPLERTYPE_LinearGrayscale
    texture=&/T_MistyBead_TechArt_Body_ReactivityMask
   }
  }
  MaterialExpressionTextureSampleParameter2D_4 {
   class=MaterialExpressionTextureSampleParameter2D
   inputs [
    UVs {from="", output=""}
    Apply View MipBias {from="", output=""}
   ]
   properties {parameterName=WPO_Mask, samplerType=SAMPLERTYPE_Masks, texture=&T_WhiteMask}
  }
  MaterialExpressionTextureSample_0 {
   class=MaterialExpressionTextureSample
   inputs [
    UVs {from=MaterialExpressionPanner_2}
    Tex {from=""}
    Apply View MipBias {from=""}
   ]
   properties {texture=&T_PinkSpike_Voronoi2D_Height}
  }
  MaterialExpressionTextureSample_1 {
   class=MaterialExpressionTextureSample
   inputs [
    UVs {from=MaterialExpressionPanner_1}
    Tex {from=""}
    Apply View MipBias {from=""}
   ]
   properties {texture=&T_PinkSpike_Voronoi2D_Height}
  }
  MaterialExpressionTextureSample_11 {
   class=MaterialExpressionTextureSample
   inputs [
    UVs {from=MaterialExpressionPanner_8}
    Tex {from=""}
    "DDX(UVs)" {from=MaterialExpressionMaterialFunctionCall_28, output=DDX}
    "DDY(UVs)" {from=MaterialExpressionMaterialFunctionCall_28, output=DDX}
    Apply View MipBias {from=""}
   ]
   properties {
    mipValueMode=TMVM_Derivative
    samplerType=SAMPLERTYPE_LinearColor
    texture=&T_MistyBead_Mask
   }
  }
  MaterialExpressionTextureSample_2 {
   class=MaterialExpressionTextureSample
   inputs [
    UVs {from=MaterialExpressionPanner_0}
    Tex {from=""}
    Apply View MipBias {from=""}
   ]
   properties {samplerType=SAMPLERTYPE_LinearColor, texture=&T_MistyBead_Mask}
  }
  MaterialExpressionTextureSample_8 {
   class=MaterialExpressionTextureSample
   inputs [
    UVs {from=MaterialExpressionPanner_6}
    Tex {from=""}
    Apply View MipBias {from=""}
   ]
   properties {texture=&T_PinkSpike_Voronoi2D_Height}
  }
  MaterialExpressionTime_0 {class=MaterialExpressionTime}
  MaterialExpressionTime_1 {class=MaterialExpressionTime}
  MaterialExpressionTime_11 {class=MaterialExpressionTime}
  MaterialExpressionTime_2 {class=MaterialExpressionTime}
  MaterialExpressionTime_3 {class=MaterialExpressionTime}
  MaterialExpressionTime_4 {class=MaterialExpressionTime}
  MaterialExpressionTime_7 {class=MaterialExpressionTime}
  MaterialExpressionTime_8 {class=MaterialExpressionTime}
  MaterialExpressionTime_9 {class=MaterialExpressionTime}
  MaterialExpressionTransformPosition_3 {
   class=MaterialExpressionTransformPosition
   inputs [None {from=MaterialExpressionCameraPositionWS_3, output=""}]
   properties {transformSourceType=TRANSFORMPOSSOURCE_World}
  }
  MaterialExpressionTransform_0 {class=MaterialExpressionTransform, inputs [None {from=MaterialExpressionNamedRerouteUsage_38}]}
  MaterialExpressionTransform_1 {
   class=MaterialExpressionTransform
   inputs [None {from=MaterialExpressionVertexNormalWS_2}]
   properties {transformSourceType=TRANSFORMSOURCE_World, transformType=TRANSFORM_Local}
  }
  MaterialExpressionTransform_2 {
   class=MaterialExpressionTransform
   inputs [None {from=MaterialExpressionComponentMask_19}]
   properties {transformSourceType=TRANSFORMSOURCE_World, transformType=TRANSFORM_Local}
  }
  MaterialExpressionVectorParameter_0 {
   class=MaterialExpressionVectorParameter
   properties {group=MistyBead|PanningNormal, parameterName=PanningNormalScaleRatio, defaultValue=<rgba 0.4 1 1 1>}
  }
  MaterialExpressionVectorParameter_1 {
   class=MaterialExpressionVectorParameter
   properties {group=MistyBead|PanningNormal, parameterName=PanningNormalSpeedRatio, defaultValue=<rgba 0 0.5 1 1>}
  }
  MaterialExpressionVectorParameter_10 {
   class=MaterialExpressionVectorParameter
   properties {group=MistyBead|Caustics, parameterName=Triplanar_1_SpeedRatio, defaultValue=<rgba 0 0 1 1>}
  }
  MaterialExpressionVectorParameter_11 {
   class=MaterialExpressionVectorParameter
   properties {group=MistyBead|Caustics, parameterName=Triplanar_2_SpeedRatio, defaultValue=<rgba 0 0 -1 1>}
  }
  MaterialExpressionVectorParameter_13 {
   class=MaterialExpressionVectorParameter
   properties {group=MistyBead|BaseWater, parameterName=FacialColor, defaultValue=<rgba 0 0.1303 0.4062 1>}
  }
  MaterialExpressionVectorParameter_14 {
   class=MaterialExpressionVectorParameter
   properties {group=MistyBead|BaseWater, parameterName=SubsurfaceColor, defaultValue=<rgba 0 3.5 3.3446 1>}
  }
  MaterialExpressionVectorParameter_15 {
   class=MaterialExpressionVectorParameter
   properties {group=MistyBead, parameterName=SkinCausticsColor, defaultValue=<rgba 0.6719 0.8995 1 1>}
  }
  MaterialExpressionVectorParameter_2 {
   class=MaterialExpressionVectorParameter
   properties {group=MistyBead|Triplannar, parameterName=Triplanar_1_Offset, defaultValue=<rgba 2.395 0.49 3.9684 1>}
  }
  MaterialExpressionVectorParameter_22 {
   class=MaterialExpressionVectorParameter
   properties {group=MistyBead|Spherical Harmonics, parameterName=SHMax, defaultValue=<rgba 1.298 0.112 0.1881 0.1547>}
  }
  MaterialExpressionVectorParameter_23 {
   class=MaterialExpressionVectorParameter
   properties {group=MistyBead|Spherical Harmonics, parameterName=SHMin, defaultValue=<rgba 1.0672 -0.1509 -0.1808 -0.1674>}
  }
  MaterialExpressionVectorParameter_25 {
   class=MaterialExpressionVectorParameter
   properties {group=MistyBead|Spherical Harmonics, parameterName=Slurp_CustomDirectionVector, defaultValue=<rgba 0 0 -1 1>}
  }
  MaterialExpressionVectorParameter_3 {class=MaterialExpressionVectorParameter, properties {group=MistyBead|Foam, parameterName=FoamColor, defaultValue=<rgba 1 1 1 0.9>}}
  MaterialExpressionVectorParameter_4 {
   class=MaterialExpressionVectorParameter
   properties {group=MistyBead|BaseWater, parameterName=WaterColorMin, defaultValue=<rgba 0.0828 0.358 0.7812 1>}
  }
  MaterialExpressionVectorParameter_5 {
   class=MaterialExpressionVectorParameter
   properties {group=MistyBead|BaseWater, parameterName=WaterColorMax, defaultValue=<rgba 0 0.7473 2 1>}
  }
  MaterialExpressionVectorParameter_6 {
   class=MaterialExpressionVectorParameter
   properties {group=MistyBead|Caustics, parameterName=CausticsColor, defaultValue=<rgba 0.0892 0.5099 1 1>}
  }
  MaterialExpressionVectorParameter_7 {
   class=MaterialExpressionVectorParameter
   properties {group=MistyBead|Triplannar, parameterName=Triplanar_1_Scale_Ratio, defaultValue=<rgba 1 1 0.8 1>}
  }
  MaterialExpressionVectorParameter_8 {
   class=MaterialExpressionVectorParameter
   properties {group=MistyBead|BaseWater, parameterName=WaterColorExterior, defaultValue=<rgba 0.0272 0.0629 0.3073 1>}
  }
  MaterialExpressionVectorParameter_9 {class=MaterialExpressionVectorParameter, properties {group=MistyBead|Foam, parameterName=FoamScaleRatio, defaultValue=<rgba 1 1 0.5 1>}}
  MaterialExpressionVertexColor_0 {class=MaterialExpressionVertexColor, inputs [], properties {}}
  MaterialExpressionVertexColor_1 {class=MaterialExpressionVertexColor, inputs [], properties {}}
  MaterialExpressionVertexInterpolator_0 {class=MaterialExpressionVertexInterpolator, inputs [VS {from=MaterialExpressionAdd_0, output=""}], properties {}}
  MaterialExpressionVertexInterpolator_1 {class=MaterialExpressionVertexInterpolator, inputs [VS {from=MaterialExpressionPreSkinnedNormal_1, output=""}], properties {}}
  MaterialExpressionVertexInterpolator_2 {class=MaterialExpressionVertexInterpolator, inputs [VS {from=MaterialExpressionLocalPosition_1, output=XYZ}], properties {}}
  MaterialExpressionVertexNormalWS_0 {class=MaterialExpressionVertexNormalWS, inputs [], properties {}}
  MaterialExpressionVertexNormalWS_2 {class=MaterialExpressionVertexNormalWS, inputs [], properties {}}
  MaterialExpressionWorldPosition_0 {class=MaterialExpressionWorldPosition, inputs [], properties {worldPositionShaderOffset=WPT_ExcludeAllShaderOffsets}}
 }}
