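# Parent material for the MistyBead female medium player body cosmetic, built on material attributes for skeletal meshes with morph targets and mesh deformers.
# The shared 2023 character PreFX function supplies the base material and masks, with named reroutes around the character-specific MF_MistyBead effect and the shared PostFX function.
# Shading model comes from the material expression, spanning lit, subsurface, clear coat and cloth models.
# A face accessory FX mask texture sample has no visible connections.
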Material /BRCosmetics/Characters/Player/Female/Medium/Bodies/F_MED_MistyBead/Materials/M_MistyBead_Parent
graph {
 properties {
  bAutomaticallySetUsageInEditor=false
  bUseMaterialAttributes=true
  bUsedWithMeshDeformer=true
  bUsedWithMorphTargets=true
  bUsedWithSkeletalMesh=true
  pixelDepthOffsetMode=PDOM_Legacy
  shadingModel=MSM_FromMaterialExpression
  usedShadingModels=Default Lit | Subsurface | Clear Coat | Subsurface Profile | Cloth
 }
 nodes [
  n1 = MaterialFunctionCall() {materialFunction=&MF_CharacterParent_2023_PreFX}
  MaterialFunctionCall(Material=NamedRerouteUsage(), M=@n1.M, Pre Skinned Local Position=@n1.M, Distance Blend=@n1.M, Fresnel=@n1.M) {materialFunction=&MF_CharacterParent_2023_PostFX}
  NamedRerouteDeclaration(None=@n1.Result) {name=InMaterial, nodeColor=<rgba 0.5234 1 0 1>}
  NamedRerouteDeclaration(None=MaterialFunctionCall!(InMaterial=NamedRerouteUsage(), MaskMap=NamedRerouteUsage(), FXMask=NamedRerouteUsage()) {materialFunction=&/MF_MistyBead}.Result) {name=OutMaterial, nodeColor=<rgba 1 0 0.8437 1>}
  NamedRerouteDeclaration(None=@n1.Tech Art Mask) {name=InFXMask, nodeColor=<rgba 0.0312 0 1 1>}
  NamedRerouteDeclaration(None=@n1.M) {name=InMask, nodeColor=<rgba 0 1 0.6719 1>}
  TextureSample() {samplerType=SAMPLERTYPE_Masks, texture=/BRCosmetics/Characters/Player/Female/Medium/Bodies/F_MED_MistyBead/Textures/T_MistyBead_FaceAcc_FX}
 ]
}
