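# Physics asset for the MistyBead female medium character's face accessory part, covering two earrings and three back-hair chains.
# Capsules on torso, head, neck, clavicles and upper arms form the body, linked mostly by free constraints.
# Each earring locks its first dynamic bone to its base and limits the second.
# Hair chains hang from master bones constrained to spine_02, widen swing limits toward the tips, and have some free cross-links between strands.
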PhysicsAsset /BRCosmetics/Characters/Player/Female/Medium/Bodies/F_MED_MistyBead/Meshes/Parts/F_MED_MistyBead_FaceAcc_Physics
bodies {
 base_earring_l_01 [{shapeType=Capsule, length=0.505, radius=0.505, shapeName=base_earring_l_01_capsule, rotation=<rot 0 -90 0>}]
 base_earring_r_01 [{shapeType=Capsule, length=0.505, radius=0.505, shapeName=base_earring_r_01_capsule, rotation=<rot 0 -90 0>}]
 clavicle_l [{shapeType=Capsule, length=8.5, radius=6, shapeName=clavicle_l_capsule, center=<4 0 0>, rotation=<rot 0 -90 -90>}]
 clavicle_r [{shapeType=Capsule, length=8.5, radius=6, shapeName=clavicle_r_capsule, center=<-4 0 0>, rotation=<rot 0 -90 -90>}]
 dyn_earring_l_01 [{shapeType=Capsule, length=0.1763, radius=0.642, shapeName=dyn_earring_l_01_capsule, center=<0.2609 0 0>, rotation=<rot 0 91.7717 -84.8809>}]
 dyn_earring_l_02 [{
   shapeType=Capsule
   length=2.007
   radius=0.5
   shapeName=dyn_earring_l_02_capsule
   center=<0.552 -0.0186 0.2683>
   rotation=<rot 0 90.7483 -89.339>
  }]
 dyn_earring_r_01 [{shapeType=Capsule, length=0.1763, radius=0.642, shapeName=dyn_earring_r_01_capsule, center=<0.2142 0 0>, rotation=<rot 0 91.7717 -84.8809>}]
 dyn_earring_r_02 [{
   shapeType=Capsule
   length=2.007
   radius=0.5
   shapeName=dyn_earring_r_02_capsule
   center=<0.8871 0.0186 0.2683>
   rotation=<rot 0 89.2517 -89.339>
  }]
 dyn_hair_bk_01 [{shapeType=Capsule, length=7, radius=3, shapeName=dyn_hair_bk_01_capsule, center=<5 0 2>, rotation=<rot -90 0 0>}]
 dyn_hair_bk_01_L [{shapeType=Capsule, length=7, radius=3, shapeName=dyn_hair_bk_01_L_capsule, center=<5 0 2>, rotation=<rot -90 0.0001 0>}]
 dyn_hair_bk_01_R [{shapeType=Capsule, length=7, radius=3, shapeName=dyn_hair_bk_01_R_capsule, center=<5 0 2>, rotation=<rot -90 -0.0001 0>}]
 dyn_hair_bk_02 [{shapeType=Capsule, length=7, radius=3, shapeName=dyn_hair_bk_02_capsule, center=<5 0 2>, rotation=<rot -90 0 0>}]
 dyn_hair_bk_02_L [{shapeType=Capsule, length=6.5, radius=3.5, shapeName=dyn_hair_bk_02_L_capsule, center=<5 0 2>, rotation=<rot -90 0.0001 0>}]
 dyn_hair_bk_02_R [{shapeType=Capsule, length=6.5, radius=3.5, shapeName=dyn_hair_bk_02_R_capsule, center=<5 0 2>, rotation=<rot -90 -0.0001 0>}]
 dyn_hair_bk_03 [{shapeType=Capsule, length=5, radius=5, shapeName=dyn_hair_bk_03_capsule, center=<5 0 2>, rotation=<rot -90 0 0>}]
 dyn_hair_bk_03_L [{shapeType=Capsule, length=5, radius=5, shapeName=dyn_hair_bk_03_L_capsule, center=<5 0 2>, rotation=<rot -90 0 0>}]
 dyn_hair_bk_03_R [{shapeType=Capsule, length=5, radius=5, shapeName=dyn_hair_bk_03_R_capsule, center=<5 0 2>, rotation=<rot -90 0 0>}]
 dyn_hair_bk_04 [{shapeType=Capsule, length=4, radius=6, shapeName=dyn_hair_bk_04_capsule, center=<5 0 2>, rotation=<rot -90 0 0>}]
 dyn_hair_bk_04_L [{shapeType=Capsule, length=4, radius=6, shapeName=dyn_hair_bk_04_L_capsule, center=<5 0 2>, rotation=<rot -90 0.0001 0>}]
 dyn_hair_bk_04_R [{shapeType=Capsule, length=4, radius=6, shapeName=dyn_hair_bk_04_R_capsule, center=<5 0 2>, rotation=<rot -90 -0.0001 0>}]
 dyn_hair_bk_05 [{shapeType=Capsule, length=2.5, radius=7.5, shapeName=dyn_hair_bk_05_capsule, center=<5 0 2>, rotation=<rot -90 0 0>}]
 dyn_hair_bk_05_L [{shapeType=Capsule, length=2.5, radius=7.5, shapeName=dyn_hair_bk_05_L_capsule, center=<5 0 2>, rotation=<rot -90 0.0001 0>}]
 dyn_hair_bk_05_R [{shapeType=Capsule, length=2.5, radius=7.5, shapeName=dyn_hair_bk_05_R_capsule, center=<5 0 2>, rotation=<rot -90 -0.0001 0>}]
 dyn_hair_bk_06 [{shapeType=Capsule, length=5, radius=5, shapeName=dyn_hair_bk_06_capsule, center=<5 0 2>, rotation=<rot -90 0 0>}]
 dyn_hair_bk_06_L [{shapeType=Capsule, length=5, radius=5, shapeName=dyn_hair_bk_06_L_capsule, center=<5 0 2>, rotation=<rot -90 0.0001 0>}]
 dyn_hair_bk_06_R [{shapeType=Capsule, length=5, radius=5, shapeName=dyn_hair_bk_06_R_capsule, center=<5 0 2>, rotation=<rot -90 -0.0001 0>}]
 dyn_hair_bk_MSTR [{shapeType=Capsule, length=0.5, radius=0.5, shapeName=dyn_hair_bk_MSTR_capsule, rotation=<rot -90 0 0>}]
 dyn_hair_bk_MSTR_L [{shapeType=Capsule, length=0.5, radius=0.5, shapeName=dyn_hair_bk_MSTR_L_capsule, rotation=<rot -90 0 0>}]
 dyn_hair_bk_MSTR_R [{shapeType=Capsule, length=0.5, radius=0.5, shapeName=dyn_hair_bk_MSTR_R_capsule, rotation=<rot -90 0 0>}]
 head [{shapeType=Capsule, length=2.778, radius=7.7585, shapeName=head_capsule, center=<5.6304 1.7188 0>, rotation=<rot -90 9.9997 0>}]
 neck_01 [{shapeType=Capsule, length=5, radius=4.5, shapeName=neck_01_capsule, rotation=<rot -90 0 0>}]
 pelvis [{shapeType=Capsule, length=11, radius=12, shapeName=pelvis_capsule, rotation=<rot 0 -90 0>}]
 root [{shapeType=Capsule, length=0.505, radius=0.505, shapeName=root_capsule, rotation=<rot 0 -90 0>}]
 spine_01 [{shapeType=Capsule, length=6, radius=8.505, shapeName=spine_04_capsule, center=<15.5739 0.8534 0>, rotation=<rot 0 -90.0003 0>}]
 spine_02 [{shapeType=Capsule, length=17.6102, radius=8, shapeName=spine_02_capsule, center=<6.9899 0 0>, rotation=<rot 0 -90 0>}]
 spine_03 [{shapeType=Capsule, length=14, radius=6, shapeName=spine_03_capsule, rotation=<rot 0 -90 0>}]
 spine_04 [{shapeType=Capsule, length=12.005, radius=9.005, shapeName=spine_04_capsule, center=<6.9395 1.0518 0>, rotation=<rot 0 -90.0003 180>}]
 spine_05 [{shapeType=Capsule, length=11, radius=9, shapeName=spine_05_capsule, center=<0 2 0>, rotation=<rot 0 -90 0>}]
 upperarm_l [{shapeType=Capsule, length=17.5, radius=5.5, shapeName=upperarm_l_capsule, center=<13 0 0>, rotation=<rot 0 -89.9996 90>}]
 upperarm_r [{shapeType=Capsule, length=20.8465, radius=6.0605, shapeName=upperarm_r_capsule, center=<-13 0 0>, rotation=<rot 0 -90 90>}]
}
constraints [
 {
  bone1Name=spine_05
  bone2Name=spine_04
  swing1LimitDegrees=45
  swing1Motion=Free
  swing2LimitDegrees=45
  swing2Motion=Free
  twistLimitDegrees=45
  twistMotion=Free
 }
 {
  bone1Name=spine_04
  bone2Name=spine_02
  swing1LimitDegrees=45
  swing1Motion=Free
  swing2LimitDegrees=45
  swing2Motion=Free
  twistLimitDegrees=45
  twistMotion=Free
 }
 {
  bone1Name=clavicle_l
  bone2Name=spine_05
  swing1LimitDegrees=45
  swing1Motion=Free
  swing2LimitDegrees=45
  swing2Motion=Free
  twistLimitDegrees=45
  twistMotion=Free
 }
 {
  bone1Name=clavicle_r
  bone2Name=spine_05
  swing1LimitDegrees=45
  swing1Motion=Free
  swing2LimitDegrees=45
  swing2Motion=Free
  twistLimitDegrees=45
  twistMotion=Free
 }
 {
  bone1Name=upperarm_l
  bone2Name=clavicle_l
  swing1LimitDegrees=45
  swing1Motion=Free
  swing2LimitDegrees=45
  swing2Motion=Free
  twistLimitDegrees=45
  twistMotion=Free
 }
 {
  bone1Name=head
  bone2Name=spine_05
  swing1LimitDegrees=45
  swing1Motion=Free
  swing2LimitDegrees=45
  swing2Motion=Free
  twistLimitDegrees=45
  twistMotion=Free
 }
 {
  bone1Name=base_earring_l_01
  bone2Name=head
  swing1LimitDegrees=45
  swing1Motion=Free
  swing2LimitDegrees=45
  swing2Motion=Free
  twistLimitDegrees=45
  twistMotion=Free
 }
 {
  bone1Name=dyn_earring_l_01
  bone2Name=base_earring_l_01
  swing1LimitDegrees=45
  swing1Motion=Locked
  swing2LimitDegrees=45
  swing2Motion=Locked
  twistLimitDegrees=45
  twistMotion=Locked
 }
 {bone1Name=dyn_earring_l_02, bone2Name=dyn_earring_l_01, swing1LimitDegrees=25, swing2LimitDegrees=30, twistLimitDegrees=5}
 {
  bone1Name=base_earring_r_01
  bone2Name=head
  swing1LimitDegrees=45
  swing1Motion=Free
  swing2LimitDegrees=45
  swing2Motion=Free
  twistLimitDegrees=45
  twistMotion=Free
 }
 {
  bone1Name=dyn_earring_r_01
  bone2Name=base_earring_r_01
  swing1LimitDegrees=45
  swing1Motion=Locked
  swing2LimitDegrees=45
  swing2Motion=Locked
  twistLimitDegrees=45
  twistMotion=Locked
 }
 {bone1Name=dyn_earring_r_02, bone2Name=dyn_earring_r_01, swing1LimitDegrees=25, swing2LimitDegrees=30, twistLimitDegrees=5}
 {
  bone1Name=upperarm_r
  bone2Name=clavicle_r
  swing1LimitDegrees=45
  swing1Motion=Free
  swing2LimitDegrees=45
  swing2Motion=Free
  twistLimitDegrees=45
  twistMotion=Free
 }
 {
  bone1Name=spine_01
  bone2Name=pelvis
  swing1LimitDegrees=45
  swing1Motion=Free
  swing2LimitDegrees=45
  swing2Motion=Free
  twistLimitDegrees=45
  twistMotion=Free
 }
 {
  bone1Name=pelvis
  bone2Name=root
  swing1LimitDegrees=45
  swing1Motion=Free
  swing2LimitDegrees=45
  swing2Motion=Free
  twistLimitDegrees=45
  twistMotion=Free
 }
 {
  bone1Name=spine_03
  bone2Name=spine_02
  swing1LimitDegrees=45
  swing1Motion=Free
  swing2LimitDegrees=45
  swing2Motion=Free
  twistLimitDegrees=45
  twistMotion=Free
 }
 {
  bone1Name=neck_01
  bone2Name=spine_05
  swing1LimitDegrees=45
  swing1Motion=Free
  swing2LimitDegrees=45
  swing2Motion=Free
  twistLimitDegrees=45
  twistMotion=Free
 }
 {
  bone1Name=head
  bone2Name=neck_01
  swing1LimitDegrees=45
  swing1Motion=Free
  swing2LimitDegrees=45
  swing2Motion=Free
  twistLimitDegrees=45
  twistMotion=Free
 }
 {bone1Name=dyn_hair_bk_01, bone2Name=dyn_hair_bk_MSTR, swing1LimitDegrees=10, swing2LimitDegrees=10, twistLimitDegrees=45}
 {bone1Name=dyn_hair_bk_02, bone2Name=dyn_hair_bk_01, swing1LimitDegrees=5, swing2LimitDegrees=15, twistLimitDegrees=5}
 {bone1Name=dyn_hair_bk_03, bone2Name=dyn_hair_bk_02, swing1LimitDegrees=10, swing2LimitDegrees=30, twistLimitDegrees=2.5}
 {bone1Name=dyn_hair_bk_04, bone2Name=dyn_hair_bk_03, swing1LimitDegrees=10, swing2LimitDegrees=40, twistLimitDegrees=2.5}
 {bone1Name=dyn_hair_bk_05, bone2Name=dyn_hair_bk_04, swing1LimitDegrees=10, swing2LimitDegrees=45, twistLimitDegrees=2.5}
 {
  bone1Name=dyn_hair_bk_05_R
  bone2Name=dyn_hair_bk_04
  swing1LimitDegrees=45
  swing1Motion=Free
  swing2LimitDegrees=45
  swing2Motion=Free
  twistLimitDegrees=45
  twistMotion=Free
 }
 {
  bone1Name=dyn_hair_bk_05_L
  bone2Name=dyn_hair_bk_04
  swing1LimitDegrees=45
  swing1Motion=Free
  swing2LimitDegrees=45
  swing2Motion=Free
  twistLimitDegrees=45
  twistMotion=Free
 }
 {bone1Name=dyn_hair_bk_06, bone2Name=dyn_hair_bk_05, swing1LimitDegrees=10, swing2LimitDegrees=50, twistLimitDegrees=2.5}
 {bone1Name=dyn_hair_bk_01_L, bone2Name=dyn_hair_bk_MSTR_L, swing1LimitDegrees=2.5, swing2LimitDegrees=10, twistLimitDegrees=10}
 {bone1Name=dyn_hair_bk_02_L, bone2Name=dyn_hair_bk_01_L, swing1LimitDegrees=5, swing2LimitDegrees=15, twistLimitDegrees=5}
 {bone1Name=dyn_hair_bk_03_L, bone2Name=dyn_hair_bk_02_L, swing1LimitDegrees=10, swing2LimitDegrees=30, twistLimitDegrees=2.5}
 {bone1Name=dyn_hair_bk_04_L, bone2Name=dyn_hair_bk_03_L, swing1LimitDegrees=10, swing2LimitDegrees=40, twistLimitDegrees=2.5}
 {bone1Name=dyn_hair_bk_05_L, bone2Name=dyn_hair_bk_04_L, swing1LimitDegrees=10, swing2LimitDegrees=45, twistLimitDegrees=2.5}
 {bone1Name=dyn_hair_bk_06_L, bone2Name=dyn_hair_bk_05_L, swing1LimitDegrees=10, swing2LimitDegrees=50, twistLimitDegrees=2.5}
 {bone1Name=dyn_hair_bk_01_R, bone2Name=dyn_hair_bk_MSTR_R, swing1LimitDegrees=2.5, swing2LimitDegrees=10, twistLimitDegrees=10}
 {bone1Name=dyn_hair_bk_02_R, bone2Name=dyn_hair_bk_01_R, swing1LimitDegrees=5, swing2LimitDegrees=15, twistLimitDegrees=5}
 {bone1Name=dyn_hair_bk_03_R, bone2Name=dyn_hair_bk_02_R, swing1LimitDegrees=10, swing2LimitDegrees=30, twistLimitDegrees=2.5}
 {bone1Name=dyn_hair_bk_04_R, bone2Name=dyn_hair_bk_03_R, swing1LimitDegrees=10, swing2LimitDegrees=40, twistLimitDegrees=2.5}
 {bone1Name=dyn_hair_bk_05_R, bone2Name=dyn_hair_bk_04_R, swing1LimitDegrees=10, swing2LimitDegrees=45, twistLimitDegrees=2.5}
 {bone1Name=dyn_hair_bk_06_R, bone2Name=dyn_hair_bk_05_R, swing1LimitDegrees=10, swing2LimitDegrees=50, twistLimitDegrees=2.5}
 {
  bone1Name=dyn_hair_bk_02_R
  bone2Name=dyn_hair_bk_02
  swing1LimitDegrees=45
  swing1Motion=Free
  swing2LimitDegrees=45
  swing2Motion=Free
  twistLimitDegrees=45
  twistMotion=Free
 }
 {
  bone1Name=dyn_hair_bk_02
  bone2Name=dyn_hair_bk_02_R
  swing1LimitDegrees=45
  swing1Motion=Free
  swing2LimitDegrees=45
  swing2Motion=Free
  twistLimitDegrees=45
  twistMotion=Free
 }
 {
  bone1Name=dyn_hair_bk_04_R
  bone2Name=dyn_hair_bk_03
  swing1LimitDegrees=45
  swing1Motion=Free
  swing2LimitDegrees=45
  swing2Motion=Free
  twistLimitDegrees=45
  twistMotion=Free
 }
 {
  bone1Name=dyn_hair_bk_04_L
  bone2Name=dyn_hair_bk_03
  swing1LimitDegrees=45
  swing1Motion=Free
  swing2LimitDegrees=45
  swing2Motion=Free
  twistLimitDegrees=45
  twistMotion=Free
 }
 {
  bone1Name=dyn_hair_bk_MSTR_R
  bone2Name=spine_02
  swing1LimitDegrees=45
  swing1Motion=Free
  swing2LimitDegrees=45
  swing2Motion=Free
  twistLimitDegrees=45
  twistMotion=Free
 }
 {
  bone1Name=dyn_hair_bk_MSTR_L
  bone2Name=spine_02
  swing1LimitDegrees=45
  swing1Motion=Free
  swing2LimitDegrees=45
  swing2Motion=Free
  twistLimitDegrees=45
  twistMotion=Free
 }
 {
  bone1Name=dyn_hair_bk_MSTR
  bone2Name=spine_02
  swing1LimitDegrees=45
  swing1Motion=Free
  swing2LimitDegrees=45
  swing2Motion=Free
  twistLimitDegrees=45
  twistMotion=Free
 }
 {
  bone1Name=spine_02
  bone2Name=spine_01
  swing1LimitDegrees=45
  swing1Motion=Free
  swing2LimitDegrees=45
  swing2Motion=Free
  twistLimitDegrees=45
  twistMotion=Free
 }
]
