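# Post-copy AnimBlueprint for the MistyBead female face accessory that adds secondary motion to earrings and front hair.
# A PreSim Control Rig corrects the parent mesh pose, then RigidBody simulation runs unless a backpack or cape is worn.
# AnimDynamics spring chains drive earring and hair bones before a PostSim Control Rig finishes the pose.
# Events supply movement, weapon, backbling and gravity-override state.
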AnimBlueprint /BRCosmetics/Characters/Player/Female/Medium/Bodies/F_MED_MistyBead/Meshes/Parts/F_MED_MistyBead_FaceAcc_AnimBP
target_skeleton=&/F_MED_MistyBead_FaceAcc_Skeleton
graphs [
 AnimGraph {nodes [
   n1 = RigidBody!(
    ComponentPose=Local To Component(LocalPose=Use cached pose 'Start Pose'())
    Alpha
    OverrideWorldGravity:Vector="0,0,0"
    ExternalForce
    SimSpaceSettings:Sim Space Settings Structure=(WorldAlpha=0,VelocityScaleZ=1,DampingAlpha=1,MaxLinearVelocity=10000,MaxAngularVelocity=10000,MaxLinearAcceleration=10000,MaxAngularAcceleration=10000,ExternalLinearDragV=(X=0,Y=0,Z=0),ExternalLinearVelocity=(X=0,Y=0,Z=0),ExternalAngularVelocity=(X=0,Y=0,Z=0))
   ) -> (Pose) {
    bOverrideWorldGravity=true
    overridePhysicsAsset=&/F_MED_MistyBead_FaceAcc_Physics
    simulationSpace=BaseBoneSpace
    baseBoneRef {boneName=spine_05}
   }
   n2 = Component To Local(ComponentPose=@n1)
   n3 = Use cached pose 'Start Pose'()
   n4 = Get isWearingBackpack/Cape() -> (isWearingBackpack/Cape:bool)
   n5 = Anim Dynamics - Bone: dyn_earring_l_02!(
    SimulationSpace:AnimPhysSimSpaceType Enum=World
    ComponentPose=Local To Component(LocalPose=Blend Poses by bool(BlendPose_0=@n3, BlendPose_1=@n2, bActiveValue=@n4))
    Alpha=1
    ExternalForce="0,0,0"
   ) -> (Pose) {
    angularDampingOverride=0.9999
    angularSpringConstant=8000
    bAngularSpring=true
    bOverrideAngularDamping=true
    bUseGravityOverride=true
    bUsePlanarLimit=false
    bUseSphericalLimits=true
    gravityScale=0.35
    simulationSpace=World
    boundBone {boneName=dyn_earring_l_02}
    componentAppliedLinearAccClamp=<3000 3000 3000>
    componentLinearAccScale=<0.5 0.5 0.5>
    componentLinearVelScale=<0.5 0.5 0.5>
    gravityOverride {z=-980}
    physicsBodyDefinitions [{
      boundBone {boneName=dyn_earring_l_02}
      boxExtents=<15 20 20>
      constraintSetup {angularLimitsMax=<0 5 30>, angularLimitsMin=<0 -5 -5>, angularTarget=<1 0 0>}
      localJointOffset=<4 0 0>
     }]
   }
   n6 = Anim Dynamics - Bone: dyn_earring_r_02!(SimulationSpace:AnimPhysSimSpaceType Enum=World, ComponentPose=@n5, Alpha=1, ExternalForce="0,0,0") -> (Pose) {
    angularDampingOverride=0.9999
    angularSpringConstant=8000
    bAngularSpring=true
    bOverrideAngularDamping=true
    bUseGravityOverride=true
    bUseSphericalLimits=true
    gravityScale=0.35
    simulationSpace=World
    boundBone {boneName=dyn_earring_r_02}
    componentAppliedLinearAccClamp=<3000 3000 3000>
    componentLinearAccScale=<0.5 0.5 0.5>
    componentLinearVelScale=<0.5 0.5 0.5>
    gravityOverride {z=-980}
    physicsBodyDefinitions [{
      boundBone {boneName=dyn_earring_r_02}
      boxExtents=<15 20 20>
      constraintSetup {angularLimitsMax=<0 5 10>, angularLimitsMin=<0 -5 -30>, angularTarget=<1 0 0>}
      localJointOffset=<4 0 0>
     }]
   }
   n7 = Anim Dynamics - Bone: dyn_r_front_hair_01!(ComponentPose=@n6, Alpha=1, GravityOverride:Vector="0,0,-980", ExternalForce="0,0,0") -> (Pose) {
    angularDampingOverride=1
    angularSpringConstant=5000
    bAngularSpring=true
    bOverrideAngularDamping=true
    bUseGravityOverride=true
    gravityScale=0.35
    simulationSpace=RootRelative
    boundBone {boneName=dyn_r_front_hair_01}
    componentAppliedLinearAccClamp=<4000 4000 4000>
    componentLinearAccScale=<1 1 1>
    componentLinearVelScale=<1 1 1>
    gravityOverride {z=-980}
    physicsBodyDefinitions [{
      boundBone {boneName=dyn_r_front_hair_01}
      boxExtents=<20 35 35>
      constraintSetup {angularLimitsMax=<0 1 1>, angularLimitsMin=<0 -5 -4>, angularTarget=<1 0 0>}
      localJointOffset=<4 0 0>
     }]
   }
   n8 = Anim Dynamics - Bone: dyn_l_front_hair_01!(ComponentPose=@n7, Alpha=0.4, GravityOverride:Vector="0,0,-980", ExternalForce="0,0,0") -> (Pose) {
    angularDampingOverride=1
    angularSpringConstant=5000
    bAngularSpring=true
    bOverrideAngularDamping=true
    bUseGravityOverride=true
    gravityScale=0.35
    simulationSpace=RootRelative
    boundBone {boneName=dyn_l_front_hair_01}
    componentAppliedLinearAccClamp=<4000 4000 4000>
    componentLinearAccScale=<1 1 1>
    componentLinearVelScale=<1 1 1>
    gravityOverride {z=-980}
    physicsBodyDefinitions [{
      boundBone {boneName=dyn_l_front_hair_01}
      boxExtents=<20 35 35>
      constraintSetup {angularLimitsMax=<0 1 1>, angularLimitsMin=<0 -5 -4>, angularTarget=<1 0 0>}
      localJointOffset=<4 0 0>
     }]
   }
   n9 = Control Rig!(
    Source=Component To Local(ComponentPose=@n8)
    Alpha
    Crouching:double=Get Crouching() -> (Crouching:double)
    targeting:double=Get isTargeting() -> (isTargeting:double)
    AimPitch:double=Get AimPitch()
    Backbiling:bool=Get isWearingBackpack/Cape()
    "Playing Emote ":double=Get Emote() -> (Emote:double)
    Skydiving:bool=Get bIsSkyDiving()
    isInFrontEnd:double=0
   ) -> (Pose) {controlRigAssetReference {blueprintRigClass=&/F_MED_MistyBead_FaceAcc_PostSim_CtrlRig.F_MED_MistyBead_FaceAcc_PostSim_CtrlRig_C}}
   Output Pose(Result=@n9) {layerGroup=None}
   n10 = Control Rig!(
    Source=Copy Pose From Mesh(SourceMeshComponent=Get Skeletal_Mesh() -> (Skeletal_Mesh:Skeletal Mesh Component Object Reference)) {bUseAttachedParent=true}
    Alpha
    targeting:bool=Get bIsTargeting()
    Crouching=Get Crouching()
    Jumping:double=Get IsJumping() -> (IsJumping:double)
    AR_Equiped:double=Get Rifle() -> (Rifle:double)
    PistolEquiped:double=Get pistol() -> (pistol:double)
    Skydive:bool=Get bIsSkyDiving()
   ) -> (Pose) {controlRigAssetReference {blueprintRigClass=&/F_MED_MistyBead_FaceAcc_PreSim_CtrlRig.F_MED_MistyBead_FaceAcc_PreSim_CtrlRig_C}}
   Save cached pose 'Start Pose'(Pose=@n10) {cacheName=Start Pose}
  ]}
 EventGraph {nodes [
   Event BlueprintUpdateAnimation()
   > n1 = Sequence!(execute) -> (then_0, then_1, then_2:Exec, then_3:Exec)
   n2 = Branch(execute=@n1.then_2, Condition=Get bIsInFrontEnd())
   > Set Sim Space Anim(Sim Space Anim:AnimPhysSimSpaceType Enum=World) -> (Output_Get:AnimPhysSimSpaceType Enum)
   > Set AnimDynAlpha(AnimDynAlpha:double=0.4) -> (Output_Get:double)
   Set Sim Space Anim(execute=@n2.else)
   > Set AnimDynAlpha(AnimDynAlpha=1)
   n3 = "Gravity Override State Function (Parameter Asset) (Nativized)"(
    execute=@n1.then_0
    CharacterPartAnimInstance=Self-Reference()
    GameTimeInSeconds=GetGameTimeInSeconds()
    Parameters=&/F_MED_MistyBead_FaceAcc_New_Parameters
   )
   > Set Gravity Override(Gravity Override=@n3)
   > Set RBAN Settings Out(RBAN Settings Out=@n3.OutRigidBodyAnimNodeParameters)
   > Set Crouching(Crouching:double="To Float (Boolean)"(InBool=Get bIsCrouching())) -> (Output_Get:double)
   n4 = OR Boolean!(A=Get bIsSkydiveFloating(), B=Get bIsSkydiveDiveMode(), C:bool=Get bIsParachuteOpened()) -> (ReturnValue)
   n5 = Get bIsJumping()
   n6 = Get bIsFalling()
   > n7 = Branch(Condition=AND Boolean(A=OR Boolean!(A=@n5, B=@n6) -> (ReturnValue), B=NOT Boolean(A=@n4)))
   > Set IsJumping(IsJumping:double=1) -> (Output_Get:double)
   Set IsJumping(execute=@n7.else, IsJumping=0)
   Set isTargeting(execute=@n1.then_1, isTargeting:double="To Float (Boolean)"(InBool=Get bIsTargeting())) -> (Output_Get:double)
   > Set Emote(Emote:double="To Float (Boolean)"(InBool=Get bIsPlayingEmote())) -> (Output_Get:double)
   > Set isInFrontEndFloat(isInFrontEndFloat:double="To Float (Boolean)"(InBool=Get bIsInFrontEnd())) -> (Output_Get:double)
   Event OnCharacterCustomizationLoaded()
   n8 = CheckForBackbling()
   > Set isWearingBackpack/Cape(isWearingBackpack/Cape:bool=OR Boolean(A=@n8, B=@n8.bIsBackblingACape)) -> (Output_Get:bool)
   Event OnWeaponChanged()
   > n9 = Cast To FortPlayerAnimInstance(Object=GetAnimInstance(self=GetOwnerSkeletalMeshComponent()))
   > Set pistol(pistol:double=0) -> (Output_Get:double)
   > Set Melee(Melee:double=0) -> (Output_Get:double)
   > Set Rifle(Rifle:double=0) -> (Output_Get:double)
   > n10 = Switch on EFortWeaponCoreAnimation!(execute, Selection:EFortWeaponCoreAnimation Enum=Get WeaponCoreAnim(self=@n9)) -> (
    Melee:Exec
    Pistol:Exec
    Shotgun:Exec
    PaperBlueprint:Exec
    Rifle:Exec
    MeleeOneHand:Exec
    MachinePistol:Exec
    RocketLauncher:Exec
    GrenadeLauncher:Exec
    GoingCommando:Exec
    AssaultRifle:Exec
    TacticalShotgun:Exec
    SniperRifle:Exec
    TrapPlacement:Exec
    ShoulderLauncher:Exec
    AbilityDecoTool:Exec
    Crossbow:Exec
    C4:Exec
    RemoteControl:Exec
    DualWield:Exec
    AR_BullPup:Exec
    AR_ForwardGrip:Exec
    MedPackPaddles:Exec
    SMG_P90:Exec
    AR_DrumGun:Exec
    Consumable_Small:Exec
    Consumable_Large:Exec
    Balloon:Exec
    MountedTurret:Exec
    CreativeTool:Exec
    ExplosiveBow:Exec
    AshtonIndigo:Exec
    AshtonChicago:Exec
    MeleeDualWield:Exec
    Unarmed:Exec
   )
   Set Rifle(
    execute=[
     @n10.AR_BullPup
     @n10.AR_DrumGun
     @n10.AR_ForwardGrip
     @n10.AssaultRifle
     @n10.Crossbow
     @n10.GrenadeLauncher
     @n10.Rifle
     @n10.RocketLauncher
     @n10.SMG_P90
     @n10.Shotgun
     @n10.SniperRifle
     @n10.TacticalShotgun
    ]
    Rifle=1
   )
   > Set Only_Rifle(Only_Rifle:bool=1) -> (Output_Get:bool)
   > Set Only_Pistol(Only_Pistol:bool=false) -> (Output_Get:bool)
   Set pistol(execute=[@n10.MachinePistol, @n10.Pistol], pistol=1)
   > Set Only_Rifle(Only_Rifle=false)
   > Set Only_Pistol(Only_Pistol=true)
   Set Melee(execute=@n10.Melee, Melee=1)
   > Set Only_Rifle(Only_Rifle=false)
   > Set Only_Pistol(Only_Pistol=false)
   Event BlueprintInitializeAnimation()
   > Set Skeletal_Mesh(Skeletal_Mesh:Skeletal Mesh Component Object Reference=GetOwnerSkeletalMeshComponent()) -> (Output_Get:Skeletal Mesh Component Object Reference)
   Get bIsTacticalSprinting()
  ]}
]
variables [
 AnimDynAlpha {type=number}
 Crouching {type=number}
 Emote {type=number}
 Gravity Override {type=Vector}
 isInFrontEndFloat {type=number}
 IsJumping {type=number}
 isTargeting {type=number}
 isWearingBackpack/Cape {type=boolean}
 Melee {type=number}
 Only_Pistol {type=boolean}
 Only_Rifle {type=boolean}
 pistol {type=number}
 RBAN Settings Out {type=FortRigidBodyAnimNodeParameters}
 Rifle {type=number}
 Sim Space Anim {type=AnimPhysSimSpaceType}
 Skeletal_Mesh {type=SkeletalMeshComponent}
]
